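# Frontend lobby modal for Habanero mode that plays an animated message, subclassing HabaneroModeAnimatedMessage.
# OnPlayAnimation branches on message type: HabaneroEngaged plays the UpdateTransition animation, ReadyUp plays NewIntro.
# Each path first passes welcome, season and game mode text to the child Habanero_TransitionAnim widget, then binds animation-finished before playing.
# The two finish handlers unbind their animation's delegates, call HandleAnimationComplete and deactivate the modal.
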WidgetBlueprint /Game/UI/Frontend/Lobby/Habanero/HabaneroAnimatedMessageModal
graphs [
 EventGraph {nodes [
   n1 = Event OnPlayAnimation!() -> (
    OutputDelegate
    then
    InTopText:"Text (by ref)"
    InGameMode:"Text (by ref)"
    InSeasonName:"Text (by ref)"
    InMessageType:EHabaneroAnimatedMessage Enum=None
   ) {is=title K2Node_Event}
   > n2 = Branch(Condition="Equal (Enum)"(A:EHabaneroAnimatedMessage Enum=@n1.InMessageType, B:EHabaneroAnimatedMessage Enum=HabaneroEngaged))
   > SetupTransitionAnim!(
    execute
    self:Habanero Transition Anim Object Reference=Get Habanero_TransitionAnim() -> (Habanero_TransitionAnim:Habanero Transition Anim Object Reference)
    Text Welcome:Text=@n1.InTopText
    Text Season:Text=@n1.InSeasonName
    Text Game Mode:Text=@n1.InGameMode
   ) -> (then)
   n3 = Get Habanero_TransitionAnim()
   n4 = Get UpdateTransition!(self:Habanero Transition Anim Object Reference=@n3) -> (UpdateTransition:Widget Animation Object Reference)
   > BindToAnimationFinished(self=@n3, Animation=@n4, Delegate=Create Event())
   > PlayAnimation(self=@n3, InAnimation=@n4)
   Branch(execute=@n2.else, Condition="Equal (Enum)"(A=@n1.InMessageType, B=ReadyUp))
   > SetupIntroAnim!(
    execute
    self:Habanero Transition Anim Object Reference=Get Habanero_TransitionAnim()
    Text Welcome:Text=@n1.InTopText
    Test Season:Text=@n1.InSeasonName
    Text Game Mode:Text=@n1.InGameMode
   ) -> (then)
   n5 = Get Habanero_TransitionAnim()
   n6 = Get NewIntro!(self:Habanero Transition Anim Object Reference=@n5) -> (NewIntro:Widget Animation Object Reference)
   > BindToAnimationFinished(self=@n5, Animation=@n6, Delegate=Create Event())
   > PlayAnimation(self=@n5, InAnimation=@n6)
   Event PreConstruct()
   Event Construct()
   Event Tick()
  ]}
 Handle ReadyUp Animation Finished {nodes [
   Handle ReadyUp Animation Finished() {is=title K2Node_FunctionEntry}
   > HandleAnimationComplete(self:Habanero Mode Animated Message Object Reference)
   n1 = Get Habanero_TransitionAnim()
   > UnbindAllFromAnimationFinished!(execute, self:User Widget Object Reference=@n1, Animation:Widget Animation Object Reference=Get NewIntro(self=@n1)) -> (then)
   > DeactivateWidget()
  ]}
 Handle Transition Animation Finished {nodes [
   Handle Transition Animation Finished() {is=title K2Node_FunctionEntry}
   n1 = Get Habanero_TransitionAnim()
   > UnbindAllFromAnimationFinished(self=@n1, Animation=Get UpdateTransition(self=@n1))
   > HandleAnimationComplete()
   > DeactivateWidget()
  ]}
]
variables []
widgets {
 info {inheritedWidgetCount=0, parentClass=/Script/FortniteGame.HabaneroModeAnimatedMessage, rootWidgetClass=&Overlay, widgetCount=2}
 widgets [
  Overlay_0 {widgetClassPath=&Overlay, widget=Overlay_0}
  Habanero_TransitionAnim {
   widgetClassPath=&/Habanero_TransitionAnim
   bIsVariable=true
   parent=Overlay_0
   widget=Habanero_TransitionAnim
   slot {class=OverlaySlot, properties {horizontalAlignment=HAlign_Fill, verticalAlignment=VAlign_Fill}}
  }
 ]
}
