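# UI sound cue for switching Ranked on in the front end, routed to the Fort_UI sound class at reduced volume.
# A Mixer plays the Ranked toggle-on wave at full input level alongside a modulated Item Shop click branch, whose input volume is zero, so the click is silent.
# A concurrency override caps overlapping instances at two, stopping the oldest, with a short retrigger time.
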SoundCue /Game/Sounds/Fort_UI_Sounds/FrontEnd/New_FrontEnd/Ranked/Fortnite_UI_Ranked_On_Cue
root=Mixer_0
nodes [
 {id=Mixer_0, children [Modulator_0,WavePlayer_0], properties {inputVolume [0,1]}}
 {id=Modulator_0, children [WavePlayer_2], properties {volumeMin=1}}
 {id=WavePlayer_0, properties {soundWaveAssetPtr=/Game/Sounds/Fort_UI_Sounds/FrontEnd/New_FrontEnd/Ranked/wav/UI_Ranked_Select_Icon_Toggle_On}}
 {id=WavePlayer_2, properties {soundWaveAssetPtr=/Game/Sounds/Fort_UI_Sounds/FrontEnd/ItemShop/wav/Fort_UI_Shop_Click_Simple}}
]
properties {
 bOverrideConcurrency=true
 maxDistance=2097152
 soundClassObject=&Fort_UI
 volumeMultiplier=0.3
 concurrencyOverrides {maxCount=2, resolutionRule=StopOldest, retriggerTime=0.1}
}
