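# Lobby intro modal announcing that Ranked modes are unlocked for Battle Royale and Zero Build, with three bullet points, a badge graphic and engage/dismiss buttons.
# On activation it plays a whoosh sound and the staggered fade, slide and scale intro animation.
# GetCanEngageHabanero from the native parent shows or collapses the engage button and sets the dismiss label to "Maybe later" or "Ok!".
# PreConstruct mirrors the text and image columns and gradient flare shear for Arabic.
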WidgetBlueprint /Game/UI/Frontend/Lobby/Habanero/HabaneroIntroModal
animations [Anim_Intro {
  binding_count=10
  length_seconds=0.6
  object=&self:Anim_Intro
  sequence {
   bindings [
    Text_BP2 {
     id=#1
     possessed_class=GenericModalEntry_C
     tracks [
      {
       type=MovieSceneFloatTrack
       display_name=Render Opacity
       sections [{type=MovieSceneFloatSection, end_frame=33, start_frame=0, channels [None {type=MovieSceneScriptingFloatChannel, keys [[0,0],[18,0],[33,1]]}]}]
      }
      {
       type=MovieScene2DTransformTrack
       sections [{
         type=MovieScene2DTransformSection
         end_frame=33
         start_frame=0
         channels [
          Translation.X {type=MovieSceneScriptingFloatChannel}
          Translation.Y {type=MovieSceneScriptingFloatChannel, keys [[0,0],[18,72],[33,0]]}
          Angle {type=MovieSceneScriptingFloatChannel}
          Scale.X {type=MovieSceneScriptingFloatChannel}
          Scale.Y {type=MovieSceneScriptingFloatChannel}
          Shear.X {type=MovieSceneScriptingFloatChannel}
          Shear.Y {type=MovieSceneScriptingFloatChannel}
         ]
        }]
      }
     ]
    }
    Button_Engage {
     id=#2
     possessed_class=WBP_UIKit_Button_Regular_C
     tracks [{
       type=MovieScene2DTransformTrack
       sections [{
         type=MovieScene2DTransformSection
         end_frame=27
         start_frame=0
         channels [
          Translation.X {type=MovieSceneScriptingFloatChannel}
          Translation.Y {type=MovieSceneScriptingFloatChannel}
          Angle {type=MovieSceneScriptingFloatChannel}
          Scale.X {type=MovieSceneScriptingFloatChannel, keys [[0,0],[9,0],[15,1],[27,1]]}
          Scale.Y {type=MovieSceneScriptingFloatChannel, keys [[0,0],[9,0],[15,1],[27,1]]}
          Shear.X {type=MovieSceneScriptingFloatChannel}
          Shear.Y {type=MovieSceneScriptingFloatChannel}
         ]
        }]
      }]
    }
    Text_BP1 {
     id=#3
     possessed_class=GenericModalEntry_C
     tracks [
      {
       type=MovieSceneFloatTrack
       display_name=Render Opacity
       sections [{type=MovieSceneFloatSection, end_frame=33, start_frame=0, channels [None {type=MovieSceneScriptingFloatChannel, keys [[0,0],[15,0],[30,1]]}]}]
      }
      {
       type=MovieScene2DTransformTrack
       sections [{
         type=MovieScene2DTransformSection
         end_frame=33
         start_frame=0
         channels [
          Translation.X {type=MovieSceneScriptingFloatChannel}
          Translation.Y {type=MovieSceneScriptingFloatChannel, keys [[0,0],[15,72],[30,0]]}
          Angle {type=MovieSceneScriptingFloatChannel}
          Scale.X {type=MovieSceneScriptingFloatChannel}
          Scale.Y {type=MovieSceneScriptingFloatChannel}
          Shear.X {type=MovieSceneScriptingFloatChannel}
          Shear.Y {type=MovieSceneScriptingFloatChannel}
         ]
        }]
      }
     ]
    }
    Text_BP3 {
     id=#4
     possessed_class=GenericModalEntry_C
     tracks [
      {
       type=MovieSceneFloatTrack
       display_name=Render Opacity
       sections [{type=MovieSceneFloatSection, end_frame=36, start_frame=0, channels [None {type=MovieSceneScriptingFloatChannel, keys [[0,0],[21,0],[36,1]]}]}]
      }
      {
       type=MovieScene2DTransformTrack
       sections [{
         type=MovieScene2DTransformSection
         end_frame=36
         start_frame=0
         channels [
          Translation.X {type=MovieSceneScriptingFloatChannel}
          Translation.Y {type=MovieSceneScriptingFloatChannel, keys [[0,0],[21,72],[36,0]]}
          Angle {type=MovieSceneScriptingFloatChannel}
          Scale.X {type=MovieSceneScriptingFloatChannel}
          Scale.Y {type=MovieSceneScriptingFloatChannel}
          Shear.X {type=MovieSceneScriptingFloatChannel}
          Shear.Y {type=MovieSceneScriptingFloatChannel}
         ]
        }]
      }
     ]
    }
    GridPanel {
     id=#5
     possessed_class=GridPanel
     tracks [{
       type=MovieSceneFloatTrack
       display_name=Render Opacity
       sections [{type=MovieSceneFloatSection, end_frame=27, start_frame=0, channels [None {type=MovieSceneScriptingFloatChannel, keys [[0,0],[27,1]]}]}]
      }]
    }
    Text_Header {
     id=#6
     possessed_class=VerticalBox
     tracks [
      {
       type=MovieSceneFloatTrack
       display_name=Render Opacity
       sections [{type=MovieSceneFloatSection, end_frame=33, start_frame=0, channels [None {type=MovieSceneScriptingFloatChannel, keys [[0,0],[6,0],[21,1]]}]}]
      }
      {
       type=MovieScene2DTransformTrack
       sections [{
         type=MovieScene2DTransformSection
         end_frame=33
         start_frame=0
         channels [
          Translation.X {type=MovieSceneScriptingFloatChannel}
          Translation.Y {type=MovieSceneScriptingFloatChannel, keys [[0,0],[6,72],[21,0]]}
          Angle {type=MovieSceneScriptingFloatChannel}
          Scale.X {type=MovieSceneScriptingFloatChannel}
          Scale.Y {type=MovieSceneScriptingFloatChannel}
          Shear.X {type=MovieSceneScriptingFloatChannel}
          Shear.Y {type=MovieSceneScriptingFloatChannel}
         ]
        }]
      }
     ]
    }
    Button_CloseMobile {
     id=#7
     possessed_class=CloseButton_C
     tracks [
      {
       type=MovieSceneFloatTrack
       display_name=Render Opacity
       sections [{type=MovieSceneFloatSection, end_frame=33, start_frame=0, channels [None {type=MovieSceneScriptingFloatChannel, keys [[0,0],[6,0],[21,1]]}]}]
      }
      {
       type=MovieScene2DTransformTrack
       sections [{
         type=MovieScene2DTransformSection
         end_frame=21
         start_frame=0
         channels [
          Translation.X {type=MovieSceneScriptingFloatChannel}
          Translation.Y {type=MovieSceneScriptingFloatChannel, keys [[0,100],[6,100],[21,0]]}
          Angle {type=MovieSceneScriptingFloatChannel}
          Scale.X {type=MovieSceneScriptingFloatChannel}
          Scale.Y {type=MovieSceneScriptingFloatChannel}
          Shear.X {type=MovieSceneScriptingFloatChannel}
          Shear.Y {type=MovieSceneScriptingFloatChannel}
         ]
        }]
      }
     ]
    }
    Icon {
     id=#8
     possessed_class=Overlay
     tracks [
      {
       type=MovieSceneFloatTrack
       display_name=Render Opacity
       sections [{type=MovieSceneFloatSection, end_frame=33, start_frame=0, channels [None {type=MovieSceneScriptingFloatChannel, keys [[0,0],[15,0],[33,1]]}]}]
      }
      {
       type=MovieScene2DTransformTrack
       sections [{
         type=MovieScene2DTransformSection
         end_frame=33
         start_frame=0
         channels [
          Translation.X {type=MovieSceneScriptingFloatChannel}
          Translation.Y {type=MovieSceneScriptingFloatChannel, keys [[0,72],[15,72],[33,0]]}
          Angle {type=MovieSceneScriptingFloatChannel}
          Scale.X {type=MovieSceneScriptingFloatChannel}
          Scale.Y {type=MovieSceneScriptingFloatChannel}
          Shear.X {type=MovieSceneScriptingFloatChannel}
          Shear.Y {type=MovieSceneScriptingFloatChannel}
         ]
        }]
      }
     ]
    }
    Text_Title {
     id=#9
     possessed_class=CommonTextBlock
     tracks [
      {
       type=MovieSceneFloatTrack
       display_name=Render Opacity
       sections [{type=MovieSceneFloatSection, end_frame=33, start_frame=0, channels [None {type=MovieSceneScriptingFloatChannel, keys [[0,0],[12,0],[27,1]]}]}]
      }
      {
       type=MovieScene2DTransformTrack
       sections [{
         type=MovieScene2DTransformSection
         end_frame=33
         start_frame=0
         channels [
          Translation.X {type=MovieSceneScriptingFloatChannel}
          Translation.Y {type=MovieSceneScriptingFloatChannel, keys [[0,72],[12,72],[27,0]]}
          Angle {type=MovieSceneScriptingFloatChannel}
          Scale.X {type=MovieSceneScriptingFloatChannel}
          Scale.Y {type=MovieSceneScriptingFloatChannel}
          Shear.X {type=MovieSceneScriptingFloatChannel}
          Shear.Y {type=MovieSceneScriptingFloatChannel}
         ]
        }]
      }
     ]
    }
    Button_Dismiss {
     id=#10
     possessed_class=WBP_UIKit_Button_Regular_C
     tracks [{
       type=MovieScene2DTransformTrack
       sections [{
         type=MovieScene2DTransformSection
         end_frame=33
         start_frame=0
         channels [
          Translation.X {type=MovieSceneScriptingFloatChannel}
          Translation.Y {type=MovieSceneScriptingFloatChannel}
          Angle {type=MovieSceneScriptingFloatChannel}
          Scale.X {type=MovieSceneScriptingFloatChannel, keys [[0,0],[15,0],[21,1],[33,1]]}
          Scale.Y {type=MovieSceneScriptingFloatChannel, keys [[0,0],[15,0],[21,1],[33,1]]}
          Shear.X {type=MovieSceneScriptingFloatChannel}
          Shear.Y {type=MovieSceneScriptingFloatChannel}
         ]
        }]
      }]
    }
   ]
   display_rate {numerator=60}
  }
 }]
graphs [
 Arabic Alignment Fixes {nodes [
   Arabic Alignment Fixes() {is=title K2Node_FunctionEntry}
   n1 = "Equal, Case Insensitive (String)"(A=GetCurrentLanguage!() -> (ReturnValue:String), B=ar)
   n2 = Get GridPanel() -> (GridPanel:Grid Panel Object Reference)
   > SetColumnFill!(execute, self:Grid Panel Object Reference=@n2, ColumnIndex:int=1, Coefficient:float=SelectFloat(A=2, B=3, bPickA=@n1)) -> (then)
   > SetColumnFill(self=@n2, ColumnIndex=3, Coefficient=SelectFloat(A=3, B=2, bPickA=@n1))
   > SetColumn!(
    execute
    self:Grid Slot Object Reference=SlotAsGridSlot(Widget=Get TextContent() -> (TextContent:Vertical Box Object Reference))
    InColumn:int=SelectInt(A=3, B=1, bPickA=@n1)
   ) -> (then)
   > SetColumn(self=SlotAsGridSlot(Widget=Get ImageContent() -> (ImageContent:Overlay Object Reference)), InColumn=SelectInt(A=1, B=3, bPickA=@n1))
   n3 = Get GradientFlare() -> (GradientFlare:Image Object Reference)
   n4 = Break Widget Transform!(WidgetTransform:"Widget Transform Structure (by ref)"=Get RenderTransform!(self:Widget Object Reference=@n3) -> (RenderTransform:Widget Transform Structure)) -> (Translation:Vector 2D Structure, Scale:Vector 2D Structure, Shear:Vector 2D Structure, Angle:float=~) {is=title K2Node_BreakStruct}
   n5 = BreakVector2D(InVec=@n4.Shear)
   n6 = float * float(A=@n5, B=-1) {is=title K2Node_PromotableOperator}
   > SetRenderShear!(
    execute
    self:Widget Object Reference=@n3
    Shear:Vector 2D Structure=Select(Option 0:Vector 2D Structure=@n4.Shear, Option 1:Vector 2D Structure=MakeVector2D(X=@n6, Y=@n5.Y), Index=@n1) -> (ReturnValue:Vector 2D Structure)
   ) -> (then)
   n7 = Get GradientFlare_Bg() -> (GradientFlare_Bg:Image Object Reference)
   n8 = Break Widget Transform(WidgetTransform=Get RenderTransform(self=@n7)) -> (Angle=~)
   n9 = BreakVector2D(InVec=@n8.Shear)
   n10 = float * float(A=@n9, B=-1)
   > SetRenderShear(self=@n7, Shear=Select(Option 0=@n8.Shear, Option 1=MakeVector2D(X=@n10, Y=@n9.Y), Index=@n1))
  ]}
 EventGraph {nodes [
   Event PreConstruct()
   > Arabic Alignment Fixes() {is=title K2Node_CallFunction}
   Event On Activated()
   > PlaySound2D(Sound=Get IntroSoundCue() -> (IntroSoundCue:Sound Base Object Reference))
   > PlayAnimation(InAnimation=Get Anim_Intro() -> (Anim_Intro:Widget Animation Object Reference))
   n1 = GetCanEngageHabanero!(self:Habanero Intro Modal Object Reference) -> (ReturnValue:bool)
   > Set Visibility(
    Visibility=Select(Option 0:ESlateVisibility Enum=Collapsed, Option 1:ESlateVisibility Enum=SelfHitTestInvisible, Index=@n1) -> (ReturnValue:ESlateVisibility Enum)
    self=Get Button_Engage() -> (Button_Engage:Common Button Base Object Reference)
   )
   > n2 = Cast To WBP_UIKit_Button_Regular!(execute, Object=Get Button_Dismiss() -> (Button_Dismiss:Common Button Base Object Reference)) -> (then, CastFailed, AsWBP UIKit Button Regular:WBP UIKit Button Regular Object Reference) {is=title K2Node_DynamicCast}
   > SetText!(
    execute
    self:Fort CTAButton Object Reference=@n2
    ButtonText:"Text (by ref)"=Select(Option 0:Text="LOCTEXT(\"Ok!\")", Option 1:Text="LOCTEXT(\"Maybe later\")", Index=@n1) -> (ReturnValue:Text)
   ) -> (then)
   Event Tick()
  ]}
]
variables [IntroSoundCue {type=SoundBase, default=/Game/Sounds/Fort_UI_Sounds/FrontEnd/New_FrontEnd/Ranked/Fort_UI_Ranked_Whoosh_Intro_Cue}]
widgets {
 info {inheritedWidgetCount=3, parentClass=/Script/FortniteGame.HabaneroIntroModal, rootWidgetClass=&Overlay, widgetCount=34}
 widgets [
  Overlay_23 {widgetClassPath=&Overlay, widget=Overlay_23}
  WBP_UIKit_Scrim {
   widgetClassPath=/FortUILibrary/Systems/Dialog/WBP_UIKit_Scrim
   bIsVariable=true
   parent=Overlay_23
   widget=WBP_UIKit_Scrim
   properties {useClickCatcher=true}
   slot {class=OverlaySlot, properties {horizontalAlignment=HAlign_Fill, verticalAlignment=VAlign_Fill}}
  }
  Lightbox_HitTestBlocker_1 {
   widgetClassPath=&Image
   bIsVariable=true
   parent=Overlay_23
   widget=Lightbox_HitTestBlocker_1
   properties {brush {
     bIsDynamicallyLoaded=false
     drawAs=Image
     imageType=NoImage
     mirroring=NoMirror
     resourceName=None
     resourceObject=null
     tiling=NoTile
     imageSize=<32 32>
     margin=<ltrb 0 0 0 0>
     outlineSettings {
      bUseBrushTransparency=false
      roundingType=HalfHeightRadius
      width=0
      color {colorUseRule=UseColor_Specified, specifiedColor=<rgba 0 0 0 0>}
      cornerRadii=<0 0 0 0>
     }
     tintColor {colorUseRule=UseColor_Specified, specifiedColor=<rgba 0 0 0 0.75>}
     uVRegion {bIsValid=false, max=<0 0>, min=<0 0>}
    }}
   slot {class=OverlaySlot, properties {horizontalAlignment=HAlign_Fill, verticalAlignment=VAlign_Fill}}
  }
  MobileClose {
   widgetClassPath=&SafeZone
   bIsVariable=true
   parent=Overlay_23
   widget=MobileClose
   properties {padBottom=false, padLeft=false}
   slot {class=OverlaySlot, properties {horizontalAlignment=HAlign_Right, padding=<ltrb 0 32 32 0>}}
  }
  VisibilityBorder_Touch {
   widgetClassPath=&CommonHardwareVisibilityBorder
   parent=MobileClose
   widget=VisibilityBorder_Touch
   properties {visibilityQuery {
     autoDescription=" ANY( Input.Touch )"
     tokenStreamVersion=0
     userDescription=touch-only
     queryTokenStream [0,1,1,1,0]
     tagDictionary [{tagName=Input.Touch}]
    }}
   slot {class=SafeZoneSlot}
  }
  Button_CloseMobile {
   widgetClassPath=/FortUILibrary/Foundation/Buttons/CloseButton
   bInherited=true
   bIsVariable=true
   parent=VisibilityBorder_Touch
   widget=Button_CloseMobile
   properties {
    onButtonBaseClicked=(&self:WidgetTree.Button_CloseMobile.HandleOnButtonClicked)
    onButtonBaseDoubleClicked=(&self:WidgetTree.Button_CloseMobile.HandleOnButtonDoubleClicked)
    onButtonBaseHovered=(&self:WidgetTree.Button_CloseMobile.HandleOnButtonHovered)
    onButtonBaseUnhovered=(&self:WidgetTree.Button_CloseMobile.HandleOnButtonUnhovered)
    onSelectedChangedBase=(&self:WidgetTree.Button_CloseMobile.HandleOnSelectedChanged)
   }
   slot {class=BorderSlot, properties {padding=<ltrb 0 0 0 0>}}
  }
  SafeZone {
   widgetClassPath=&SafeZone
   bIsVariable=true
   parent=Overlay_23
   widget=SafeZone
   slot {class=OverlaySlot, properties {horizontalAlignment=HAlign_Fill, verticalAlignment=VAlign_Fill}}
  }
  GridPanel {
   widgetClassPath=&GridPanel
   bIsVariable=true
   parent=SafeZone
   widget=GridPanel
   properties {columnFill [0.75,3,0.1,2,0.5], rowFill [0.1,0,0.1]}
   slot {class=SafeZoneSlot}
  }
  PanelBg {
   widgetClassPath=&Image
   parent=GridPanel
   widget=PanelBg
   properties {
    brush {
     bIsDynamicallyLoaded=false
     drawAs=Image
     imageType=NoImage
     mirroring=NoMirror
     resourceName=None
     resourceObject=/Game/UI/Frontend/Lobby/Habanero/Habanero_Materials/MaterialInstances/Background/MI_UI_Habanero_IntroModal_Bg
     tiling=NoTile
     imageSize=<32 32>
     margin=<ltrb 0 0 0 0>
     outlineSettings {
      bUseBrushTransparency=false
      roundingType=HalfHeightRadius
      width=0
      color {colorUseRule=UseColor_Specified, specifiedColor=<rgba 0 0 0 0>}
      cornerRadii=<0 0 0 0>
     }
     tintColor {colorUseRule=UseColor_Specified, specifiedColor=<rgba 1 1 1 1>}
     uVRegion {bIsValid=false, max=<0 0>, min=<0 0>}
    }
    renderTransformPivot=<0.5 1>
   }
   slot {class=GridSlot, properties {columnSpan=5, row=1}}
  }
  TextContent {
   widgetClassPath=&VerticalBox
   bIsVariable=true
   parent=GridPanel
   widget=TextContent
   slot {class=GridSlot, properties {column=1, row=1}}
  }
  VBox {
   widgetClassPath=&VerticalBox
   bIsVariable=true
   parent=TextContent
   widget=VBox
   slot {class=VerticalBoxSlot, properties {padding=<ltrb 0 64 0 80>, size {sizeRule=Fill, value=1}}}
  }
  Text_Title {
   widgetClassPath=&CommonTextBlock
   parent=VBox
   widget=Text_Title
   properties {
    text=RANKED MODES
    textTransformPolicy=ToUpper
    font {
     bForceMonospaced=false
     bMaterialIsStencil=false
     fontMaterial=null
     fontObject=&Burbank
     letterSpacing=0
     monospacedWidth=1
     size=35
     skewAmount=0
     typefaceFontName=BurbankRegular-Black
     outlineSettings {
      bApplyOutlineToDropShadows=false
      bMiteredCorners=false
      bSeparateFillAlpha=false
      outlineMaterial=null
      outlineSize=0
      outlineColor=<rgba 1 1 1 0>
     }
    }
    shadowOffset=<0 0>
   }
   slot {class=VerticalBoxSlot, properties {padding=<ltrb 0 0 0 32>}}
  }
  VerticalBox_BulletPoints {widgetClassPath=&VerticalBox, parent=VBox, widget=VerticalBox_BulletPoints, slot {class=VerticalBoxSlot}}
  Text_BP1 {
   widgetClassPath=/Game/Athena/UI/GenericModalEntry
   bIsVariable=true
   parent=VerticalBox_BulletPoints
   widget=Text_BP1
   properties {text="<b>Complete Ranked Quests</> to earn rewards."}
   slot {class=VerticalBoxSlot, properties {size {sizeRule=Fill, value=1}}}
  }
  Text_BP2 {
   widgetClassPath=&GenericModalEntry
   bIsVariable=true
   parent=VerticalBox_BulletPoints
   widget=Text_BP2
   properties {text="<b>Rank up</> to unlock tournament play in the Compete tab."}
   slot {class=VerticalBoxSlot}
  }
  Text_BP3 {
   widgetClassPath=&GenericModalEntry
   bIsVariable=true
   parent=VerticalBox_BulletPoints
   widget=Text_BP3
   properties {text="Ranked modes follow <b>Tournament Rules</>, which may have different loot and settings than unranked."}
   slot {class=VerticalBoxSlot}
  }
  ImageContent {
   widgetClassPath=&Overlay
   bIsVariable=true
   parent=GridPanel
   widget=ImageContent
   properties {visibility=HitTestInvisible}
   slot {class=GridSlot, properties {column=3, row=1}}
  }
  Spacer_72 {
   widgetClassPath=&Spacer
   parent=ImageContent
   widget=Spacer_72
   properties {size=<512 512>}
   slot {class=OverlaySlot, properties {horizontalAlignment=HAlign_Fill, verticalAlignment=VAlign_Fill}}
  }
  GradientFlare_Bg {
   widgetClassPath=&Image
   bIsVariable=true
   parent=ImageContent
   widget=GradientFlare_Bg
   properties {
    brush {
     bIsDynamicallyLoaded=false
     drawAs=Image
     imageType=NoImage
     mirroring=NoMirror
     resourceName=None
     resourceObject=/Game/UI/Frontend/Lobby/Habanero/Habanero_Materials/MaterialInstances/Background/MI_UI_Habanero_IntroModal_FlareGradient_Secondary
     tiling=NoTile
     imageSize=<32 32>
     margin=<ltrb 0 0 0 0>
     outlineSettings {
      bUseBrushTransparency=false
      roundingType=HalfHeightRadius
      width=0
      color {colorUseRule=UseColor_Specified, specifiedColor=<rgba 0 0 0 0>}
      cornerRadii=<0 0 0 0>
     }
     tintColor {colorUseRule=UseColor_Specified, specifiedColor=<rgba 1 1 1 1>}
     uVRegion {bIsValid=false, max=<0 0>, min=<0 0>}
    }
    renderTransform {angle=0, scale=<1 1>, shear=<-32 0>, translation=<0 0>}
   }
   slot {class=OverlaySlot, properties {horizontalAlignment=HAlign_Fill, verticalAlignment=VAlign_Fill}}
  }
  GradientFlare {
   widgetClassPath=&Image
   bIsVariable=true
   parent=ImageContent
   widget=GradientFlare
   properties {
    brush {
     bIsDynamicallyLoaded=false
     drawAs=Image
     imageType=NoImage
     mirroring=NoMirror
     resourceName=None
     resourceObject=/Game/UI/Frontend/Lobby/Habanero/Habanero_Materials/MaterialInstances/Background/MI_UI_Habanero_IntroModal_FlareGradient
     tiling=NoTile
     imageSize=<32 32>
     margin=<ltrb 0 0 0 0>
     outlineSettings {
      bUseBrushTransparency=false
      roundingType=HalfHeightRadius
      width=0
      color {colorUseRule=UseColor_Specified, specifiedColor=<rgba 0 0 0 0>}
      cornerRadii=<0 0 0 0>
     }
     tintColor {colorUseRule=UseColor_Specified, specifiedColor=<rgba 1 1 1 1>}
     uVRegion {bIsValid=false, max=<0 0>, min=<0 0>}
    }
    renderTransform {angle=0, scale=<0.875 1>, shear=<-32 0>, translation=<0 0>}
   }
   slot {class=OverlaySlot, properties {horizontalAlignment=HAlign_Fill, verticalAlignment=VAlign_Fill}}
  }
  Icon {
   widgetClassPath=&Overlay
   parent=ImageContent
   widget=Icon
   properties {visibility=HitTestInvisible}
   slot {class=OverlaySlot, properties {horizontalAlignment=HAlign_Center, verticalAlignment=VAlign_Center, padding=<ltrb 0 0 0 8>}}
  }
  Spacer_315 {
   widgetClassPath=&Spacer
   parent=Icon
   widget=Spacer_315
   properties {size=<356 356>}
   slot {class=OverlaySlot, properties {horizontalAlignment=HAlign_Fill, verticalAlignment=VAlign_Fill}}
  }
  NextBadge_Bg {
   widgetClassPath=&Image
   bIsVariable=true
   parent=Icon
   widget=NextBadge_Bg
   properties {
    brush {
     bIsDynamicallyLoaded=false
     drawAs=Image
     imageType=NoImage
     mirroring=NoMirror
     resourceName=None
     resourceObject=/Game/UI/Frontend/Lobby/Habanero/Habanero_Materials/MaterialInstances/Icons/Background/MI_UI_Habanero_Background_08
     tiling=NoTile
     imageSize=<32 32>
     margin=<ltrb 0 0 0 0>
     outlineSettings {
      bUseBrushTransparency=false
      roundingType=HalfHeightRadius
      width=0
      color {colorUseRule=UseColor_Specified, specifiedColor=<rgba 0 0 0 0>}
      cornerRadii=<0 0 0 0>
     }
     tintColor {colorUseRule=UseColor_Specified, specifiedColor=<rgba 1 1 1 1>}
     uVRegion {bIsValid=false, max=<0 0>, min=<0 0>}
    }
    renderTransform {angle=0, scale=<1.25 1.25>, shear=<0 0>, translation=<0 0>}
   }
   slot {class=OverlaySlot, properties {horizontalAlignment=HAlign_Fill, verticalAlignment=VAlign_Fill}}
  }
  NextBadge_Icon {
   widgetClassPath=&Image
   bIsVariable=true
   parent=Icon
   widget=NextBadge_Icon
   properties {brush {
     bIsDynamicallyLoaded=false
     drawAs=Image
     imageType=NoImage
     mirroring=NoMirror
     resourceName=None
     resourceObject=/Game/UI/Frontend/Lobby/Habanero/Habanero_Materials/MaterialInstances/Icons/Foreground/MI_UI_Habanero_Foreground_08
     tiling=NoTile
     imageSize=<32 32>
     margin=<ltrb 0 0 0 0>
     outlineSettings {
      bUseBrushTransparency=false
      roundingType=HalfHeightRadius
      width=0
      color {colorUseRule=UseColor_Specified, specifiedColor=<rgba 0 0 0 0>}
      cornerRadii=<0 0 0 0>
     }
     tintColor {colorUseRule=UseColor_Specified, specifiedColor=<rgba 1 1 1 1>}
     uVRegion {bIsValid=false, max=<0 0>, min=<0 0>}
    }}
   slot {class=OverlaySlot, properties {horizontalAlignment=HAlign_Fill, verticalAlignment=VAlign_Fill}}
  }
  Overlay_0 {
   widgetClassPath=&Overlay
   parent=GridPanel
   widget=Overlay_0
   slot {class=GridSlot, properties {columnSpan=5, horizontalAlignment=HAlign_Center, row=1, verticalAlignment=VAlign_Bottom, nudge=<0 24>}}
  }
  WBP_CaptureForPostBufferUpdate {
   widgetClassPath=&WBP_CaptureForPostBufferUpdate
   bIsVariable=true
   parent=Overlay_0
   slot {class=OverlaySlot, properties {horizontalAlignment=HAlign_Fill, verticalAlignment=VAlign_Fill}}
  }
  HorizontalBox {
   widgetClassPath=&HorizontalBox
   parent=Overlay_0
   widget=HorizontalBox
   slot {class=OverlaySlot, properties {horizontalAlignment=HAlign_Center, verticalAlignment=VAlign_Center}}
  }
  Button_Engage {
   widgetClassPath=&WBP_UIKit_Button_Regular
   bInherited=true
   bIsVariable=true
   parent=HorizontalBox
   widget=Button_Engage
   properties {
    minHeight=56
    text="Try it now!"
    hoveredSlateSoundOverride {resourceObject=/Game/Sounds/Fort_UI_Sounds/FrontEnd/New_FrontEnd/Ranked/Fortnite_UI_Ranked_Button_Hover_Cue}
    pressedSlateSoundOverride {resourceObject=/Game/Sounds/Fort_UI_Sounds/FrontEnd/New_FrontEnd/Ranked/Fortnite_UI_Ranked_On_Cue}
   }
   slot {class=HorizontalBoxSlot, properties {horizontalAlignment=HAlign_Left, verticalAlignment=VAlign_Top}}
  }
  Spacer_103 {widgetClassPath=&Spacer, parent=HorizontalBox, widget=Spacer_103, properties {size=<64 1>}, slot {class=HorizontalBoxSlot}}
  Button_Dismiss {
   widgetClassPath=&WBP_UIKit_Button_Regular
   bInherited=true
   bIsVariable=true
   parent=HorizontalBox
   widget=Button_Dismiss
   properties {
    minHeight=56
    text=Maybe Later
    hoveredSlateSoundOverride {resourceObject=&Fortnite_UI_Ranked_Button_Hover_Cue}
    pressedSlateSoundOverride {resourceObject=/Game/Sounds/Fort_UI_Sounds/FrontEnd/New_FrontEnd/Ranked/Fortnite_UI_Ranked_Off_Cue}
   }
   slot {class=HorizontalBoxSlot, properties {horizontalAlignment=HAlign_Left, verticalAlignment=VAlign_Top}}
  }
  Image_BackgroundBottomBar {
   widgetClassPath=&Image
   parent=GridPanel
   widget=Image_BackgroundBottomBar
   properties {brush {
     bIsDynamicallyLoaded=false
     drawAs=Image
     imageType=NoImage
     mirroring=NoMirror
     resourceName=None
     resourceObject=/FortUILibrary/Foundation/Art_ToolKit/AngledBoxes/M_UI_MoreInfo_BottomDividerLIne_Blue
     tiling=NoTile
     imageSize=<32 26>
     margin=<ltrb 0 0 0 0>
     outlineSettings {
      bUseBrushTransparency=false
      roundingType=HalfHeightRadius
      width=0
      color {colorUseRule=UseColor_Specified, specifiedColor=<rgba 0 0 0 0>}
      cornerRadii=<0 0 0 0>
     }
     tintColor {colorUseRule=UseColor_Specified, specifiedColor=<rgba 0.3263 0.3457 0.4149 1>}
     uVRegion {bIsValid=false, max=<0 0>, min=<0 0>}
    }}
   slot {class=GridSlot, properties {columnSpan=5, row=1, verticalAlignment=VAlign_Bottom}}
  }
  Text_Header {
   widgetClassPath=&VerticalBox
   parent=GridPanel
   widget=Text_Header
   slot {class=GridSlot, properties {column=1, columnSpan=3, row=1, verticalAlignment=VAlign_Top, nudge=<0 -128>}}
  }
  Text_HeaderSub {
   widgetClassPath=&CommonTextBlock
   bIsVariable=true
   parent=Text_Header
   widget=Text_HeaderSub
   properties {
    mobileFontSizeMultiplier=1.4
    text="BATTLE ROYALE && ZERO BUILD"
    textTransformPolicy=ToUpper
    colorAndOpacity {colorUseRule=UseColor_Specified, specifiedColor=<rgba 0.0314 0.702 1 1>}
    font {
     bForceMonospaced=false
     bMaterialIsStencil=false
     fontMaterial=null
     fontObject=&Burbank
     letterSpacing=0
     monospacedWidth=1
     size=18
     skewAmount=0
     typefaceFontName=BurbankRegular-Black
     outlineSettings {
      bApplyOutlineToDropShadows=false
      bMiteredCorners=false
      bSeparateFillAlpha=false
      outlineMaterial=null
      outlineSize=0
      outlineColor=<rgba 0 0 0 1>
     }
    }
    margin=<ltrb 4 0 4 0>
    shadowOffset=<0 0>
   }
   slot {class=VerticalBoxSlot, properties {padding=<ltrb 0 0 0 4>}}
  }
  Text_HeaderMain {
   widgetClassPath=&CommonTextBlock
   bIsVariable=true
   parent=Text_Header
   widget=Text_HeaderMain
   properties {
    text=RANKED MODES UNLOCKED
    textTransformPolicy=ToUpper
    font {
     bForceMonospaced=false
     bMaterialIsStencil=false
     fontMaterial=null
     fontObject=&Burbank
     letterSpacing=0
     monospacedWidth=1
     size=42
     skewAmount=0
     typefaceFontName=BurbankRegular-Black
     outlineSettings {
      bApplyOutlineToDropShadows=false
      bMiteredCorners=false
      bSeparateFillAlpha=false
      outlineMaterial=null
      outlineSize=0
      outlineColor=<rgba 0 0 0 1>
     }
    }
    margin=<ltrb 4 0 4 0>
    shadowOffset=<0 0>
   }
   slot {class=VerticalBoxSlot}
  }
 ]
}
