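# Habanero (ranked) tier icon widget that layers soft-loaded foreground and background material images over a spacer that sets its size.
# DisplayBadge resolves tier data from rank, season, MMR and position through an override list, then UpdateWidgetDisplay applies large or small icons.
# Helper functions set icons from tier or progress data, desaturate the foreground while hiding the background, and read the gem glow color.
# PreConstruct applies size, background visibility and design-time debug icons.
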WidgetBlueprint /Game/UI/Frontend/Lobby/Habanero/WBP_Habanero_Icon
graphs [
 DesaturateForeground {nodes [
   n1 = DesaturateForeground!() -> (then, Desaturate:bool=~) {is=title K2Node_FunctionEntry}
   > n2 = GetDynamicMaterial(self=Get TierForeground() -> (TierForeground:Image Object Reference))
   > SetScalarParameterValue(self=@n2, ParameterName=Desaturate, Value="To Float (Boolean)"(InBool=@n1))
   n3 = NOT Boolean(A=Get UseBackground() -> (UseBackground:bool))
   > SetVisibility(
    self=Get TierBackground() -> (TierBackground:Image Object Reference)
    InVisibility=Select(Option 0=SelfHitTestInvisible, Option 1=Collapsed, Index=OR Boolean(A=@n1, B=@n3))
   )
  ]}
 EventGraph {nodes [
   n1 = SetIcon!() -> (OutputDelegate, then, ForegroundInstance:Material Instance Soft Object Reference, BackgroundInstance:Material Instance Soft Object Reference) {is=title K2Node_CustomEvent}
   > SetBrushFromSoftMaterial!(execute, self:Image Object Reference=Get TierForeground(), SoftMaterial:Material Interface Soft Object Reference=@n1.ForegroundInstance) -> (then)
   > SetBrushFromSoftMaterial(self=Get TierBackground(), SoftMaterial=@n1.BackgroundInstance)
   n2 = Event PreConstruct()
   > SetSize!(execute, self, IconSize:Vector 2D Structure=Get IconMinSize() -> (IconMinSize:Vector 2D Structure)) -> (then)
   > SetVisibility(self=Get TierBackground(), InVisibility=Select(Option 0=Collapsed, Option 1=SelfHitTestInvisible, Index=Get UseBackground()))
   > Branch(Condition=@n2.IsDesignTime)
   > UpdateDetails!(
    execute
    self
    ForegroundIcon:Material Instance Soft Object Reference=Get DebugForeground() -> (DebugForeground:Material Instance Soft Object Reference)
    BackgroundIcon:Material Instance Soft Object Reference=Get DebugBackground() -> (DebugBackground:Material Instance Soft Object Reference)
   ) -> (then)
   n3 = Event DisplayBadge!() -> (OutputDelegate, then, HabaneroType:String, CurrentHabanero:int=0, SeasonName:String, MatchmakingRating:int=0, CurrentPlayerPosition:int=0) {is=title K2Node_Event}
   n4 = GetHabaneroDisplayDataOverride!(
    InDisplayData:Fort Habanero Display Data Object Reference=Get HabaneroDisplayData() -> (HabaneroDisplayData:Fort Habanero Display Data Object Reference)
    InOverrideData:Fort Compete UIHabanero Tier Override List Object Reference=Get RankInfoOverrides() -> (RankInfoOverrides:Fort Compete UIHabanero Tier Override List Object Reference)
    InKey:"Fort Compete UIHabanero Tier Override Key Structure (by ref)"=Make Fort Compete UIHabanero Tier Override Key!(
     RankedTrack:Name=String To Name(InString=@n3.HabaneroType)
     Season:String=@n3.SeasonName
     Rank:int=@n3.CurrentHabanero
     MatchmakingRating:int=@n3.MatchmakingRating
     CurrentPosition:int=@n3.CurrentPlayerPosition
    ) -> (FortCompeteUIHabaneroTierOverrideKey:Fort Compete UIHabanero Tier Override Key Structure) {is=title K2Node_MakeStruct}
   ) -> (OutData:Fort Habanero Tier Override Structure, ReturnValue:bool)
   > n5 = Branch(Condition=@n4.ReturnValue)
   > UpdateWidgetDisplay!(execute, self:Fort Habanero Tier Widget Object Reference, DisplayData:"Fort Habanero Tier Structure (by ref)"=@n4) -> (then)
   FailedToFindRank(execute=@n5.else, self:Fort Habanero Tier Widget Object Reference)
   n6 = Event UpdateWidgetDisplay!() -> (OutputDelegate, then, DisplayData:"Fort Habanero Tier Structure (by ref)") {is=title K2Node_Event}
   > Set CurrentDisplayData(CurrentDisplayData:Fort Habanero Tier Structure=@n6.DisplayData) -> (Output_Get:Fort Habanero Tier Structure)
   n7 = Break Fort Habanero Tier!(FortHabaneroTier:"Fort Habanero Tier Structure (by ref)"=@n6.DisplayData) -> (
    ForegroundIcon:Material Instance Soft Object Reference
    ForegroundIconSmall:Material Instance Soft Object Reference
    BackgroundIcon:Material Instance Soft Object Reference
    BackgroundIconSmall:Material Instance Soft Object Reference
    WhiteIcon:Texture 2D Soft Object Reference
    SimpleIconSmall:Texture 2D Soft Object Reference
    Name:Text
    TierType:Fort Habanero Tier Type Data Object Reference
    AdditionalLeftIcon:Material Instance Soft Object Reference
    AdditionalRightIcon:Material Instance Soft Object Reference
   ) {is=title K2Node_BreakStruct}
   n8 = Get DisplayLargeIcon() -> (DisplayLargeIcon:bool)
   > UpdateDetails(
    ForegroundIcon=Select(Option 0:Material Instance Soft Object Reference=@n7.ForegroundIconSmall, Option 1:Material Instance Soft Object Reference=@n7, Index=@n8) -> (ReturnValue:Material Instance Soft Object Reference)
    BackgroundIcon=Select(Option 0=@n7.BackgroundIconSmall, Option 1=@n7.BackgroundIcon, Index=@n8)
   )
  ]}
 Get Tier Foreground Color {nodes [
   Get Tier Foreground Color() {is=title K2Node_FunctionEntry}
   > n1 = GetDynamicMaterial(self=Get TierForeground())
   > n2 = Get Vector Parameter Value!(execute, self:Material Instance Dynamic Object Reference=@n1, ParameterName:Name=Gem Glow Color) -> (then, ReturnValue:Linear Color Structure) {is=title K2Node_CallFunction}
   > Return Node!(execute, Color:Linear Color Structure=@n2)
  ]}
 SetSize {nodes [
   n1 = SetSize!() -> (then, IconSize:Vector 2D Structure) {is=title K2Node_FunctionEntry}
   > SetSize!(execute, self:Spacer Object Reference=Get IconSizeSpacer() -> (IconSizeSpacer:Spacer Object Reference), InSize:Vector 2D Structure=@n1) -> (then) {is=title K2Node_CallFunction}
  ]}
 UpdateDetails {nodes [
   n1 = UpdateDetails!() -> (then, ForegroundIcon:Material Instance Soft Object Reference, BackgroundIcon:Material Instance Soft Object Reference) {is=title K2Node_FunctionEntry}
   > SetBrushFromSoftMaterial(self=Get TierForeground(), SoftMaterial=@n1)
   > SetBrushFromSoftMaterial(self=Get TierBackground(), SoftMaterial=@n1.BackgroundIcon)
  ]}
 UpdateDetailsFromTierData {nodes [
   n1 = UpdateDetailsFromTierData!() -> (then, FortHabaneroTier:"Fort Habanero Tier Structure (by ref)") {is=title K2Node_FunctionEntry}
   n2 = Break Fort Habanero Tier(FortHabaneroTier=@n1)
   > UpdateDetails(ForegroundIcon=@n2, BackgroundIcon=@n2.BackgroundIcon) {is=title K2Node_CallFunction}
  ]}
 UpdateDetaislFromTierDataUsingRankedProgressData {nodes [
   n1 = UpdateDetaislFromTierDataUsingRankedProgressData!() -> (
    then
    RankedProgressData_RankType:String
    RankedProgressData_Rank:int
    RankedProgressData_HighestRank:int
    RankedProgressData_CurrentMMR:int
    RankedProgressData_CurrentPosition:int
   ) {is=title K2Node_FunctionEntry}
   n2 = GetHabaneroDisplayDataOverride(
    InDisplayData=Get HabaneroDisplayData()
    InOverrideData=Get RankInfoOverrides()
    InKey=Make Fort Compete UIHabanero Tier Override Key(
     RankedTrack=String To Name(InString=@n1)
     Rank=@n1.RankedProgressData_Rank
     MatchmakingRating=@n1.RankedProgressData_CurrentMMR
     CurrentPosition=@n1.RankedProgressData_CurrentPosition
    )
   )
   > UpdateDetailsFromTierData!(execute, self, FortHabaneroTier:"Fort Habanero Tier Structure (by ref)"=@n2) -> (then) {is=title K2Node_CallFunction}
   > Return Node!(execute, FoundTier:bool=@n2.ReturnValue)
  ]}
]
variables [
 CurrentDisplayData {
  type=FortHabaneroTier
  default {
   name=""
   additionalLeftIcon=null
   additionalRightIcon=null
   backgroundIcon=null
   backgroundIconSmall=null
   foregroundIcon=null
   foregroundIconSmall=null
   simpleIconSmall=null
   tierType=null
   whiteIcon=null
  }
 }
 DebugBackground {
  type=MaterialInstance
  default=/Game/UI/Frontend/Lobby/Habanero/Habanero_Materials/MaterialInstances/Icons/Background/MI_UI_Habanero_Background_00
 }
 DebugForeground {
  type=MaterialInstance
  default=/Game/UI/Frontend/Lobby/Habanero/Habanero_Materials/MaterialInstances/Icons/Foreground/MI_UI_Habanero_Foreground_00
 }
 DisplayLargeIcon {type=boolean}
 IconMinSize {type=Vector2D, default=<96 96>}
 RankInfoOverrides {type=FortCompeteUIHabaneroTierOverrideList, default=/Game/UI/Habanero/DA_RankDisplayDataOverride}
 UseBackground {type=boolean, default=true}
]
widgets {
 info {inheritedWidgetCount=0, parentClass=/Script/FortniteUI.FortHabaneroTierWidget, rootWidgetClass=&Overlay, widgetCount=4}
 widgets [
  Overlay_57 {widgetClassPath=&Overlay, widget=Overlay_57, properties {visibility=HitTestInvisible}}
  IconSizeSpacer {
   widgetClassPath=&Spacer
   bIsVariable=true
   parent=Overlay_57
   widget=IconSizeSpacer
   properties {size=<96 96>}
   slot {class=OverlaySlot, properties {horizontalAlignment=HAlign_Fill, verticalAlignment=VAlign_Fill}}
  }
  TierBackground {
   widgetClassPath=&Image
   bIsVariable=true
   parent=Overlay_57
   widget=TierBackground
   properties {
    visibility=HitTestInvisible
    brush {
     bIsDynamicallyLoaded=false
     drawAs=Image
     imageType=NoImage
     mirroring=NoMirror
     resourceName=None
     resourceObject=&MI_UI_Habanero_Background_00
     tiling=NoTile
     imageSize=<32 32>
     margin=<ltrb 0 0 0 0>
     outlineSettings {
      bUseBrushTransparency=false
      roundingType=HalfHeightRadius
      width=0
      color {colorUseRule=UseColor_Specified, specifiedColor=<rgba 0 0 0 0>}
      cornerRadii=<0 0 0 0>
     }
     tintColor {colorUseRule=UseColor_Specified, specifiedColor=<rgba 1 1 1 1>}
     uVRegion {bIsValid=false, max=<0 0>, min=<0 0>}
    }
    renderTransform {angle=0, scale=<1.25 1.25>, shear=<0 0>, translation=<0 0>}
   }
   slot {class=OverlaySlot, properties {horizontalAlignment=HAlign_Fill, verticalAlignment=VAlign_Fill}}
  }
  TierForeground {
   widgetClassPath=&Image
   bIsVariable=true
   parent=Overlay_57
   widget=TierForeground
   properties {
    visibility=HitTestInvisible
    brush {
     bIsDynamicallyLoaded=false
     drawAs=Image
     imageType=NoImage
     mirroring=NoMirror
     resourceName=None
     resourceObject=&MI_UI_Habanero_Foreground_00
     tiling=NoTile
     imageSize=<32 32>
     margin=<ltrb 0 0 0 0>
     outlineSettings {
      bUseBrushTransparency=false
      roundingType=HalfHeightRadius
      width=0
      color {colorUseRule=UseColor_Specified, specifiedColor=<rgba 0 0 0 0>}
      cornerRadii=<0 0 0 0>
     }
     tintColor {colorUseRule=UseColor_Specified, specifiedColor=<rgba 1 1 1 1>}
     uVRegion {bIsValid=false, max=<0 0>, min=<0 0>}
    }
   }
   slot {class=OverlaySlot, properties {horizontalAlignment=HAlign_Fill, verticalAlignment=VAlign_Fill}}
  }
 ]
}
