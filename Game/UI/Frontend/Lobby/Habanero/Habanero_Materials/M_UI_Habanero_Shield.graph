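# UI material that draws a translucent shield graphic for the Habanero frontend lobby, with glow and stripe highlights.
# An SDF texture with scale and position parameters supplies the shield in red and a black detail in green, both edged by MF_UI_AntiAliasing.
# An inner-border threshold layers the fill colors.
# A rotatable angular mask from arctangent UVs shapes the glow highlight, and a positionable slanted band adds the stripe highlight.
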Material /Game/UI/Frontend/Lobby/Habanero/Habanero_Materials/M_UI_Habanero_Shield
graph {
 properties {
  bUsedWithStaticMesh=false
  blendMode=BLEND_Translucent
  materialDomain=MD_UI
  pixelDepthOffsetMode=PDOM_Legacy
  usedShadingModels=Default Lit
 }
 nodes [
  n1 = NamedRerouteUsage()
  n2 = Multiply(A=@n1) {constB=10}
  n3 = OneMinus(None=Saturate(None=Subtract(A=@n1) {constB=0.04}))
  n4 = Multiply(A=@n3, B=@n3)
  n5 = Multiply(A=Max(A=Subtract(A=@n2) {constB=2.5}, B=OneMinus(None=@n4)) {constB=0}, B=ScalarParameter() {defaultValue=1, parameterName=GlowHighlight})
  n6 = Subtract(A=TextureCoordinate(), B=Constant2Vector() {g=0.5, r=0.5})
  n7 = Arctangent2Fast(Y=ComponentMask(None=@n6) {r=true}, X=ComponentMask(None=@n6) {g=true})
  n8 = Frac(None=Add(A=Multiply(A=@n7) {constB=0.3185}, B=ScalarParameter() {parameterName=GlowHighlightRotation}))
  n9 = Subtract(
   A=Min(A=OneMinus(None=@n8), B=Multiply(A=@n8) {constB=6})
   B=OneMinus(None=ScalarParameter() {defaultValue=0.5, parameterName=GlowHighlightSize})
  )
  n10 = Multiply(A=NamedRerouteUsage()) {constB=0.8}
  n11 = Saturate(None=Multiply(A=Add(A=@n9, B=@n10)) {constB=2.5})
  Max(A=NamedRerouteUsage(), B=Min(A=Saturate(None=@n5), B=NamedRerouteDeclaration(None=@n11) {name=GlowHighlight, nodeColor=<rgba 0.2188 0 1 1>}))
  n12 = Divide(A=Subtract(A=TextureCoordinate()) {constB=0.5}, B=ScalarParameter() {defaultValue=0.8, parameterName=Scale})
  n13 = ScalarParameter() {parameterName=PositionX}
  n14 = ScalarParameter() {parameterName=PositionY}
  n15 = TextureSample(UVs=Subtract(A=Add(A=@n12) {constB=0.5}, B=AppendVector(A=@n13, B=@n14))) {samplerType=SAMPLERTYPE_LinearColor, texture=/Game/UI/Frontend/Lobby/Habanero/Habanero_Textures/T_UI_SDF_Shield}
  NamedRerouteDeclaration(None=@n15.R) {name=ShieldShapeSDF, nodeColor=<rgba 0.2891 0 1 1>}
  n16 = MaterialFunctionCall(DistanceField=NamedRerouteUsage(), Threshold=Constant() {r=0.5}) {materialFunction=&MF_UI_AntiAliasing}
  OneMinus(None=@n16.Inner)
  n17 = NamedRerouteUsage()
  Saturate(None=Add(A=Constant3Vector() {constant=<rgba 0.8469 0.4072 0 1>}, B=Multiply(A=@n17) {constB=2}))
  n18 = LinearInterpolate(
   A=Constant3Vector() {constant=<rgba 1 1 1 0>}
   B=Constant3Vector() {constant=<rgba 0.5647 0.4179 0 1>}
   Alpha=NamedRerouteDeclaration(None=@n16.Outer) {name=ShieldShape, nodeColor=<rgba 0 1 0.5781 1>}
  )
  n19 = Constant3Vector() {constant=<rgba 0.9301 1 0.0103 1>}
  n20 = Subtract(A=NamedRerouteUsage(), B=ScalarParameter() {defaultValue=0.35, parameterName=InnerBorder})
  n21 = MaterialFunctionCall(DistanceField=Add(A=Saturate(None=@n20)) {constB=0.0078}, Threshold=Constant() {r=0.5}) {materialFunction=&MF_UI_AntiAliasing}
  n22 = Multiply(A=NamedRerouteUsage()) {constB=0.75}
  n23 = Saturate(None=Multiply(A=LinearInterpolate(A=@n18, B=@n19, Alpha=@n21.Outer), B=Add(B=@n22) {constA=1}))
  n24 = Constant3Vector() {constant=<rgba 0 0 0 1>}
  n25 = MaterialFunctionCall(DistanceField=@n15.G, Threshold=Constant() {r=0.5}) {materialFunction=&MF_UI_AntiAliasing}
  n26 = TextureCoordinate()
  n27 = ScalarParameter() {parameterName=StripeHighlightPosition}
  n28 = Add(
   A=Subtract(A=ComponentMask(None=@n26) {r=true}, B=Subtract(A=@n27) {constB=0.5}) {constB=0.5}
   B=Multiply(A=ComponentMask(None=@n26) {g=true}) {constB=-0.4}
  )
  n29 = OneMinus(None=Abs(None=Multiply(A=@n28) {constB=3}))
  Saturate(None=Add(A=LinearInterpolate(A=@n23, B=@n24, Alpha=@n25.Outer), B=Saturate(None=@n29)))
 ]
}
