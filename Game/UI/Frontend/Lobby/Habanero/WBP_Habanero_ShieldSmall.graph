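# Small Habanero shield widget for the frontend lobby: a single Image drawing the M_UI_Habanero_Shield material.
# PlayShieldResetedAnimation plays IntroSmall, which fades the shield in while animating material scalars for scale, glow, vertical position and a stripe highlight sweep.
# PreConstruct calls DebugPreview, which at design time only sets Scale and PositionY on the dynamic material to the animation's end pose.
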WidgetBlueprint /Game/UI/Frontend/Lobby/Habanero/WBP_Habanero_ShieldSmall
animations [IntroSmall {
  binding_count=1
  length_seconds=3.7704
  object=&self:IntroSmall
  sequence {
   bindings [Image_36 {
     id=#1
     possessed_class=Image
     tracks [
      {
       type=MovieSceneWidgetMaterialTrack
       sections [{
         type=MovieSceneParameterSection
         end_frame=113
         start_frame=0
         channels [
          GlowHighlightRotation {type=MovieSceneScriptingFloatChannel, keys [[15,0.3],[84,1.0189],[90,1.05]]}
          GlowHighlightSize {type=MovieSceneScriptingFloatChannel, keys [[15,0.5],[81,0.75],[84,1.1],[90,1.2]]}
          GlowHighlight {type=MovieSceneScriptingFloatChannel, keys [[15,1],[81,1],[84,1.33],[90,0]]}
          Scale {type=MovieSceneScriptingFloatChannel, keys [[15,0],[23,0.9],[25,0.9],[27,0.7],[81,0.7],[84,0.88],[90,0.55]]}
          StripeHighlightPosition {type=MovieSceneScriptingFloatChannel, keys [[88,0],[110,2]]}
          PositionY {type=MovieSceneScriptingFloatChannel, keys [[84,0],[90,0.2]]}
         ]
        }]
      }
      {
       type=MovieSceneFloatTrack
       display_name=Render Opacity
       sections [{
         type=MovieSceneFloatSection
         end_frame=113
         start_frame=0
         channels [None {type=MovieSceneScriptingFloatChannel, keys [[0,0],[15,0],[23,1],[110,1]]}]
        }]
      }
     ]
    }]
   display_rate {denominator=1001, numerator=30000}
  }
 }]
graphs [
 DebugPreview {nodes [
   n1 = DebugPreview!() -> (then, IsDesignTime:bool=~) {is=title K2Node_FunctionEntry}
   > Branch(Condition=@n1)
   > n2 = GetDynamicMaterial(self=Get Image_36() -> (Image_36:Image Object Reference))
   > SetScalarParameterValue(self=@n2, ParameterName=Scale, Value=0.55)
   > SetScalarParameterValue(self=@n2, ParameterName=PositionY, Value=0.2)
  ]}
 EventGraph {nodes [
   n1 = Event PreConstruct()
   > DebugPreview!(execute, self, IsDesignTime:bool=@n1.IsDesignTime) -> (then) {is=title K2Node_CallFunction}
   Event Construct()
   Event Tick()
  ]}
 PlayShieldResetedAnimation {nodes [
   PlayShieldResetedAnimation() {is=title K2Node_FunctionEntry}
   > PlayAnimation(InAnimation=Get IntroSmall() -> (IntroSmall:Widget Animation Object Reference))
  ]}
]
variables []
widgets {
 info {inheritedWidgetCount=0, parentClass=&UserWidget, rootWidgetClass=&Image, widgetCount=1}
 widgets [Image_36 {
   widgetClassPath=&Image
   bIsVariable=true
   widget=Image_36
   properties {brush {
     bIsDynamicallyLoaded=false
     drawAs=Image
     imageType=NoImage
     mirroring=NoMirror
     resourceName=None
     resourceObject=/Game/UI/Frontend/Lobby/Habanero/Habanero_Materials/M_UI_Habanero_Shield
     tiling=NoTile
     imageSize=<50 50>
     margin=<ltrb 0 0 0 0>
     outlineSettings {
      bUseBrushTransparency=false
      roundingType=HalfHeightRadius
      width=0
      color {colorUseRule=UseColor_Specified, specifiedColor=<rgba 0 0 0 0>}
      cornerRadii=<0 0 0 0>
     }
     tintColor {colorUseRule=UseColor_Specified, specifiedColor=<rgba 1 1 1 1>}
     uVRegion {bIsValid=false, max=<0 0>, min=<0 0>}
    }}
  }]
}
