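# Small solid rectangle widget for the hero rank UI in the Habanero lobby frontend, apparently a building block for a larger rank display.
# An Overlay stacks a Spacer, which sets the minimum size, under a textureless Image that fills the slot.
# Set Rectangle Color applies a given color to the Image's ColorAndOpacity.
# PreConstruct and Construct both call it with white, resetting the tint at design time and runtime.
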WidgetBlueprint /Game/UI/Frontend/Lobby/Habanero/WBP_HeroRank_Rectangle
graphs [
 EventGraph {nodes [
   n1 = Event PreConstruct()
   Set Rectangle Color!(execute=[@n1.then, @n2.then], self, Color:Linear Color Structure=(R=1,G=1,B=1,A=1)) -> (then) {is=title K2Node_CallFunction}
   n2 = Event Construct()
  ]}
 Set Rectangle Color {nodes [
   n1 = Set Rectangle Color!() -> (then, Color:Linear Color Structure) {is=title K2Node_FunctionEntry}
   > Set ColorAndOpacity!(
    execute
    ColorAndOpacity:Linear Color Structure=@n1
    self:Image Object Reference=Get Image_Rectangle() -> (Image_Rectangle:Image Object Reference)
   ) -> (then, Output_Get:Linear Color Structure)
  ]}
]
variables []
widgets {
 info {inheritedWidgetCount=0, parentClass=&UserWidget, rootWidgetClass=&Overlay, widgetCount=3}
 widgets [
  Overlay_0 {widgetClassPath=&Overlay, widget=Overlay_0}
  Spacer_42 {
   widgetClassPath=&Spacer
   parent=Overlay_0
   widget=Spacer_42
   properties {size=<4 8>}
   slot {class=OverlaySlot, properties {horizontalAlignment=HAlign_Fill, verticalAlignment=VAlign_Fill}}
  }
  Image_Rectangle {
   widgetClassPath=&Image
   bIsVariable=true
   parent=Overlay_0
   widget=Image_Rectangle
   properties {brush {
     bIsDynamicallyLoaded=false
     drawAs=Image
     imageType=NoImage
     mirroring=NoMirror
     resourceName=None
     resourceObject=null
     tiling=NoTile
     imageSize=<1 1>
     margin=<ltrb 0 0 0 0>
     outlineSettings {
      bUseBrushTransparency=false
      roundingType=HalfHeightRadius
      width=0
      color {colorUseRule=UseColor_Specified, specifiedColor=<rgba 0 0 0 0>}
      cornerRadii=<0 0 0 0>
     }
     tintColor {colorUseRule=UseColor_Specified, specifiedColor=<rgba 1 1 1 1>}
     uVRegion {bIsValid=false, max=<0 0>, min=<0 0>}
    }}
   slot {class=OverlaySlot, properties {horizontalAlignment=HAlign_Fill, verticalAlignment=VAlign_Fill}}
  }
 ]
}
