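# WBP_HabaneroProgressIcon is a lobby widget showing the player's ranked (Habanero) tier icon, circular progress and rank protection shield.
# HandleBinding functions take ranked view model values, resolve tier data through an override list, and drive the icon and progress material parameter.
# The tier name's Roman numeral sets how many tier-colored rectangles spawn.
# SetupRankVisual varies icon size and rectangle/progress visibility by tier; the unranked handler is empty.
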WidgetBlueprint /Game/UI/Frontend/Lobby/Habanero/WBP_HabaneroProgressIcon
graphs [
 EventGraph {nodes [
   Event Construct()
   > InitializeRankedViewModal(self:Ranked View Model Object Reference=Get FortRankedViewModel() -> (FortRankedViewModel:Ranked View Model Object Reference))
   > SetupRankVisual()
   Event PreConstruct()
   Event Tick()
  ]}
 HandleBinding_CurrentPlayerRank {nodes [
   n1 = HandleBinding_CurrentPlayerRank!() -> (then, CurrentRankIndex:int) {is=title K2Node_FunctionEntry}
   > n2 = Branch(Condition=Get Is Player Ranked() -> (Is Player Ranked:bool))
   n3 = Get FortRankedViewModel()
   n4 = Get RankType!(self:Ranked View Model Object Reference=@n3) -> (RankType:String)
   n5 = GetHabaneroDisplayDataOverride!(
    InDisplayData:Fort Habanero Display Data Object Reference=Get RankInfo() -> (RankInfo:Fort Habanero Display Data Object Reference)
    InOverrideData:Fort Compete UIHabanero Tier Override List Object Reference=Get Rank Info Overrides() -> (Rank Info Overrides:Fort Compete UIHabanero Tier Override List Object Reference)
    InKey:"Fort Compete UIHabanero Tier Override Key Structure (by ref)"=Make Fort Compete UIHabanero Tier Override Key!(
     RankedTrack:Name=String To Name(InString=@n4)
     Season:String
     Rank:int=@n1
     MatchmakingRating:int=Get MatchmakingRating!(self:Ranked View Model Object Reference=@n3) -> (MatchmakingRating:int=0)
     CurrentPosition:int=Get CurrentPlayerPosition!(self:Ranked View Model Object Reference=@n3) -> (CurrentPlayerPosition:int=0)
    ) -> (FortCompeteUIHabaneroTierOverrideKey:Fort Compete UIHabanero Tier Override Key Structure) {is=title K2Node_MakeStruct}
   ) -> (OutData:Fort Habanero Tier Override Structure, ReturnValue:bool)
   > Branch(Condition=@n5.ReturnValue)
   > Branch()
   n6 = Break Fort Habanero Tier!(FortHabaneroTier:"Fort Habanero Tier Structure (by ref)"=@n5) -> (
    ForegroundIcon:Material Instance Soft Object Reference
    ForegroundIconSmall:Material Instance Soft Object Reference
    BackgroundIcon:Material Instance Soft Object Reference
    BackgroundIconSmall:Material Instance Soft Object Reference
    WhiteIcon:Texture 2D Soft Object Reference
    SimpleIconSmall:Texture 2D Soft Object Reference
    Name:Text
    TierType:Fort Habanero Tier Type Data Object Reference
    AdditionalLeftIcon:Material Instance Soft Object Reference
    AdditionalRightIcon:Material Instance Soft Object Reference
   ) {is=title K2Node_BreakStruct}
   > Branch(Condition=IsValid(Object=@n6.TierType))
   > Set Current Category Data(Current Category Data:Fort Habanero Tier Type Data Object Reference=@n6.TierType) -> (Output_Get:Fort Habanero Tier Type Data Object Reference)
   > Set CurrentTierCategory(CurrentTierCategory:EHabaneroTierType Enum=Get Category!(self:Fort Habanero Tier Type Data Object Reference=@n6.TierType) -> (Category:EHabaneroTierType Enum=HabaneroSauce0)) -> (Output_Get:EHabaneroTierType Enum=HabaneroSauce0)
   > UpdateDetails!(
    execute
    self:WBP Habanero Icon Object Reference=Get WBP_Habanero_Icon() -> (WBP_Habanero_Icon:WBP Habanero Icon Object Reference)
    ForegroundIcon:Material Instance Soft Object Reference=@n6.ForegroundIconSmall
    BackgroundIcon:Material Instance Soft Object Reference=@n6.BackgroundIconSmall
   ) -> (then)
   > SetRankRactangle!(execute, self, RankName:Text=@n6.Name) -> (then)
   > SetupRankVisual()
   SetDefaultPlayerUnRank(execute=@n2.else)
  ]}
 HandleBinding_RankCircleProgress {nodes [
   n1 = HandleBinding_RankCircleProgress!() -> (then, Progress:double=~) {is=title K2Node_FunctionEntry}
   > Set ProgressValue(ProgressValue:double=@n1) -> (Output_Get:double)
   > n2 = GetDynamicMaterial(self=Get RankProgressBar() -> (RankProgressBar:Image Object Reference))
   > SetScalarParameterValue(self=@n2, ParameterName=Get ProgressParameter() -> (ProgressParameter:Name=None), Value=Get ProgressValue() -> (ProgressValue:double))
  ]}
 HandleBinding_RankType {nodes [
   n1 = HandleBinding_RankType!() -> (then, RankType:String) {is=title K2Node_FunctionEntry}
   > Set Rank Type(Rank Type:String=@n1) -> (Output_Get:String)
   > n2 = Get!(execute:Exec) -> (then:Exec, else:Exec, FortRankedViewModel:Ranked View Model Object Reference)
   > HandleBinding_CurrentPlayerRank!(execute, self, CurrentRankIndex:int=Get CurrentRank!(self:Ranked View Model Object Reference=@n2) -> (CurrentRank:int=0)) -> (then) {is=title K2Node_CallFunction}
  ]}
 HandleBinding_SetHasRankProtection {nodes [
   n1 = HandleBinding_SetHasRankProtection!() -> (then, HasProtection:bool=~) {is=title K2Node_FunctionEntry}
   > n2 = Branch(Condition=@n1)
   n3 = Get WBP_Habanero_ShieldSmall() -> (WBP_Habanero_ShieldSmall:WBP Habanero Shield Small Object Reference)
   > SetVisibility(self=@n3, InVisibility=HitTestInvisible)
   > PlayShieldResetedAnimation(self:WBP Habanero Shield Small Object Reference=@n3)
   SetVisibility(execute=@n2.else, self=@n3, InVisibility=Collapsed)
  ]}
 HandleBinding_SetIsPlayerRanked {nodes [
   n1 = HandleBinding_SetIsPlayerRanked!() -> (then, IsPlayerRanked:bool=~) {is=title K2Node_FunctionEntry}
   > Set Is Player Ranked(Is Player Ranked:bool=@n1) -> (Output_Get:bool)
  ]}
 SetDefaultPlayerUnRank {nodes [SetDefaultPlayerUnRank() {is=title K2Node_FunctionEntry}]}
 SetRankRactangle {nodes [
   n1 = SetRankRactangle!() -> (then, RankName:Text) {is=title K2Node_FunctionEntry}
   > SetupRankRectangle!(execute, self, ArabicNum:String=Split(SourceString="To String (Text)"(InText=@n1), InStr=" ", SearchDir=FromStart).RightS) -> (then)
  ]}
 SetupRankRectangle {nodes [
   n1 = SetupRankRectangle!() -> (then, ArabicNum:String) {is=title K2Node_FunctionEntry}
   > Clear(TargetArray:Array of WBP Hero Rank Rectangle Object References=Get RectangleArray() -> (RectangleArray:Array of WBP Hero Rank Rectangle Object References))
   > Reset!(
    execute
    self:Dynamic Entry Box Object Reference=Get DynamicEntryBox() -> (DynamicEntryBox:Dynamic Entry Box Object Reference)
    bDeleteWidgets:bool=true
   ) -> (then)
   n2 = Get RectangleMap() -> (RectangleMap:Map of Strings to Integers)
   > For Loop(LastIndex=int - int(A=Find(TargetMap:Map of Strings to Integers=@n2, Key=@n1) -> (Value:int)))
   > n3 = Create Entry!(execute, self:Dynamic Entry Box Object Reference=Get DynamicEntryBox()) -> (then, ReturnValue:User Widget Object Reference)
   > n4 = Cast To WBP_HeroRank_Rectangle!(execute, Object=@n3) -> (then, CastFailed, AsWBP Hero Rank Rectangle:WBP Hero Rank Rectangle Object Reference) {is=title K2Node_DynamicCast}
   > Add(
    TargetArray:Array of WBP Hero Rank Rectangle Object References=Get RectangleArray()
    NewItem:"WBP Hero Rank Rectangle Object Reference (by ref)"=@n4
   ) {is=title K2Node_CallArrayFunction}
   > n5 = Cast To BP_UI_Habanero_TierTypeDataBase!(execute, Object=Get Current Category Data() -> (Current Category Data:Fort Habanero Tier Type Data Object Reference)) -> (then, CastFailed, AsBP UI Habanero Tier Type Data Base:BP UI Habanero Tier Type Data Base Object Reference) {is=title K2Node_DynamicCast}
   > Set Rectangle Color!(
    execute
    self:WBP Hero Rank Rectangle Object Reference=@n4
    Color:Linear Color Structure="To LinearColor (Color)"!(InColor:Color Structure=Get Text_Color!(self:BP UI Habanero Tier Type Data Base Object Reference=@n5) -> (Text_Color:Color Structure)) -> (ReturnValue:Linear Color Structure)
   ) -> (then) {is=title K2Node_CallFunction}
   Get Tier Foreground Color!(self:WBP Habanero Icon Object Reference=Get WBP_Habanero_Icon()) -> (Color:Linear Color Structure) {is=title K2Node_CallFunction}
  ]}
 SetupRankVisual {nodes [
   SetupRankVisual() {is=title K2Node_FunctionEntry}
   n1 = Get SmallSize() -> (SmallSize:Vector 2D Structure)
   n2 = Get MediumSize() -> (MediumSize:Vector 2D Structure)
   > SetSize!(
    execute
    self:WBP Habanero Icon Object Reference=Get WBP_Habanero_Icon()
    IconSize:Vector 2D Structure=Select!(
     HabaneroSauce0:Vector 2D Structure=@n2
     HabaneroSauce1:Vector 2D Structure=@n1
     HabaneroSauce2:Vector 2D Structure=@n1
     HabaneroSauce3:Vector 2D Structure=@n1
     HabaneroSauce4:Vector 2D Structure=@n1
     HabaneroSauce5:Vector 2D Structure=@n1
     HabaneroSauce6:Vector 2D Structure=@n2
     HabaneroSauce7:Vector 2D Structure=@n2
     HabaneroSauce8:Vector 2D Structure=Get LargeSize() -> (LargeSize:Vector 2D Structure)
     Index:EHabaneroTierType Enum=Get CurrentTierCategory() -> (CurrentTierCategory:EHabaneroTierType Enum=HabaneroSauce0)
    ) -> (ReturnValue:Vector 2D Structure)
   ) -> (then)
   n3 = Get CurrentTierCategory() -> (CurrentTierCategory=HabaneroSauce0)
   > SetVisibility(
    self=Get DynamicEntryBox()
    InVisibility=Select!(
     Option 0:ESlateVisibility Enum=SelfHitTestInvisible
     Option 1:ESlateVisibility Enum=Collapsed
     Index:bool=Select!(
      HabaneroSauce0:bool=true
      HabaneroSauce1:bool=false
      HabaneroSauce2:bool=false
      HabaneroSauce3:bool=false
      HabaneroSauce4:bool=false
      HabaneroSauce5:bool=false
      HabaneroSauce6:bool=true
      HabaneroSauce7:bool=true
      HabaneroSauce8:bool=true
      Index:EHabaneroTierType Enum=@n3
     ) -> (ReturnValue:bool)
    ) -> (ReturnValue:ESlateVisibility Enum)
   )
   n4 = Get CurrentTierCategory() -> (CurrentTierCategory=HabaneroSauce0)
   > SetVisibility(
    self=Get RankProgressBar()
    InVisibility=Select!(
     Option 0:ESlateVisibility Enum=SelfHitTestInvisible
     Option 1:ESlateVisibility Enum=Collapsed
     Index:bool=Select!(
      HabaneroSauce0:bool=false
      HabaneroSauce1:bool=false
      HabaneroSauce2:bool=false
      HabaneroSauce3:bool=false
      HabaneroSauce4:bool=false
      HabaneroSauce5:bool=false
      HabaneroSauce6:bool=false
      HabaneroSauce7:bool=false
      HabaneroSauce8:bool=true
      Index:EHabaneroTierType Enum=@n4
     ) -> (ReturnValue:bool)
    ) -> (ReturnValue:ESlateVisibility Enum)
   )
  ]}
]
variables [
 Current Category Data {type=FortHabaneroTierTypeData, default=null}
 CurrentTierCategory {type=EHabaneroTierType, default=HabaneroSauce0}
 Is Player Ranked {type=boolean}
 ProgressParameter {type=string, default=None}
 ProgressValue {type=number}
 Rank Info Overrides {type=FortCompeteUIHabaneroTierOverrideList, default=/Game/UI/Habanero/DA_RankDisplayDataOverride}
 Rank Type {type=string, default=""}
 RankInfo {type=FortHabaneroDisplayData, default=/Game/UI/Habanero/HabaneroDisplayData}
 RectangleArray {type="array<WBP_HeroRank_Rectangle_C>", default []}
 RectangleMap {type="map<integer>", default {I=1, II=2, III=3, IV=4, V=5}}
]
widgets {
 info {inheritedWidgetCount=0, parentClass=&CommonUserWidget, rootWidgetClass=&GridPanel, widgetCount=9}
 widgets [
  GridPanel_25 {widgetClassPath=&GridPanel, widget=GridPanel_25, properties {columnFill [1], rowFill [1]}}
  FortVisualAttachment_0 {
   widgetClassPath=&FortVisualAttachment
   parent=GridPanel_25
   widget=FortVisualAttachment_0
   properties {bOverride_HeightOverride=true, bOverride_WidthOverride=true, heightOverride=144, widthOverride=144, contentAnchor=<0.5 0.5>}
   slot {class=GridSlot, properties {horizontalAlignment=HAlign_Center, verticalAlignment=VAlign_Center}}
  }
  RankProgressBar {
   widgetClassPath=&Image
   bIsVariable=true
   parent=FortVisualAttachment_0
   widget=RankProgressBar
   properties {brush {
     bIsDynamicallyLoaded=false
     drawAs=Image
     imageType=NoImage
     mirroring=NoMirror
     resourceName=None
     resourceObject=/Game/UI/Frontend/Lobby/Habanero/Habanero_Materials/MaterialInstances/ProgressBars/MI_UI_Habanero_ProgressBar_RankLevel
     tiling=NoTile
     imageSize=<32 32>
     margin=<ltrb 0 0 0 0>
     outlineSettings {
      bUseBrushTransparency=false
      roundingType=HalfHeightRadius
      width=0
      color {colorUseRule=UseColor_Specified, specifiedColor=<rgba 0 0 0 0>}
      cornerRadii=<0 0 0 0>
     }
     tintColor {colorUseRule=UseColor_Specified, specifiedColor=<rgba 1 1 1 1>}
     uVRegion {bIsValid=false, max=<0 0>, min=<0 0>}
    }}
   slot {class=SizeBoxSlot}
  }
  RadialDropShadow {
   widgetClassPath=&Image
   bIsVariable=true
   parent=GridPanel_25
   widget=RadialDropShadow
   properties {brush {
     bIsDynamicallyLoaded=false
     drawAs=Image
     imageType=NoImage
     mirroring=NoMirror
     resourceName=None
     resourceObject=/Game/UI/Frontend/Lobby/Habanero/Habanero_Materials/MaterialInstances/Background/MI_UI_Habanero_RadialDropShadow_Black40
     tiling=NoTile
     imageSize=<32 32>
     margin=<ltrb 0 0 0 0>
     outlineSettings {
      bUseBrushTransparency=false
      roundingType=HalfHeightRadius
      width=0
      color {colorUseRule=UseColor_Specified, specifiedColor=<rgba 0 0 0 0>}
      cornerRadii=<0 0 0 0>
     }
     tintColor {colorUseRule=UseColor_Specified, specifiedColor=<rgba 1 1 1 1>}
     uVRegion {bIsValid=false, max=<0 0>, min=<0 0>}
    }}
   slot {class=GridSlot}
  }
  VerticalBox_Level {
   widgetClassPath=&VerticalBox
   parent=GridPanel_25
   widget=VerticalBox_Level
   slot {class=GridSlot, properties {horizontalAlignment=HAlign_Center, layer=1, verticalAlignment=VAlign_Center}}
  }
  SizeBox_TEMP {
   widgetClassPath=&SizeBox
   parent=VerticalBox_Level
   widget=SizeBox_TEMP
   properties {bOverride_HeightOverride=true, bOverride_WidthOverride=true, heightOverride=64, widthOverride=64}
   slot {class=VerticalBoxSlot}
  }
  WBP_Habanero_Icon {
   widgetClassPath=&/WBP_Habanero_Icon
   bIsVariable=true
   parent=SizeBox_TEMP
   widget=WBP_Habanero_Icon
   properties {iconMinSize=<64 64>}
   slot {class=SizeBoxSlot}
  }
  DynamicEntryBox {
   widgetClassPath=&DynamicEntryBox
   bIsVariable=true
   parent=VerticalBox_Level
   widget=DynamicEntryBox
   properties {
    entryWidgetClass=&/WBP_HeroRank_Rectangle.WBP_HeroRank_Rectangle_C
    entrySpacing=<4 0>
    renderTransform {angle=0, scale=<1 1>, shear=<0 0>, translation=<0 -2>}
   }
   slot {class=VerticalBoxSlot, properties {horizontalAlignment=HAlign_Center, verticalAlignment=VAlign_Top, padding=<ltrb 0 0 0 4>}}
  }
  WBP_Habanero_ShieldSmall {
   widgetClassPath=&/WBP_Habanero_ShieldSmall
   bIsVariable=true
   parent=GridPanel_25
   widget=WBP_Habanero_ShieldSmall
   properties {visibility=Collapsed, renderTransform {angle=0, scale=<1 1>, shear=<0 0>, translation=<28 22>}}
   slot {class=GridSlot, properties {horizontalAlignment=HAlign_Right, layer=2, verticalAlignment=VAlign_Bottom}}
  }
 ]
}
