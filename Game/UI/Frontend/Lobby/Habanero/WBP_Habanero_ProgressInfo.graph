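# Habanero tier progress widget that layers a tier icon over a material-driven progress bar.
# Tier setters update the icon from tier or rank data, pick the bar material and colors by tier type, and hide the bar and enlarge the icon for HabaneroSauce8.
# SetProgress drives the material's Progress parameter.
# Elimination and placement bonus animations flash bar colors and animate badge and glow images, which sit in a collapsed overlay.
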WidgetBlueprint /Game/UI/Frontend/Lobby/Habanero/WBP_Habanero_ProgressInfo
animations [
 Anim_ElimBonus {
  binding_count=4
  length_seconds=1
  object=&self:Anim_ElimBonus
  sequence {
   bindings [
    TierForeground {
     id=#1
     possessed_class=Image
     tracks [
      {
       type=MovieSceneWidgetMaterialTrack
       sections [{
         type=MovieSceneParameterSection
         end_frame=52
         start_frame=0
         channels [
          Badge Hotness {type=MovieSceneScriptingFloatChannel, keys [[0,0],[12,0],[18.489,1],[52,0]]}
          Gem ColorR {type=MovieSceneScriptingFloatChannel, keys [[0,0],[12,1],[18.489,1],[52,0]]}
          Gem ColorG {type=MovieSceneScriptingFloatChannel, keys [[0,0.1145],[12,0],[18.489,1],[52,0.1145]]}
          Gem ColorB {type=MovieSceneScriptingFloatChannel, keys [[0,1],[12,0],[18.489,1],[52,1]]}
          Gem ColorA {type=MovieSceneScriptingFloatChannel, const [1,4]}
          Secondary Gem ColorR {type=MovieSceneScriptingFloatChannel, keys [[0,0],[12,1],[19,1],[52,0]]}
          Secondary Gem ColorG {type=MovieSceneScriptingFloatChannel, keys [[0,0.0548],[12,0],[19,1],[52,0.0548]]}
          Secondary Gem ColorB {type=MovieSceneScriptingFloatChannel, keys [[0,0.5625],[12,0],[19,1],[52,0.5625]]}
          Secondary Gem ColorA {type=MovieSceneScriptingFloatChannel, const [1,3]}
         ]
        }]
      }
      {type=MovieScene2DTransformTrack, sections []}
     ]
    }
    TierBackground {
     id=#2
     possessed_class=Image
     tracks [
      {
       type=MovieSceneWidgetMaterialTrack
       sections [{
         type=MovieSceneParameterSection
         end_frame=60
         start_frame=6
         channels [
          Outer ColorR {type=MovieSceneScriptingFloatChannel, const [0,1]}
          Outer ColorG {type=MovieSceneScriptingFloatChannel, const [0.0508,1]}
          Outer ColorB {type=MovieSceneScriptingFloatChannel, const [0.2083,1]}
          Outer ColorA {type=MovieSceneScriptingFloatChannel, const [1,1]}
         ]
        }]
      }
      {
       type=MovieSceneFloatTrack
       display_name=Render Opacity
       sections [{
         type=MovieSceneFloatSection
         end_frame=60
         start_frame=6
         channels [None {type=MovieSceneScriptingFloatChannel, keys [[6,1],[19.745,0],[51.164,0],[59.999,1]]}]
        }]
      }
     ]
    }
    Image_Progress {
     id=#3
     possessed_class=Image
     tracks [{
       type=MovieSceneWidgetMaterialTrack
       sections [{
         type=MovieSceneParameterSection
         end_frame=60
         start_frame=0
         channels [
          Glow Range {type=MovieSceneScriptingFloatChannel, keys [[0,0.17],[10.13,0],[16.364,0.17],[50.649,0],[60,0.17]]}
          Bar ColorR {type=MovieSceneScriptingFloatChannel, keys [[0,0.0075],[13.247,0.6],[41.247,0.6],[60,0.0075]]}
          Bar ColorG {type=MovieSceneScriptingFloatChannel, keys [[0,0.4233],[13.247,0],[41.247,0],[60,0.4233]]}
          Bar ColorB {type=MovieSceneScriptingFloatChannel, keys [[0,0.8632],[13.247,0],[41.247,0],[60,0.8632]]}
          Bar ColorA {type=MovieSceneScriptingFloatChannel, const [1,2]}
          Bar Glow ColorR {type=MovieSceneScriptingFloatChannel, keys [[0,0],[13.247,0.7391],[41.247,0.05],[60,0]]}
          Bar Glow ColorG {type=MovieSceneScriptingFloatChannel, keys [[0,0.3014],[13.247,0],[41.247,0],[60,0.3014]]}
          Bar Glow ColorB {type=MovieSceneScriptingFloatChannel, keys [[0,1],[13.247,0],[41.247,0],[60,1]]}
          Bar Glow ColorA {type=MovieSceneScriptingFloatChannel, const [1,4]}
          Bar Outline ColorR {type=MovieSceneScriptingFloatChannel, keys [[0,0.0622],[13.247,5],[41.247,5],[60,0.0622]]}
          Bar Outline ColorG {type=MovieSceneScriptingFloatChannel, keys [[0,0.8521],[13.247,0],[41.247,0],[60,0.8521]]}
          Bar Outline ColorB {type=MovieSceneScriptingFloatChannel, keys [[0,0.9948],[13.247,0],[41.247,0],[60,0.9948]]}
          Bar Outline ColorA {type=MovieSceneScriptingFloatChannel, const [1,4]}
          Progress Background ColorR {type=MovieSceneScriptingFloatChannel, keys [[0,0.0012],[13.247,0.1],[41.247,0.1],[60,0.0012]]}
          Progress Background ColorG {type=MovieSceneScriptingFloatChannel, keys [[0,0.0123],[13.247,0],[41.247,0],[60,0.0123]]}
          Progress Background ColorB {type=MovieSceneScriptingFloatChannel, keys [[0,0.0513],[13.247,0],[41.247,0],[60,0.0513]]}
          Progress Background ColorA {type=MovieSceneScriptingFloatChannel, const [1,4]}
          Inner ColorR {type=MovieSceneScriptingFloatChannel, keys [[0,0.0021],[13.247,0.02],[41.247,0.02],[60,0.0021]]}
          Inner ColorG {type=MovieSceneScriptingFloatChannel, keys [[0,0.0185],[13.247,0],[41.247,0],[60,0.0185]]}
          Inner ColorB {type=MovieSceneScriptingFloatChannel, keys [[0,0.0648],[13.247,0],[41.247,0],[60,0.0648]]}
          Inner ColorA {type=MovieSceneScriptingFloatChannel, const [1,4]}
         ]
        }]
      }]
    }
    Image_ElimBonusGlow {
     id=#4
     possessed_class=Image
     tracks [
      {
       type=MovieSceneWidgetMaterialTrack
       sections [{type=MovieSceneParameterSection, channels [Glow Y Movement {type=MovieSceneScriptingFloatChannel, keys [[5,1.4],[42,-0.5]]}]}]
      }
      {
       type=MovieSceneFloatTrack
       display_name=Render Opacity
       sections [{type=MovieSceneFloatSection, channels [None {type=MovieSceneScriptingFloatChannel, keys [[0,1],[38,1],[42,0]]}]}]
      }
      {
       type=MovieSceneByteTrack
       display_name=Visibility
       sections [{type=MovieSceneByteSection, channels [None {type=MovieSceneScriptingByteChannel, keys [[0,0],[42,1]]}]}]
      }
     ]
    }
   ]
   display_rate {numerator=60}
  }
 }
 Anim_PlacementBonus {
  binding_count=4
  length_seconds=1
  object=&self:Anim_PlacementBonus
  sequence {
   bindings [
    TierForeground {
     id=#1
     possessed_class=Image
     tracks [
      {
       type=MovieSceneWidgetMaterialTrack
       sections [{
         type=MovieSceneParameterSection
         end_frame=52
         start_frame=0
         channels [
          Badge Hotness {type=MovieSceneScriptingFloatChannel, keys [[0,0],[12,0],[18.489,1],[52,0]]}
          Gem ColorR {type=MovieSceneScriptingFloatChannel, keys [[0,0],[12,0],[18.489,1],[52,0]]}
          Gem ColorG {type=MovieSceneScriptingFloatChannel, keys [[0,0.1145],[12,0.1],[18.489,1],[52,0.1145]]}
          Gem ColorB {type=MovieSceneScriptingFloatChannel, const [1,4]}
          Gem ColorA {type=MovieSceneScriptingFloatChannel, const [1,4]}
          Secondary Gem ColorR {type=MovieSceneScriptingFloatChannel, keys [[0,0],[12,0],[19,1],[52,0]]}
          Secondary Gem ColorG {type=MovieSceneScriptingFloatChannel, keys [[0,0.0548],[12,0.04],[19,1],[52,0.0548]]}
          Secondary Gem ColorB {type=MovieSceneScriptingFloatChannel, keys [[0,0.5625],[12,1],[19,1],[52,0.5625]]}
          Secondary Gem ColorA {type=MovieSceneScriptingFloatChannel, const [1,3]}
         ]
        }]
      }
      {type=MovieScene2DTransformTrack, sections []}
     ]
    }
    TierBackground {
     id=#2
     possessed_class=Image
     tracks [
      {
       type=MovieSceneWidgetMaterialTrack
       sections [{
         type=MovieSceneParameterSection
         end_frame=60
         start_frame=6
         channels [
          Outer ColorR {type=MovieSceneScriptingFloatChannel, const [0,1]}
          Outer ColorG {type=MovieSceneScriptingFloatChannel, const [0.0508,1]}
          Outer ColorB {type=MovieSceneScriptingFloatChannel, const [0.2083,1]}
          Outer ColorA {type=MovieSceneScriptingFloatChannel, const [1,1]}
         ]
        }]
      }
      {
       type=MovieSceneFloatTrack
       display_name=Render Opacity
       sections [{
         type=MovieSceneFloatSection
         end_frame=60
         start_frame=6
         channels [None {type=MovieSceneScriptingFloatChannel, keys [[6,1],[19.745,0],[51.164,0],[59.999,1]]}]
        }]
      }
     ]
    }
    Image_PlacementBonusGlow {
     id=#5
     possessed_class=Image
     tracks [
      {
       type=MovieSceneWidgetMaterialTrack
       sections [{type=MovieSceneParameterSection, channels [Glow Y Movement {type=MovieSceneScriptingFloatChannel, keys [[5,1.4],[42,-0.5]]}]}]
      }
      {
       type=MovieSceneFloatTrack
       display_name=Render Opacity
       sections [{type=MovieSceneFloatSection, channels [None {type=MovieSceneScriptingFloatChannel, keys [[0,1],[38,1],[42,0]]}]}]
      }
      {
       type=MovieSceneByteTrack
       display_name=Visibility
       sections [{type=MovieSceneByteSection, channels [None {type=MovieSceneScriptingByteChannel, keys [[0,0],[42,1]]}]}]
      }
     ]
    }
    Image_Progress {
     id=#3
     possessed_class=Image
     tracks [{
       type=MovieSceneWidgetMaterialTrack
       sections [{
         type=MovieSceneParameterSection
         end_frame=60
         start_frame=0
         channels [
          Glow Range {type=MovieSceneScriptingFloatChannel, keys [[0,0.17],[10.13,0],[16.364,0.17],[50.649,0],[60,0.17]]}
          Bar ColorR {type=MovieSceneScriptingFloatChannel, keys [[0,0.0075],[13.247,0.15],[41,0.15],[60,0.0075]]}
          Bar ColorG {type=MovieSceneScriptingFloatChannel, keys [[0,0.4233],[13.247,0.3],[41,0.3],[60,0.4233]]}
          Bar ColorB {type=MovieSceneScriptingFloatChannel, keys [[0,0.8632],[13.247,1],[41,1],[60,0.8632]]}
          Bar ColorA {type=MovieSceneScriptingFloatChannel, const [1,2]}
          Bar Glow ColorR {type=MovieSceneScriptingFloatChannel, keys [[0,0],[13.247,0.2],[41,0.2],[60,0]]}
          Bar Glow ColorG {type=MovieSceneScriptingFloatChannel, keys [[0,0.3014],[13.247,0.3],[41,0.3],[60,0.3014]]}
          Bar Glow ColorB {type=MovieSceneScriptingFloatChannel, const [1,4]}
          Bar Glow ColorA {type=MovieSceneScriptingFloatChannel, const [1,4]}
          Bar Outline ColorR {type=MovieSceneScriptingFloatChannel, thinned_from=5, keys [[0,0.0622],[13,0],[41,0],[60,0.0622]]}
          Bar Outline ColorG {type=MovieSceneScriptingFloatChannel, thinned_from=5, keys [[0,0.8521],[13,0.2],[41,0.2],[60,0.8521]]}
          Bar Outline ColorB {type=MovieSceneScriptingFloatChannel, thinned_from=5, keys [[0,0.9948],[13.247,1],[41,1],[60,0.9948]]}
          Bar Outline ColorA {type=MovieSceneScriptingFloatChannel, const [1,5]}
          Progress Background ColorR {type=MovieSceneScriptingFloatChannel, keys [[0,0.0012],[13.247,0],[41,0],[60,0.0012]]}
          Progress Background ColorG {type=MovieSceneScriptingFloatChannel, keys [[0,0.0123],[13.247,0.01],[41,0.01],[60,0.0123]]}
          Progress Background ColorB {type=MovieSceneScriptingFloatChannel, keys [[0,0.0513],[13.247,0.04],[41,0.04],[60,0.0513]]}
          Progress Background ColorA {type=MovieSceneScriptingFloatChannel, const [1,4]}
          Inner ColorR {type=MovieSceneScriptingFloatChannel, keys [[0,0.0021],[13.247,0],[41,0],[60,0.0021]]}
          Inner ColorG {type=MovieSceneScriptingFloatChannel, keys [[0,0.0185],[13.247,0.03],[41,0.03],[60,0.0185]]}
          Inner ColorB {type=MovieSceneScriptingFloatChannel, keys [[0,0.0648],[13.247,0.2],[41,0.2],[60,0.0648]]}
          Inner ColorA {type=MovieSceneScriptingFloatChannel, const [1,4]}
         ]
        }]
      }]
    }
   ]
   display_rate {numerator=60}
  }
 }
]
graphs [
 EventGraph {links [], nodes []}
 SetProgress {nodes [
   SetProgress!() -> (then, ProgressValue:double=~) {is=title K2Node_FunctionEntry}
   > n1 = GetDynamicMaterial(self=Get Image_Progress() -> (Image_Progress:Image Object Reference))
   > SetScalarParameterValue(self=@n1, ParameterName=Progress, Value=Get ProgressValue() -> (ProgressValue:double))
   > SetScalarParameterValue(self=@n1, ParameterName=Delta Progress, Value=0)
  ]}
 SetProgressBarColors {nodes [
   n1 = SetProgressBarColors!() -> (then, TierData:Fort Habanero Tier Structure) {is=title K2Node_FunctionEntry}
   > n2 = Cast To BP_UI_Habanero_TierTypeDataBase!(
    execute
    Object=Break Fort Habanero Tier!(FortHabaneroTier:"Fort Habanero Tier Structure (by ref)"=@n1) -> (
     ForegroundIcon:Material Instance Soft Object Reference
     ForegroundIconSmall:Material Instance Soft Object Reference
     BackgroundIcon:Material Instance Soft Object Reference
     BackgroundIconSmall:Material Instance Soft Object Reference
     WhiteIcon:Texture 2D Soft Object Reference
     SimpleIconSmall:Texture 2D Soft Object Reference
     Name:Text
     TierType:Fort Habanero Tier Type Data Object Reference
     AdditionalLeftIcon:Material Instance Soft Object Reference
     AdditionalRightIcon:Material Instance Soft Object Reference
    ) {is=title K2Node_BreakStruct}.TierType
   ) -> (then, CastFailed, AsBP UI Habanero Tier Type Data Base:BP UI Habanero Tier Type Data Base Object Reference) {is=title K2Node_DynamicCast}
   > n3 = GetDynamicMaterial(self=Get Image_Progress())
   > SetVectorParameterValue(
    self=@n3
    ParameterName=Get Progress Bar Color() -> (Progress Bar Color:Name=None)
    Value="To LinearColor (Color)"!(InColor:Color Structure=Get ProgressBar_Color!(self:BP UI Habanero Tier Type Data Base Object Reference=@n2) -> (ProgressBar_Color:Color Structure)) -> (ReturnValue:Linear Color Structure)
   )
   > SetVectorParameterValue(
    self=@n3
    ParameterName=Get Progress Outline Color() -> (Progress Outline Color:Name=None)
    Value="To LinearColor (Color)"(InColor=Get ProgressOutline_Color!(self:BP UI Habanero Tier Type Data Base Object Reference=@n2) -> (ProgressOutline_Color:Color Structure))
   )
   > SetVectorParameterValue(
    self=@n3
    ParameterName=Get Delta Bar Color() -> (Delta Bar Color:Name=None)
    Value="To LinearColor (Color)"(InColor=Get DeltaBar_Color!(self:BP UI Habanero Tier Type Data Base Object Reference=@n2) -> (DeltaBar_Color:Color Structure))
   )
   > SetVectorParameterValue(
    self=@n3
    ParameterName=Get Bar Glow Color() -> (Bar Glow Color:Name=None)
    Value="To LinearColor (Color)"(InColor=Get BarGlow_Color!(self:BP UI Habanero Tier Type Data Base Object Reference=@n2) -> (BarGlow_Color:Color Structure))
   )
   > SetVectorParameterValue(
    self=@n3
    ParameterName=Edge Color
    Value="To LinearColor (Color)"(InColor=Get EdgeBar_Color!(self:BP UI Habanero Tier Type Data Base Object Reference=@n2) -> (EdgeBar_Color:Color Structure))
   )
   SetVectorParameterValue(
    self=@n3
    ParameterName=Get Inner Color() -> (Inner Color:Name=None)
    Value="To LinearColor (Color)"(InColor=Get InnerBackground_Color!(self:BP UI Habanero Tier Type Data Base Object Reference=@n2) -> (InnerBackground_Color:Color Structure))
   )
   SetVectorParameterValue(
    self=@n3
    ParameterName=Get Progress Background Color() -> (Progress Background Color:Name=None)
    Value="To LinearColor (Color)"(InColor=Get ProgressBackground_Color!(self:BP UI Habanero Tier Type Data Base Object Reference=@n2) -> (ProgressBackground_Color:Color Structure))
   )
  ]}
 SetTierVisuals {nodes [
   n1 = SetTierVisuals!() -> (then, TierData:Fort Habanero Tier Structure) {is=title K2Node_FunctionEntry}
   > UpdateDetailsFromTierData!(
    execute
    self:WBP Habanero Icon Object Reference=Get WBP_Habanero_Icon() -> (WBP_Habanero_Icon:WBP Habanero Icon Object Reference)
    FortHabaneroTier:"Fort Habanero Tier Structure (by ref)"=@n1
   ) -> (then)
   n2 = Get Category!(self:Fort Habanero Tier Type Data Object Reference=Break Fort Habanero Tier(FortHabaneroTier=@n1).TierType) -> (Category:EHabaneroTierType Enum=HabaneroSauce0)
   > SetBrushFromMaterial(
    self=Get Image_Progress()
    Material=SelectMaterialInstance!(
     A:Material Instance Object Reference=/Game/UI/Frontend/Lobby/Habanero/Habanero_Materials/MaterialInstances/ProgressBars/MI_UI_Habanero_ProgressBar_Initial
     B:Material Instance Object Reference=/Game/UI/Frontend/Lobby/Habanero/Habanero_Materials/MaterialInstances/ProgressBars/MI_UI_Habanero_ProgressBar
     bSelectA:bool="Equal (Enum)"(A:EHabaneroTierType Enum=@n2, B:EHabaneroTierType Enum=HabaneroSauce0)
    ) -> (ReturnValue:Material Instance Object Reference)
   )
   n3 = "Equal (Enum)"(A=@n2, B=HabaneroSauce8)
   > SetVisibility(self=Get Image_Progress(), InVisibility=Select(Option 0=SelfHitTestInvisible, Option 1=Hidden, Index=@n3))
   > SetProgressBarColors!(execute, self, TierData:Fort Habanero Tier Structure=@n1) -> (then) {is=title K2Node_CallFunction}
   > SetRenderScale!(
    execute
    self:Widget Object Reference=Get WBP_Habanero_Icon()
    Scale:Vector 2D Structure=Select(Option 0:Vector 2D Structure=(X=1,Y=1), Option 1:Vector 2D Structure=(X=1.5,Y=1.5), Index=@n3) -> (ReturnValue:Vector 2D Structure)
   ) -> (then)
  ]}
 SetTierVisualsWithRankData {nodes [
   n1 = SetTierVisualsWithRankData!() -> (then, InHabaneroType:String, Rank:int, SeasonName:String, Matchmaking Rating:int, Current Player Position:int) {is=title K2Node_FunctionEntry}
   n2 = Get WBP_Habanero_Icon()
   > DisplayBadge!(
    execute
    self:Fort Habanero Tier Widget Object Reference=@n2
    HabaneroType:String=@n1
    CurrentHabanero:int=@n1.Rank
    SeasonName:String=@n1.SeasonName
    MatchmakingRating:int=@n1.Matchmaking Rating
    CurrentPlayerPosition:int=@n1.Current Player Position
   ) -> (then)
   n3 = Get CurrentDisplayData!(self:WBP Habanero Icon Object Reference=@n2) -> (CurrentDisplayData:Fort Habanero Tier Structure)
   n4 = Get Category(self=Break Fort Habanero Tier(FortHabaneroTier=@n3).TierType) -> (Category=HabaneroSauce0)
   > SetBrushFromMaterial(
    self=Get Image_Progress()
    Material=SelectMaterialInstance(
     A=&MI_UI_Habanero_ProgressBar_Initial
     B=&MI_UI_Habanero_ProgressBar
     bSelectA="Equal (Enum)"(A=@n4, B=HabaneroSauce0)
    )
   )
   n5 = "Equal (Enum)"(A=@n4, B=HabaneroSauce8)
   > SetVisibility(
    self=Get Image_Progress()
    InVisibility=Select(Option 0:ESlateVisibility Enum=SelfHitTestInvisible, Option 1:ESlateVisibility Enum=Hidden, Index=@n5) -> (ReturnValue:ESlateVisibility Enum)
   )
   > SetProgressBarColors(TierData=@n3)
   > SetRenderScale(
    self=Get WBP_Habanero_Icon()
    Scale=Select(Option 0:Vector 2D Structure=(X=1,Y=1), Option 1:Vector 2D Structure=(X=1.5,Y=1.5), Index=@n5) -> (ReturnValue:Vector 2D Structure)
   )
   > Return Node!(execute, Current Display Data:Fort Habanero Tier Structure=@n3)
  ]}
]
variables []
widgets {
 info {inheritedWidgetCount=0, parentClass=&UserWidget, rootWidgetClass=&Overlay, widgetCount=11}
 widgets [
  Overlay_23 {widgetClassPath=&Overlay, widget=Overlay_23, properties {visibility=HitTestInvisible}}
  Image_Progress {
   widgetClassPath=&Image
   bIsVariable=true
   parent=Overlay_23
   widget=Image_Progress
   properties {brush {
     bIsDynamicallyLoaded=false
     drawAs=Image
     imageType=NoImage
     mirroring=NoMirror
     resourceName=None
     resourceObject=&MI_UI_Habanero_ProgressBar
     tiling=NoTile
     imageSize=<196 196>
     margin=<ltrb 0 0 0 0>
     outlineSettings {
      bUseBrushTransparency=false
      roundingType=HalfHeightRadius
      width=0
      color {colorUseRule=UseColor_Specified, specifiedColor=<rgba 0 0 0 0>}
      cornerRadii=<0 0 0 0>
     }
     tintColor {colorUseRule=UseColor_Specified, specifiedColor=<rgba 1 1 1 1>}
     uVRegion {bIsValid=false, max=<0 0>, min=<0 0>}
    }}
   slot {class=OverlaySlot, properties {horizontalAlignment=HAlign_Fill, verticalAlignment=VAlign_Fill}}
  }
  WBP_Habanero_Icon {
   widgetClassPath=&/WBP_Habanero_Icon
   bIsVariable=true
   parent=Overlay_23
   widget=WBP_Habanero_Icon
   properties {bInitializeToRepresentLocalUser=true, displayLargeIcon=true, iconMinSize=<128 128>}
   slot {class=OverlaySlot, properties {horizontalAlignment=HAlign_Fill, verticalAlignment=VAlign_Fill, padding=<ltrb 32 32 32 32>}}
  }
  Overlay_57 {
   widgetClassPath=&Overlay
   parent=Overlay_23
   widget=Overlay_57
   properties {visibility=Collapsed}
   slot {class=OverlaySlot, properties {horizontalAlignment=HAlign_Center, verticalAlignment=VAlign_Center}}
  }
  IconSizeSpacer {
   widgetClassPath=&Spacer
   bIsVariable=true
   parent=Overlay_57
   widget=IconSizeSpacer
   properties {size=<96 96>}
   slot {class=OverlaySlot, properties {horizontalAlignment=HAlign_Fill, verticalAlignment=VAlign_Fill}}
  }
  TierBackground {
   widgetClassPath=&Image
   bIsVariable=true
   parent=Overlay_57
   widget=TierBackground
   properties {
    visibility=HitTestInvisible
    brush {
     bIsDynamicallyLoaded=false
     drawAs=Image
     imageType=NoImage
     mirroring=NoMirror
     resourceName=None
     resourceObject=/Game/UI/Frontend/Lobby/Habanero/Habanero_Materials/MaterialInstances/Icons/Background/MI_UI_Habanero_Background_00
     tiling=NoTile
     imageSize=<128 128>
     margin=<ltrb 0 0 0 0>
     outlineSettings {
      bUseBrushTransparency=false
      roundingType=HalfHeightRadius
      width=0
      color {colorUseRule=UseColor_Specified, specifiedColor=<rgba 0 0 0 0>}
      cornerRadii=<0 0 0 0>
     }
     tintColor {colorUseRule=UseColor_Specified, specifiedColor=<rgba 1 1 1 1>}
     uVRegion {bIsValid=false, max=<0 0>, min=<0 0>}
    }
    renderTransform {angle=0, scale=<1.25 1.25>, shear=<0 0>, translation=<0 0>}
   }
   slot {class=OverlaySlot, properties {horizontalAlignment=HAlign_Fill, verticalAlignment=VAlign_Fill}}
  }
  TierForeground {
   widgetClassPath=&Image
   bIsVariable=true
   parent=Overlay_57
   widget=TierForeground
   properties {
    visibility=HitTestInvisible
    brush {
     bIsDynamicallyLoaded=false
     drawAs=Image
     imageType=NoImage
     mirroring=NoMirror
     resourceName=None
     resourceObject=/Game/UI/Frontend/Lobby/Habanero/Habanero_Materials/MaterialInstances/Icons/Foreground/MI_UI_Habanero_Foreground_00
     tiling=NoTile
     imageSize=<128 128>
     margin=<ltrb 0 0 0 0>
     outlineSettings {
      bUseBrushTransparency=false
      roundingType=HalfHeightRadius
      width=0
      color {colorUseRule=UseColor_Specified, specifiedColor=<rgba 0 0 0 0>}
      cornerRadii=<0 0 0 0>
     }
     tintColor {colorUseRule=UseColor_Specified, specifiedColor=<rgba 1 1 1 1>}
     uVRegion {bIsValid=false, max=<0 0>, min=<0 0>}
    }
   }
   slot {class=OverlaySlot, properties {horizontalAlignment=HAlign_Fill, verticalAlignment=VAlign_Fill}}
  }
  FortVisualAttachment_Elim {
   widgetClassPath=&FortVisualAttachment
   parent=Overlay_57
   widget=FortVisualAttachment_Elim
   properties {visibility=Visible, contentAnchor=<0.5 0.5>}
   slot {class=OverlaySlot, properties {horizontalAlignment=HAlign_Center, verticalAlignment=VAlign_Center}}
  }
  Overlay_0 {
   widgetClassPath=&Overlay
   parent=FortVisualAttachment_Elim
   widget=Overlay_0
   slot {class=SizeBoxSlot, properties {horizontalAlignment=HAlign_Center, verticalAlignment=VAlign_Center}}
  }
  Image_ElimBonusGlow {
   widgetClassPath=&Image
   bIsVariable=true
   parent=Overlay_0
   widget=Image_ElimBonusGlow
   properties {
    renderOpacity=0
    visibility=Collapsed
    brush {
     bIsDynamicallyLoaded=false
     drawAs=Image
     imageType=FullColor
     mirroring=NoMirror
     resourceName=None
     resourceObject=/Game/UI/Frontend/Lobby/Habanero/Habanero_Materials/M_UI_Habanero_RecapGlow
     tiling=NoTile
     imageSize=<250 250>
     margin=<ltrb 0 0 0 0>
     outlineSettings {
      bUseBrushTransparency=false
      roundingType=HalfHeightRadius
      width=0
      color {colorUseRule=UseColor_Specified, specifiedColor=<rgba 0 0 0 0>}
      cornerRadii=<0 0 0 0>
     }
     tintColor {colorUseRule=UseColor_Specified, specifiedColor=<rgba 1 1 1 1>}
     uVRegion {bIsValid=false, max=<0 0>, min=<0 0>}
    }
   }
   slot {class=OverlaySlot}
  }
  Image_PlacementBonusGlow {
   widgetClassPath=&Image
   bIsVariable=true
   parent=Overlay_0
   widget=Image_PlacementBonusGlow
   properties {
    renderOpacity=0
    visibility=Collapsed
    brush {
     bIsDynamicallyLoaded=false
     drawAs=Image
     imageType=FullColor
     mirroring=NoMirror
     resourceName=None
     resourceObject=/Game/UI/Frontend/Lobby/Habanero/Habanero_Materials/MaterialInstances/Flares/MI_UI_Habanero_RecapGlow
     tiling=NoTile
     imageSize=<250 250>
     margin=<ltrb 0 0 0 0>
     outlineSettings {
      bUseBrushTransparency=false
      roundingType=HalfHeightRadius
      width=0
      color {colorUseRule=UseColor_Specified, specifiedColor=<rgba 0 0 0 0>}
      cornerRadii=<0 0 0 0>
     }
     tintColor {colorUseRule=UseColor_Specified, specifiedColor=<rgba 1 1 1 1>}
     uVRegion {bIsValid=false, max=<0 0>, min=<0 0>}
    }
   }
   slot {class=OverlaySlot, properties {horizontalAlignment=HAlign_Center, verticalAlignment=VAlign_Center}}
  }
 ]
}
