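# Ranked (Habanero) mode info modal for the Fortnite lobby, built on FortHabaneroInfoModal.
# On activation it builds matching tabs and scrollable panels from DataInfo assets, filtered by the selected island's ranked type.
# Tab clicks, mouse wheel and scroll events update CurrentIndex, which scrolls to a panel, updates focus, plays sounds and sends analytics.
# The sidebar shows tier progress with percent or position text, has a mobile layout, and keeps its season timer collapsed.
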WidgetBlueprint /Game/UI/Frontend/Lobby/Habanero/WBP_Habanero_InfoMenu
graphs [
 DesignTimeDevPreview {nodes [
   n1 = DesignTimeDevPreview!() -> (then, IsDesignTime:bool=~) {is=title K2Node_FunctionEntry}
   > Branch(Condition=@n1)
   > Populate Panels()
  ]}
 EventGraph {nodes [
   n1 = "On Clicked (Button_Close)"!() -> (OutputDelegate, then, Button:Common Button Base Object Reference) {is=title K2Node_ComponentBoundEvent}
   RemoveFromParent(execute=[@n1.then, @n2.then])
   n2 = "On Clicked (Button_Close_Alt)"!() -> (OutputDelegate, then, Button:Common Button Base Object Reference) {is=title K2Node_ComponentBoundEvent}
   n3 = OnTabSelected!() -> (OutputDelegate, then, TabId:Name) {is=title K2Node_CustomEvent}
   n4 = Get Button_Tabs() -> (Button_Tabs:WBP UIKit Tabs Object Reference)
   > n5 = Cast To WBP_UIKit_Button_Selectable!(
    execute
    Object=GetTabButtonBaseByID!(self:Common Tab List Widget Base Object Reference=@n4, TabNameID:Name=@n3.TabId) -> (ReturnValue:Common Button Base Object Reference)
   ) -> (then, CastFailed, AsWBP UIKit Button Selectable:WBP UIKit Button Selectable Object Reference) {is=title K2Node_DynamicCast}
   > SetPressedSoundOverride!(
    execute
    self:Common Button Base Object Reference=@n5
    Sound:Sound Base Object Reference=Get SelectSound() -> (SelectSound:Sound Base Object Reference)
   ) -> (then)
   > SetHoveredSoundOverride!(
    execute
    self:Common Button Base Object Reference=@n5
    Sound:Sound Base Object Reference=Get HoverSound() -> (HoverSound:Sound Base Object Reference)
   ) -> (then)
   > n6 = Set CurrentIndex(CurrentIndex:int=GetTabIndexByID!(self:UIKit Tab Group Object Reference=@n4, TabNameID:Name=@n3.TabId) -> (ReturnValue:int=0)) -> (Output_Get:int=0)
   > UpdateFocusedPanel!(execute, self, Index:int=@n6) -> (then)
   n7 = Event SetProgress!() -> (
    OutputDelegate
    then
    ProgressKey:String
    CurrentHabanero:int=0
    CurrentPosition:int=0
    CurrentMatchmakingRating:int=0
    ProgressPercent:float
   ) {is=title K2Node_Event}
   n8 = Get WBP_Habanero_ProgressInfo() -> (WBP_Habanero_ProgressInfo:WBP Habanero Progress Info Object Reference)
   > n9 = SetTierVisualsWithRankData!(
    execute
    self:WBP Habanero Progress Info Object Reference=@n8
    InHabaneroType:String=@n7.ProgressKey
    Rank:int=@n7.CurrentHabanero
    SeasonName:String
    Matchmaking Rating:int=@n7.CurrentMatchmakingRating
    Current Player Position:int=@n7.CurrentPosition
   ) -> (then, Current Display Data:Fort Habanero Tier Structure)
   > SetProgress!(execute, self:WBP Habanero Progress Info Object Reference=@n8, ProgressValue:double=@n7.ProgressPercent) -> (then)
   > Update Percent Text!(
    execute
    self
    RankPosition:int=@n7.CurrentPosition
    Percent:double=@n7.ProgressPercent
    FortHabaneroTier:"Fort Habanero Tier Structure (by ref)"=@n9
   ) -> (then)
   > SetPercentageTextDisplay!(execute, self, TierData:Fort Habanero Tier Structure=@n9) -> (then)
   n10 = Event Construct()
   n11 = IsMobileGame()
   SetPadding!(
    execute=[@n16.then, @n10.then]
    self:Border Slot Object Reference=SlotAsBorderSlot!(Widget:Widget Object Reference=Get SafeZone_content() -> (SafeZone_content:Safe Zone Object Reference)) -> (ReturnValue:Border Slot Object Reference)
    InPadding_Left:float=32
    InPadding_Top:float=SelectFloat(A=32, B=64, bPickA=@n11)
    InPadding_Right:float=SelectFloat(A=64, B=32, bPickA=@n11)
    InPadding_Bottom:float=32
   ) -> (then)
   > SetMobileTabStyle()
   n12 = Event SetTimerText!() -> (OutputDelegate, then, Timespan:"Timespan Structure (by ref)") {is=title K2Node_Event}
   > UpdateTime!(execute, self, InTimespan:Timespan Structure=@n12.Timespan) -> (then)
   n13 = Event SetTimerVisibility!() -> (OutputDelegate, then, TimerVisibility:ESlateVisibility Enum=Visible) {is=title K2Node_Event}
   > SetVisibility(self=Get Timer_HBox() -> (Timer_HBox:Horizontal Box Object Reference), InVisibility=Collapsed)
   Event On Activated()
   > Populate Panels()
   > Populate Buttons()
   > UpdateFocusedPanel(Index=Get CurrentIndex() -> (CurrentIndex:int=0))
   > Fire Event Element Interacted(Element=Self-Reference())
   > PlaySound2D(Sound=Get AppearSoundCue() -> (AppearSoundCue:Sound Base Object Reference))
   > DelayUntilNextTick!(execute) -> (then)
   n14 = Get Button_Tabs()
   > SetFocus(self=GetChildAt!(
     self:Panel Widget Object Reference=Get TabsContainer!(self:WBP UIKit Tabs Base Object Reference=@n14) -> (TabsContainer:Grid Panel Object Reference)
     Index:int=0
    ) -> (ReturnValue:Widget Object Reference))
   n15 = Event PreConstruct()
   > n16 = DesignTimeDevPreview!(execute, self, IsDesignTime:bool=@n15.IsDesignTime) -> (then) {is=title K2Node_CallFunction}
   n17 = Event On Scroll Completed!() -> (OutputDelegate, then, bScrolledDown:bool) {is=title K2Node_Event}
   > EndInertialScrolling(self:Scroll Box Object Reference=Get ScrollBox_Panels() -> (ScrollBox_Panels:Athena Scroll Box Object Reference))
   > n18 = Branch(Condition=@n17.bScrolledDown)
   > ScrolledDown()
   ScrolledUp(execute=@n18.else)
   Event On Deactivated()
   > RemoveAllTabs(self:Common Tab List Widget Base Object Reference=Get Button_Tabs())
   > Clear(TargetArray:Array of UI Kit Tab Group Button View Model Object References=Get TabsDataInfo() -> (TabsDataInfo:Array of UI Kit Tab Group Button View Model Object References))
   > ClearChildren(self=Get Vbox_Panels() -> (Vbox_Panels:Vertical Box Object Reference))
   Mouse Wheel Down!() -> (Pressed:Exec, Released:Exec, Key:Key Structure) {is=title K2Node_InputKey, bOverrideParentBinding=false, inputKey=MouseScrollDown}
   n19 = IsActivated!(self:Common Activatable Widget Object Reference) -> (ReturnValue:bool)
   > Branch(Condition=@n19)
   > ScrolledDown()
   Mouse Wheel Up!() -> (Pressed:Exec, Released:Exec, Key:Key Structure) {is=title K2Node_InputKey, bOverrideParentBinding=false, inputKey=MouseScrollUp}
   > Branch(Condition=@n19)
   > ScrolledUp()
   Reroute Node(InputPin:ESlateVisibility Enum=@n13.TimerVisibility) -> (OutputPin:ESlateVisibility Enum)
  ]}
 HasToShowPanel {nodes [
   n1 = HasToShowPanel!() -> (then, Data:BP UI Habanero Info Panel Object Reference) {is=title K2Node_FunctionEntry}
   n2 = Get ExclusiveRankTypes!(self:BP UI Habanero Info Panel Object Reference=@n1) -> (ExclusiveRankTypes:Array of Strings)
   n3 = GetSelectedIslandRankedType!(self:Fort Habanero Info Modal Object Reference) -> (ReturnValue:String)
   > Return Node!(
    execute
    Show:bool=OR Boolean(A=Is Empty!(TargetArray=@n2) -> (ReturnValue:bool), B=Contains Item(TargetArray:Array of Strings=@n2, ItemToFind:"String (by ref)"=@n3))
   )
  ]}
 Populate Buttons {nodes [
   Populate Buttons() {is=title K2Node_FunctionEntry}
   > n1 = Construct Object from Class(Class=/Script/FortniteUI.UIKitTabGroupViewModel) -> (ReturnValue:UI Kit Tab Group View Model Object Reference)
   > n2 = For Each Loop(Array:Array of BP UI Habanero Info Panel Object References=Get DataInfo() -> (DataInfo:Array of BP UI Habanero Info Panel Object References)) -> (Array Element:BP UI Habanero Info Panel Object Reference)
   Set Tabs!(
    execute=@n2.Completed
    Tabs:Array of UI Kit Tab Group Button View Model Object References=Get TabsDataInfo()
    self:UI Kit Tab Group View Model Object Reference=@n1
   ) -> (then, Output_Get:Array of UI Kit Tab Group Button View Model Object References)
   n3 = Get Button_Tabs()
   > Set UIKitTabGroupViewModel!(execute, UIKitTabGroupViewModel:UI Kit Tab Group View Model Object Reference=@n1, self:WBP UIKit Tabs Object Reference=@n3) -> (then, Output_Get:UI Kit Tab Group View Model Object Reference)
   > Bind Event to On Tab Selected(self:Common Tab List Widget Base Object Reference=@n3, Delegate=Create Event()) {is=title K2Node_AddDelegate}
   n4 = HasToShowPanel!(execute=@n2.LoopBody, self, Data:BP UI Habanero Info Panel Object Reference=@n2) -> (then, Show:bool) {is=title K2Node_CallFunction}
   > Branch(Condition=@n4)
   > n5 = Construct Object from Class(Class=/Script/FortniteUI.UIKitTabGroupButtonViewModel) -> (ReturnValue:UI Kit Tab Group Button View Model Object Reference)
   > Set Label!(
    execute
    Label:Text=Get ButtonName!(self:BP UI Habanero Info Panel Object Reference=@n2) -> (ButtonName:Text)
    self:UI Kit Selectable Button View Model Object Reference=@n5
   ) -> (then, Output_Get:Text)
   > Add(
    TargetArray:Array of UI Kit Tab Group Button View Model Object References=Get TabsDataInfo()
    NewItem:"UI Kit Tab Group Button View Model Object Reference (by ref)"=@n5
   ) {is=title K2Node_CallArrayFunction}
  ]}
 Populate Panels {nodes [
   Populate Panels() {is=title K2Node_FunctionEntry}
   > n1 = For Each Loop(Array=Get DataInfo())
   > n2 = HasToShowPanel(Data=@n1)
   > Branch(Condition=@n2)
   > n3 = Create Widget!(
    execute:Exec
    Class:User Widget Class Reference=Get Panel Type!(self:BP UI Habanero Info Panel Object Reference=@n1) -> (Panel Type:User Widget Class Reference)
    OwningPlayer:Player Controller Object Reference
   ) -> (then:Exec, ReturnValue:User Widget Object Reference)
   > AddChildToVerticalBox(self=Get Vbox_Panels(), Content=@n3)
   > UpdatePanel!(execute:Exec, self:Object Reference=@n3, Object In:Object Reference=@n1) -> (then:Exec) {is=title K2Node_Message}
   > UpdateFocus!(
    execute:Exec
    self:Object Reference=@n3
    IsFocused:bool=Select(Option 0:bool=false, Option 1:bool=true, Index="Equal (Integer)"(A=@n1.Array Index)) -> (ReturnValue:bool)
   ) -> (then:Exec) {is=title K2Node_Message}
   > UpdateEdgePadding!(execute, self, Index:int=@n1.Array Index, Object:Object Reference=@n3) -> (then)
  ]}
 ScrolledDown {nodes [
   ScrolledDown() {is=title K2Node_FunctionEntry}
   n1 = Get DataInfo()
   n2 = Get CurrentIndex() -> (CurrentIndex=0)
   > n3 = Branch(Condition="integer < integer"(A=@n2, B=Last Index!(TargetArray:Array of BP UI Habanero Info Panel Object References=@n1) -> (ReturnValue=0)))
   > n4 = Increment Int(Value=@n2)
   > n5 = Set CurrentIndex(CurrentIndex=@n4) -> (Output_Get=0)
   > n6 = Set bScrollUp(bScrollUp:bool=false) -> (Output_Get:bool)
   UpdateFocusedPanel(execute=[@n6.then, @n3.else], Index=@n5)
   > SelectTabByIndex!(
    execute
    self:UIKit Tab Group Object Reference=Get Button_Tabs()
    Index:int=Get CurrentIndex() -> (CurrentIndex=0)
    bSuppressClickFeedback:bool=true
   ) -> (then, ReturnValue:bool)
  ]}
 ScrolledUp {nodes [
   ScrolledUp() {is=title K2Node_FunctionEntry}
   n1 = Get CurrentIndex() -> (CurrentIndex=0)
   > n2 = Branch(Condition="Equal (Integer)"(A=@n1))
   n3 = Decrement Int(" "=@n2.else, Value=@n1)
   > n4 = Set CurrentIndex(CurrentIndex=@n3) -> (Output_Get=0)
   > Set bScrollUp(bScrollUp=true)
   > UpdateFocusedPanel(Index=@n4)
   > SelectTabByIndex(self=Get Button_Tabs(), Index=Get CurrentIndex() -> (CurrentIndex=0), bSuppressClickFeedback=true)
  ]}
 SetMobileTabStyle {nodes [
   SetMobileTabStyle() {is=title K2Node_FunctionEntry}
   > Branch(Condition=IsMobileGame())
   n1 = Get SizeBox_ProgressInfo() -> (SizeBox_ProgressInfo:Size Box Object Reference)
   > SetHeightOverride(self=@n1, InHeightOverride=176)
   > SetWidthOverride(self=@n1, InWidthOverride=176)
   > SetSize!(execute, self:Spacer Object Reference=Get Spacer_72() -> (Spacer_72:Spacer Object Reference), InSize:Vector 2D Structure=(X=212,Y=170)) -> (then)
   n2 = Get TabsContainer(self=Get Button_Tabs())
   > n3 = For Loop(LastIndex=GetChildrenCount!(self:Panel Widget Object Reference=@n2) -> (ReturnValue:int=0))
   > n4 = Cast To WBP_UIKit_Button_Tab!(execute, Object=GetChildAt(self=@n2, Index=@n3)) -> (then, CastFailed, AsWBP UIKit Button Tab:WBP UIKit Button Tab Object Reference) {is=title K2Node_DynamicCast}
   > n5 = Set MobileMinHeight!(execute, MobileMinHeight:int=64, self:WBP UIKit Button Generic Base Object Reference=@n4) -> (then, Output_Get:int=0)
   > n6 = Set MobileMinWidth!(execute, MobileMinWidth:int=387, self:WBP UIKit Button Generic Base Object Reference=@n4) -> (then, Output_Get:int=0)
   > SetMinDimensions!(execute, self:Common Button Base Object Reference=@n4, InMinWidth:int=@n6, InMinHeight:int=@n5) -> (then)
   > SetButtonContentAlignment!(
    execute
    self:WBP UIKit Button Generic Base Object Reference=@n4
    HorizontalAlignment:EHorizontalAlignment Enum=HAlign_Left
    VerticalAlignment:EVerticalAlignment Enum=VAlign_Fill
   ) -> (then)
   n7 = Get VerticalBox_Mobile() -> (VerticalBox_Mobile:Vertical Box Object Reference)
   n8 = Get HorizontalBox_Mobile() -> (HorizontalBox_Mobile:Horizontal Box Object Reference)
   > AddChild(self=@n8, Content=@n7)
   > SetHorizontalAlignment(self:Vertical Box Slot Object Reference=SlotAsVerticalBoxSlot(Widget=@n8), InHorizontalAlignment=HAlign_Left)
   n9 = SlotAsHorizontalBoxSlot!(Widget:Widget Object Reference=@n7) -> (ReturnValue:Horizontal Box Slot Object Reference)
   > SetVerticalAlignment(self:Horizontal Box Slot Object Reference=@n9, InVerticalAlignment=VAlign_Center)
   > SetHorizontalAlignment(self:Horizontal Box Slot Object Reference=@n9, InHorizontalAlignment=HAlign_Left)
   > SetHorizontalAlignment(
    self:Vertical Box Slot Object Reference=SlotAsVerticalBoxSlot(Widget=Get Tier_Name() -> (Tier_Name:Common Text Object Reference))
    InHorizontalAlignment=HAlign_Left
   )
   > SetHorizontalAlignment(
    self=SlotAsVerticalBoxSlot(Widget=Get PercentageTextWrapper() -> (PercentageTextWrapper:Border Object Reference))
    InHorizontalAlignment=HAlign_Left
   )
   > SetSize(self=Get Spacer_Mobile() -> (Spacer_Mobile:Spacer Object Reference), InSize=(X=0,Y=16))
   > SetPadding(
    self=SlotAsBorderSlot(Widget=Get Text_Percent() -> (Text_Percent:Common Text Object Reference))
    InPadding_Left=0
    InPadding_Top=0
    InPadding_Right=0
    InPadding_Bottom=-16
   )
  ]}
 SetPercentageTextDisplay {nodes [
   n1 = SetPercentageTextDisplay!() -> (then, TierData:Fort Habanero Tier Structure) {is=title K2Node_FunctionEntry}
   n2 = Break Fort Habanero Tier!(FortHabaneroTier:"Fort Habanero Tier Structure (by ref)"=@n1) -> (
    ForegroundIcon:Material Instance Soft Object Reference
    ForegroundIconSmall:Material Instance Soft Object Reference
    BackgroundIcon:Material Instance Soft Object Reference
    BackgroundIconSmall:Material Instance Soft Object Reference
    WhiteIcon:Texture 2D Soft Object Reference
    SimpleIconSmall:Texture 2D Soft Object Reference
    Name:Text
    TierType:Fort Habanero Tier Type Data Object Reference
    AdditionalLeftIcon:Material Instance Soft Object Reference
    AdditionalRightIcon:Material Instance Soft Object Reference
   ) {is=title K2Node_BreakStruct}
   n3 = Get Tier_Name()
   > "SetText (Text)"(self=@n3, InText=@n2.Name)
   n4 = Get Category!(self:Fort Habanero Tier Type Data Object Reference=@n2.TierType) -> (Category:EHabaneroTierType Enum=HabaneroSauce0)
   > SetVisibility(
    self=@n3
    InVisibility=Select(
     Option 0:ESlateVisibility Enum=HitTestInvisible
     Option 1:ESlateVisibility Enum=Collapsed
     Index="Equal (Enum)"(A:EHabaneroTierType Enum=@n4, B:EHabaneroTierType Enum=HabaneroSauce0)
    ) -> (ReturnValue:ESlateVisibility Enum)
   )
   n5 = "Equal (Enum)"(A=@n4, B=HabaneroSauce8)
   > SetStyle!(
    execute
    self:Common Text Object Reference=Get Text_Percent()
    InStyle:Common Text Style Class Reference=Select(
     Option 0:Common Text Style Class Reference=/FortUILibrary/Styles/Typography/_TextStyle_Body1._TextStyle_Body1_C
     Option 1:Common Text Style Class Reference=/FortUILibrary/Styles/Typography/_TextStyle_SubH2._TextStyle_SubH2_C
     Index=@n5
    ) -> (ReturnValue:Common Text Style Class Reference)
   ) -> (then)
   n6 = Get PercentageTextWrapper()
   > SetContentColorAndOpacity!(
    execute
    self:Border Object Reference=@n6
    InContentColorAndOpacity:Linear Color Structure=Select(Option 0:Linear Color Structure=(R=1,G=1,B=1,A=1), Option 1:Linear Color Structure=(R=0.8469,G=0.5711,B=0.9734,A=1), Index=@n5) -> (ReturnValue:Linear Color Structure)
   ) -> (then)
   n7 = Break Linear Color!(LinearColor:"Linear Color Structure (by ref)"=Get BrushColor!(self:Border Object Reference=@n6) -> (BrushColor:Linear Color Structure)) -> (R:float=~, G:float=~, B:float=~, A:float=~) {is=title K2Node_BreakStruct}
   > SetBrushColor(
    self=@n6
    InBrushColor=Make Linear Color!(R:float=@n7, G:float=@n7.G, B:float=@n7.B, A:float="To Float (Boolean)"(InBool=@n5)) -> (LinearColor:Linear Color Structure) {is=title K2Node_MakeStruct}
   )
  ]}
 Update Percent Text {nodes [
   n1 = Update Percent Text!() -> (then, RankPosition:int, Percent:double=~, FortHabaneroTier:"Fort Habanero Tier Structure (by ref)") {is=title K2Node_FunctionEntry}
   n2 = Get Category(self=Break Fort Habanero Tier(FortHabaneroTier=@n1.FortHabaneroTier).TierType) -> (Category=HabaneroSauce0)
   n3 = "Equal (Enum)"(A=@n2, B=HabaneroSauce8)
   n4 = float * float(A=Get Percent() -> (Percent:double), B=100) {is=title K2Node_PromotableOperator}
   n5 = Format Text!(
    Format="LOCTEXT(\"{number}%\")"
    number:int="Clamp (Integer)"(Value=Round(A=@n4) {is=title K2Node_CallFunction}, Min=1, Max=99) -> (ReturnValue=0)
   ) -> (Result)
   n6 = Get RankPosition() -> (RankPosition:int=0)
   n7 = Select(
    Option 0:Text=Format Text!(Format="LOCTEXT(\"##{rank}\")", rank:Text="To Text (Integer)"(Value=@n6)) -> (Result)
    Option 1:Text=Get UnrealRankCapText() -> (UnrealRankCapText:Text)
    Index="integer >= integer"(A=@n6, B=100000)
   ) -> (ReturnValue:Text)
   > "SetText (Text)"(
    self=Get Text_Percent()
    InText=Select(Option 0=Select(Option 0=@n5, Option 1=@n7, Index=@n3), Index="Equal (Enum)"(A=@n2, B=HabaneroSauce0))
   )
  ]}
 UpdateEdgePadding {nodes [
   n1 = UpdateEdgePadding!() -> (then, Index:int, Object:Object Reference) {is=title K2Node_FunctionEntry}
   > n2 = Cast To BPI_UI_Habanero_Entries!(execute, Object=@n1.Object) -> (then, CastFailed, AsBPI UI Habanero Entries:BPI UI Habanero Entries Interface) {is=title K2Node_DynamicCast}
   n3 = Get DataInfo()
   > UpdateEndListPadding!(
    execute
    self:BPI UI Habanero Entries Interface=@n2
    IsFirstEntry:bool="Equal (Integer)"(A=@n1)
    IsLastEntry:bool="Equal (Integer)"(A=@n1, B=Last Index(TargetArray=@n3) -> (ReturnValue=0))
   ) -> (then)
  ]}
 UpdateFocusedPanel {nodes [
   n1 = UpdateFocusedPanel!() -> (then, Index:int) {is=title K2Node_FunctionEntry}
   > ScrollWidgetIntoView!(
    execute
    self:Scroll Box Object Reference=Get ScrollBox_Panels()
    WidgetToFind:Widget Object Reference=GetChildAt(self=Get Vbox_Panels(), Index=@n1)
    AnimateScroll:bool=true
    ScrollDestination:EDescendantScrollDestination Enum=TopOrLeft
    Padding:float=256
   ) -> (then)
   > Branch(Condition="Not Equal (Integer)"(A=Get Index() -> (Index:int=0), B=Get LastIndex() -> (LastIndex:int=0)))
   n2 = Get Index() -> (Index=0)
   n3 = GetChildAt(self=Get Vbox_Panels(), Index=@n2)
   > n4 = Cast To BPI_UI_Habanero_Entries(Object=@n3)
   > UpdateFocus!(execute, self:BPI UI Habanero Entries Interface=@n4, IsFocused:bool=true) -> (then) {is=title K2Node_CallFunction}
   > SendPanelFocusedAnalytics!(execute, self:Fort Habanero Info Modal Object Reference, Panel:Widget Object Reference=@n3, Text:String="To String (Integer)"(InInt=@n2)) -> (then)
   > n5 = Cast To BPI_UI_Habanero_Entries(Object=GetChildAt(self=Get Vbox_Panels(), Index=Get LastIndex() -> (LastIndex=0)))
   > UpdateFocus(self=@n5, IsFocused=false)
   > Set LastIndex(LastIndex:int=Get Index() -> (Index=0)) -> (Output_Get:int=0)
   > Branch(Condition=IsVisible())
   > n6 = Branch(Condition=Get bScrollUp() -> (bScrollUp:bool))
   > PlaySound2D(Sound=Get ScrollUpSoundCue() -> (ScrollUpSoundCue:Sound Base Object Reference))
   PlaySound2D(execute=@n6.else, Sound=Get ScrollDownSoundCue() -> (ScrollDownSoundCue:Sound Base Object Reference))
  ]}
 UpdateTime {nodes [
   n1 = UpdateTime!() -> (then, InTimespan:Timespan Structure) {is=title K2Node_FunctionEntry}
   n2 = BreakTimespan!(InTimespan:Timespan Structure=@n1) -> (Days:int=0, Hours:int=0, Minutes:int=0, Seconds:int=0, Milliseconds:int=0)
   > n3 = Branch(Condition="integer > integer"(A=@n2))
   > "SetText (Text)"(
    self=Get Text_TimerTime() -> (Text_TimerTime:Common Text Object Reference)
    InText=Format Text!(Format="LOCTEXT(\"{0}D {1}H\")", 0:int=@n2, 1:int=@n2.Hours) -> (Result)
   )
   n4 = BreakTimespan(InTimespan=@n1) -> (Days=0, Hours=0, Minutes=0, Seconds=0, Milliseconds=0)
   n5 = Branch(execute=@n3.else, Condition="integer > integer"(A=@n4.Hours))
   > "SetText (Text)"(self=Get Text_TimerTime(), InText=Format Text(Format="LOCTEXT(\"{0}H {1}M\")", 0=@n4.Hours, 1=@n4.Minutes))
   "SetText (Text)"(execute=@n5.else, self=Get Text_TimerTime(), InText=Format Text(Format="LOCTEXT(\"{0}M {1}S\")", 0=@n4.Minutes, 1=@n4.Seconds))
  ]}
]
variables [
 AppearSoundCue {type=SoundBase, default=/Game/Sounds/Fort_UI_Sounds/FrontEnd/New_FrontEnd/Ranked/Fort_UI_Ranked_Whoosh_Intro_Cue}
 bScrollUp {type=boolean}
 CurrentIndex {type=integer}
 DataInfo {
  type="array<BP_UI_Habanero_InfoPanel_C>"
  default [
   /Game/UI/Frontend/Lobby/Habanero/HabaneroData/Habanero_HabaneroInfoData
   /Game/UI/Frontend/Lobby/Habanero/HabaneroData/Habanero_LeaverPenaltiesData
   /Game/UI/Frontend/Lobby/Habanero/HabaneroData/Habanero_ProgressionInfoData
   /Game/UI/Frontend/Lobby/Habanero/HabaneroData/Habanero_SeasonInfoData
   /Game/UI/Frontend/Lobby/Habanero/HabaneroData/Habanero_MatchmakingInfoData
   /Game/UI/Frontend/Lobby/Habanero/HabaneroData/Habanero_TournamentsInfoData
   /Game/UI/Frontend/Lobby/Habanero/HabaneroData/Habanero_RankProtectionShieldInfoData
   /Game/UI/Frontend/Lobby/Habanero/HabaneroData/Habanero_PlaylistsScheduleInfoData_Blastberry
  ]
 }
 HoverSound {type=SoundBase, default=/Game/Sounds/Fort_UI_Sounds/FrontEnd/New_FrontEnd/Ranked/Fortnite_UI_Ranked_Button_Hover_Cue}
 LastIndex {type=integer, default=1}
 Pins {type=Object, default=null}
 RankData {type=FortHabaneroDisplayData, default=/Game/UI/Habanero/HabaneroDisplayData}
 ScrollDownSoundCue {type=SoundBase, default=/Game/Sounds/Fort_UI_Sounds/FrontEnd/New_FrontEnd/Ranked/Fort_UI_Shop_Whoosh_Down_Cue}
 ScrollUpSoundCue {type=SoundBase, default=/Game/Sounds/Fort_UI_Sounds/FrontEnd/New_FrontEnd/Ranked/Fort_UI_Ranked_Whoosh_Up_Cue}
 SelectSound {type=SoundBase, default=/Game/Sounds/Fort_UI_Sounds/FrontEnd/New_FrontEnd/Ranked/Fortnite_UI_Ranked_Button_Default_Cue}
 TabsDataInfo {type="array<UIKitTabGroupButtonViewModel>", default []}
 UnrealRankCapText {type=string, default="##100k+"}
]
widgets {
 info {inheritedWidgetCount=4, parentClass=/Script/FortniteUI.FortHabaneroInfoModal, rootWidgetClass=&Overlay, widgetCount=41}
 widgets [
  Overlay_31 {widgetClassPath=&Overlay, widget=Overlay_31}
  BackgroundImage_HitTestBlocker {
   widgetClassPath=&Image
   bIsVariable=true
   parent=Overlay_31
   widget=BackgroundImage_HitTestBlocker
   properties {brush {
     bIsDynamicallyLoaded=false
     drawAs=Image
     imageType=NoImage
     mirroring=NoMirror
     resourceName=None
     resourceObject=/Game/Athena/UI/Frontend/Art/M_UI_RadialGradientMaster_Dark
     tiling=NoTile
     imageSize=<32 32>
     margin=<ltrb 0 0 0 0>
     outlineSettings {
      bUseBrushTransparency=false
      roundingType=HalfHeightRadius
      width=0
      color {colorUseRule=UseColor_Specified, specifiedColor=<rgba 0 0 0 0>}
      cornerRadii=<0 0 0 0>
     }
     tintColor {colorUseRule=UseColor_Specified, specifiedColor=<rgba 0.5972 0.5972 0.5972 1>}
     uVRegion {bIsValid=false, max=<0 0>, min=<0 0>}
    }}
   slot {class=OverlaySlot, properties {horizontalAlignment=HAlign_Fill, verticalAlignment=VAlign_Fill}}
  }
  Overlay_23 {
   widgetClassPath=&Overlay
   parent=Overlay_31
   widget=Overlay_23
   slot {class=OverlaySlot, properties {horizontalAlignment=HAlign_Fill, verticalAlignment=VAlign_Fill}}
  }
  HorizontalBox_0 {
   widgetClassPath=&HorizontalBox
   parent=Overlay_23
   widget=HorizontalBox_0
   slot {class=OverlaySlot, properties {horizontalAlignment=HAlign_Fill, verticalAlignment=VAlign_Fill, padding=<ltrb 0 0 64 0>}}
  }
  Border_LeftContent {
   widgetClassPath=&CommonBorder
   parent=HorizontalBox_0
   widget=Border_LeftContent
   properties {
    background {
     bIsDynamicallyLoaded=false
     drawAs=Image
     imageType=NoImage
     mirroring=NoMirror
     resourceName=None
     resourceObject=/Game/UI/Frontend/Lobby/Habanero/Habanero_Materials/MaterialInstances/Background/MI_UI_Habanero_SidePanel_Background
     tiling=NoTile
     imageSize=<64 64>
     margin=<ltrb 0 0 0 0>
     outlineSettings {
      bUseBrushTransparency=false
      roundingType=HalfHeightRadius
      width=0
      color {colorUseRule=UseColor_Specified, specifiedColor=<rgba 0 0 0 0>}
      cornerRadii=<0 0 0 0>
     }
     tintColor {colorUseRule=UseColor_Specified, specifiedColor=<rgba 1 1 1 1>}
     uVRegion {bIsValid=false, max=<0 0>, min=<0 0>}
    }
    padding=<ltrb 0 0 0 0>
   }
   slot {class=HorizontalBoxSlot}
  }
  Overlay_1 {
   widgetClassPath=&Overlay
   parent=Border_LeftContent
   widget=Overlay_1
   slot {class=BorderSlot, properties {padding=<ltrb 0 0 0 0>}}
  }
  WBP_CaptureForPostBufferUpdate {
   widgetClassPath=&WBP_CaptureForPostBufferUpdate
   bIsVariable=true
   parent=Overlay_1
   slot {class=OverlaySlot, properties {horizontalAlignment=HAlign_Fill, verticalAlignment=VAlign_Fill}}
  }
  SafeZone_content {
   widgetClassPath=&SafeZone
   bIsVariable=true
   parent=Overlay_1
   widget=SafeZone_content
   properties {padBottom=false, padRight=false, padTop=false}
   slot {class=OverlaySlot, properties {horizontalAlignment=HAlign_Fill, verticalAlignment=VAlign_Fill, padding=<ltrb 32 32 32 32>}}
  }
  SizeBox_0 {
   widgetClassPath=&SizeBox
   parent=SafeZone_content
   widget=SizeBox_0
   properties {bOverride_MaxDesiredWidth=true, bOverride_WidthOverride=true, maxDesiredWidth=360, widthOverride=360}
   slot {class=SafeZoneSlot}
  }
  LeftSide {widgetClassPath=&VerticalBox, parent=SizeBox_0, widget=LeftSide, slot {class=SizeBoxSlot}}
  HorizontalBox_Mobile {
   widgetClassPath=&HorizontalBox
   bIsVariable=true
   parent=LeftSide
   widget=HorizontalBox_Mobile
   slot {class=VerticalBoxSlot, properties {horizontalAlignment=HAlign_Center, verticalAlignment=VAlign_Center}}
  }
  VerticalBox_0 {widgetClassPath=&VerticalBox, parent=HorizontalBox_Mobile, widget=VerticalBox_0, slot {class=HorizontalBoxSlot}}
  Overlay_0 {widgetClassPath=&Overlay, parent=VerticalBox_0, widget=Overlay_0, slot {class=VerticalBoxSlot}}
  Spacer_72 {
   widgetClassPath=&Spacer
   bIsVariable=true
   parent=Overlay_0
   widget=Spacer_72
   properties {size=<212 212>}
   slot {class=OverlaySlot, properties {horizontalAlignment=HAlign_Fill, verticalAlignment=VAlign_Fill}}
  }
  SizeBox_ProgressInfo {
   widgetClassPath=&SizeBox
   bIsVariable=true
   parent=Overlay_0
   widget=SizeBox_ProgressInfo
   slot {class=OverlaySlot, properties {horizontalAlignment=HAlign_Center, verticalAlignment=VAlign_Center}}
  }
  WBP_Habanero_ProgressInfo {
   widgetClassPath=&/WBP_Habanero_ProgressInfo
   bIsVariable=true
   parent=SizeBox_ProgressInfo
   widget=WBP_Habanero_ProgressInfo
   properties {visibility=HitTestInvisible}
   slot {class=SizeBoxSlot}
  }
  VerticalBox_Mobile {
   widgetClassPath=&VerticalBox
   bIsVariable=true
   parent=VerticalBox_0
   widget=VerticalBox_Mobile
   slot {class=VerticalBoxSlot, properties {horizontalAlignment=HAlign_Center, verticalAlignment=VAlign_Top}}
  }
  PercentageTextWrapper {
   widgetClassPath=&Border
   bIsVariable=true
   parent=VerticalBox_Mobile
   widget=PercentageTextWrapper
   properties {
    background {
     bIsDynamicallyLoaded=false
     drawAs=Image
     imageType=NoImage
     mirroring=NoMirror
     resourceName=None
     resourceObject=/Game/UI/Frontend/Lobby/Habanero/Habanero_Materials/MaterialInstances/Background/MI_UI_Habanero_ProgressText_Bg
     tiling=NoTile
     imageSize=<32 32>
     margin=<ltrb 0 0 0 0>
     outlineSettings {
      bUseBrushTransparency=false
      roundingType=HalfHeightRadius
      width=0
      color {colorUseRule=UseColor_Specified, specifiedColor=<rgba 0 0 0 0>}
      cornerRadii=<0 0 0 0>
     }
     tintColor {colorUseRule=UseColor_Specified, specifiedColor=<rgba 1 1 1 0>}
     uVRegion {bIsValid=false, max=<0 0>, min=<0 0>}
    }
    padding=<ltrb 8 0 8 0>
   }
   slot {class=VerticalBoxSlot, properties {horizontalAlignment=HAlign_Center, verticalAlignment=VAlign_Top}}
  }
  Text_Percent {
   widgetClassPath=&CommonTextBlock
   bIsVariable=true
   parent=PercentageTextWrapper
   widget=Text_Percent
   properties {
    autoWrapText=true
    justification=Center
    lineHeightPercentage=1.29
    mobileFontSizeMultiplier=1.5
    style=&_TextStyle_Body1_C
    text=0%
    textTransformPolicy=ToUpper
    font {
     bForceMonospaced=false
     bMaterialIsStencil=false
     fontMaterial=null
     fontObject=&BurbankSmall
     letterSpacing=0
     monospacedWidth=1
     size=18
     skewAmount=0
     typefaceFontName=Bold
     outlineSettings {
      bApplyOutlineToDropShadows=false
      bMiteredCorners=false
      bSeparateFillAlpha=false
      outlineMaterial=null
      outlineSize=0
      outlineColor=<rgba 0 0 0 1>
     }
    }
    shadowOffset=<0 0>
    strikeBrush {
     bIsDynamicallyLoaded=false
     drawAs=NoDrawType
     imageType=NoImage
     mirroring=NoMirror
     resourceName=None
     resourceObject=null
     tiling=NoTile
     imageSize=<32 32>
     margin=<ltrb 0 0 0 0>
     outlineSettings {
      bUseBrushTransparency=false
      roundingType=HalfHeightRadius
      width=0
      color {colorUseRule=UseColor_Specified, specifiedColor=<rgba 0 0 0 0>}
      cornerRadii=<0 0 0 0>
     }
     tintColor {colorUseRule=UseColor_Specified, specifiedColor=<rgba 1 1 1 1>}
     uVRegion {bIsValid=false, max=<0 0>, min=<0 0>}
    }
   }
   slot {class=BorderSlot, properties {padding=<ltrb 8 0 8 0>}}
  }
  Tier_Name {
   widgetClassPath=&CommonTextBlock
   bIsVariable=true
   parent=VerticalBox_Mobile
   widget=Tier_Name
   properties {
    justification=Center
    mobileFontSizeMultiplier=1.5
    style=/FortUILibrary/Styles/Typography/_TextStyle_SubH4._TextStyle_SubH4_C
    text="{Habanero}"
    textTransformPolicy=ToUpper
    font {
     bForceMonospaced=false
     bMaterialIsStencil=false
     fontMaterial=null
     fontObject=&BurbankSmall
     letterSpacing=0
     monospacedWidth=1
     size=24
     skewAmount=0
     typefaceFontName=Black
     outlineSettings {
      bApplyOutlineToDropShadows=false
      bMiteredCorners=false
      bSeparateFillAlpha=false
      outlineMaterial=null
      outlineSize=0
      outlineColor=<rgba 0 0 0 1>
     }
    }
    shadowOffset=<0 0>
    strikeBrush {
     bIsDynamicallyLoaded=false
     drawAs=NoDrawType
     imageType=NoImage
     mirroring=NoMirror
     resourceName=None
     resourceObject=null
     tiling=NoTile
     imageSize=<32 32>
     margin=<ltrb 0 0 0 0>
     outlineSettings {
      bUseBrushTransparency=false
      roundingType=HalfHeightRadius
      width=0
      color {colorUseRule=UseColor_Specified, specifiedColor=<rgba 0 0 0 0>}
      cornerRadii=<0 0 0 0>
     }
     tintColor {colorUseRule=UseColor_Specified, specifiedColor=<rgba 1 1 1 1>}
     uVRegion {bIsValid=false, max=<0 0>, min=<0 0>}
    }
   }
   slot {class=VerticalBoxSlot, properties {horizontalAlignment=HAlign_Center, verticalAlignment=VAlign_Center}}
  }
  Spacer_Mobile {
   widgetClassPath=&Spacer
   bIsVariable=true
   parent=VerticalBox_0
   widget=Spacer_Mobile
   properties {size=<0 0>}
   slot {class=VerticalBoxSlot}
  }
  Button_Tabs {
   widgetClassPath=/FortUILibrary/UIKit/Tabs/WBP_UIKit_Tabs
   bIsVariable=true
   parent=LeftSide
   widget=Button_Tabs
   properties {
    hidePrompts=true
    inputActions=None
    isNaturalNavigation=true
    isVertical=true
    minButtonWidth=400
    navigation=&self:WidgetTree.Button_Tabs.WidgetNavigation_0
    nextPromptHorizontalAlignment=HAlign_Left
    iconSize=<0 0>
    nextPromptPadding=<ltrb 64 0 0 0>
   }
   slot {class=VerticalBoxSlot, properties {horizontalAlignment=HAlign_Center, size {sizeRule=Fill, value=1}}}
  }
  Timer_HBox {
   widgetClassPath=&HorizontalBox
   bIsVariable=true
   parent=LeftSide
   widget=Timer_HBox
   properties {flowDirectionPreference=Culture, visibility=Collapsed}
   slot {class=VerticalBoxSlot, properties {horizontalAlignment=HAlign_Left, verticalAlignment=VAlign_Bottom, padding=<ltrb 32 0 32 0>}}
  }
  Image_Icon_Countdown {
   widgetClassPath=&Image
   parent=Timer_HBox
   widget=Image_Icon_Countdown
   properties {
    visibility=SelfHitTestInvisible
    brush {
     bIsDynamicallyLoaded=false
     drawAs=Image
     imageType=NoImage
     mirroring=NoMirror
     resourceName=None
     resourceObject=/Game/Athena/UI/Frontend/ActivityBrowser/Art/M_UI_Timer_Icon_ShopOffer
     tiling=NoTile
     imageSize=<64 64>
     margin=<ltrb 0 0 0 0>
     outlineSettings {
      bUseBrushTransparency=false
      roundingType=HalfHeightRadius
      width=0
      color {colorUseRule=UseColor_Specified, specifiedColor=<rgba 0 0 0 0>}
      cornerRadii=<0 0 0 0>
     }
     tintColor {colorUseRule=UseColor_Specified, specifiedColor=<rgba 1 1 1 1>}
     uVRegion {bIsValid=false, max=<0 0>, min=<0 0>}
    }
    colorAndOpacity=<rgba 0.0314 0.702 1 1>
   }
   slot {class=HorizontalBoxSlot, properties {horizontalAlignment=HAlign_Center, verticalAlignment=VAlign_Center, size {sizeRule=Fill, value=1}}}
  }
  VerticalBox {widgetClassPath=&VerticalBox, parent=Timer_HBox, widget=VerticalBox, slot {class=HorizontalBoxSlot}}
  Border_0 {
   widgetClassPath=&Border
   parent=VerticalBox
   widget=Border_0
   properties {
    visibility=HitTestInvisible
    background {
     bIsDynamicallyLoaded=false
     drawAs=NoDrawType
     imageType=NoImage
     mirroring=NoMirror
     resourceName=None
     resourceObject=null
     tiling=NoTile
     imageSize=<32 32>
     margin=<ltrb 0 0 0 0>
     outlineSettings {
      bUseBrushTransparency=false
      roundingType=HalfHeightRadius
      width=0
      color {colorUseRule=UseColor_Specified, specifiedColor=<rgba 0 0 0 0>}
      cornerRadii=<0 0 0 0>
     }
     tintColor {colorUseRule=UseColor_Specified, specifiedColor=<rgba 1 1 1 1>}
     uVRegion {bIsValid=false, max=<0 0>, min=<0 0>}
    }
    contentColorAndOpacity=<rgba 0.0319 0.7011 1 1>
    padding=<ltrb 0 0 0 0>
   }
   slot {class=VerticalBoxSlot}
  }
  Text_TimerLabel {
   widgetClassPath=&CommonTextBlock
   bIsVariable=true
   parent=Border_0
   widget=Text_TimerLabel
   properties {
    mobileFontSizeMultiplier=0.9
    text=Season ends in:
    font {
     bForceMonospaced=false
     bMaterialIsStencil=false
     fontMaterial=null
     fontObject=&BurbankSmall
     letterSpacing=0
     monospacedWidth=1
     size=18
     skewAmount=0
     typefaceFontName=Black
     outlineSettings {
      bApplyOutlineToDropShadows=false
      bMiteredCorners=false
      bSeparateFillAlpha=false
      outlineMaterial=null
      outlineSize=0
      outlineColor=<rgba 0 0 0 1>
     }
    }
    margin=<ltrb 4 0 4 0>
    shadowOffset=<0 0>
   }
   slot {class=BorderSlot, properties {padding=<ltrb 0 0 0 0>}}
  }
  Text_TimerTime {
   widgetClassPath=&CommonTextBlock
   bIsVariable=true
   parent=VerticalBox
   widget=Text_TimerTime
   properties {
    mobileFontSizeMultiplier=0.9
    style=&_TextStyle_SubH4_C
    text="<Time>"
    textTransformPolicy=ToUpper
    font {
     bForceMonospaced=false
     bMaterialIsStencil=false
     fontMaterial=null
     fontObject=&BurbankSmall
     letterSpacing=0
     monospacedWidth=1
     size=24
     skewAmount=0
     typefaceFontName=Black
     outlineSettings {
      bApplyOutlineToDropShadows=false
      bMiteredCorners=false
      bSeparateFillAlpha=false
      outlineMaterial=null
      outlineSize=0
      outlineColor=<rgba 0 0 0 1>
     }
    }
    shadowOffset=<0 0>
    strikeBrush {
     bIsDynamicallyLoaded=false
     drawAs=NoDrawType
     imageType=NoImage
     mirroring=NoMirror
     resourceName=None
     resourceObject=null
     tiling=NoTile
     imageSize=<32 32>
     margin=<ltrb 0 0 0 0>
     outlineSettings {
      bUseBrushTransparency=false
      roundingType=HalfHeightRadius
      width=0
      color {colorUseRule=UseColor_Specified, specifiedColor=<rgba 0 0 0 0>}
      cornerRadii=<0 0 0 0>
     }
     tintColor {colorUseRule=UseColor_Specified, specifiedColor=<rgba 1 1 1 1>}
     uVRegion {bIsValid=false, max=<0 0>, min=<0 0>}
    }
   }
   slot {class=VerticalBoxSlot}
  }
  Spacer_187 {widgetClassPath=&Spacer, parent=HorizontalBox_0, widget=Spacer_187, properties {size=<32 1>}, slot {class=HorizontalBoxSlot}}
  Overlay_2 {
   widgetClassPath=&Overlay
   parent=HorizontalBox_0
   widget=Overlay_2
   slot {class=HorizontalBoxSlot, properties {size {sizeRule=Fill, value=1}}}
  }
  ScrollBox_Panels {
   widgetClassPath=/Script/FortniteUI.AthenaScrollBox
   bInherited=true
   bIsVariable=true
   parent=Overlay_2
   widget=ScrollBox_Panels
   properties {
    allowOverscroll=false
    bAllowRightClickDragScrolling=false
    bAnimateWheelScrolling=false
    bEnableClickThrough=true
    consumeMouseWheel=Never
    scrollBarVisibility=Collapsed
    scrollWhenFocusChanges=AnimatedScroll
    visibility=SelfHitTestInvisible
   }
   slot {class=OverlaySlot, properties {horizontalAlignment=HAlign_Fill, verticalAlignment=VAlign_Fill}}
  }
  Vbox_Panels {
   widgetClassPath=&VerticalBox
   bInherited=true
   bIsVariable=true
   parent=ScrollBox_Panels
   widget=Vbox_Panels
   slot {class=ScrollBoxSlot}
  }
  Text_Header {
   widgetClassPath=&VerticalBox
   parent=Overlay_2
   widget=Text_Header
   slot {class=OverlaySlot, properties {horizontalAlignment=HAlign_Fill, padding=<ltrb 64 72 92 0>}}
  }
  Text_HeaderSub {
   widgetClassPath=&CommonTextBlock
   bIsVariable=true
   parent=Text_Header
   widget=Text_HeaderSub
   properties {
    mobileFontSizeMultiplier=1.4
    text=MODE INFORMATION
    textTransformPolicy=ToUpper
    colorAndOpacity {colorUseRule=UseColor_Specified, specifiedColor=<rgba 0.0314 0.702 1 1>}
    font {
     bForceMonospaced=false
     bMaterialIsStencil=false
     fontMaterial=null
     fontObject=&Burbank
     letterSpacing=0
     monospacedWidth=1
     size=18
     skewAmount=0
     typefaceFontName=BurbankRegular-Black
     outlineSettings {
      bApplyOutlineToDropShadows=false
      bMiteredCorners=false
      bSeparateFillAlpha=false
      outlineMaterial=null
      outlineSize=0
      outlineColor=<rgba 0 0 0 1>
     }
    }
    margin=<ltrb 4 0 4 0>
    shadowOffset=<0 0>
   }
   slot {class=VerticalBoxSlot}
  }
  Text_HeaderMain {
   widgetClassPath=&CommonTextBlock
   bIsVariable=true
   parent=Text_Header
   widget=Text_HeaderMain
   properties {
    autoWrapText=true
    text=RANKED GAME MODES
    textOverflowPolicy=Ellipsis
    textTransformPolicy=ToUpper
    font {
     bForceMonospaced=false
     bMaterialIsStencil=false
     fontMaterial=null
     fontObject=&Burbank
     letterSpacing=0
     monospacedWidth=1
     size=40
     skewAmount=0
     typefaceFontName=BurbankRegular-Black
     outlineSettings {
      bApplyOutlineToDropShadows=false
      bMiteredCorners=false
      bSeparateFillAlpha=false
      outlineMaterial=null
      outlineSize=0
      outlineColor=<rgba 0 0 0 1>
     }
    }
    margin=<ltrb 4 0 4 0>
    shadowOffset=<0 0>
   }
   slot {class=VerticalBoxSlot}
  }
  WBP_CaptureForPostBufferUpdate_46 {
   widgetClassPath=&WBP_CaptureForPostBufferUpdate
   bIsVariable=true
   parent=Overlay_2
   widget=WBP_CaptureForPostBufferUpdate_46
   slot {class=OverlaySlot, properties {horizontalAlignment=HAlign_Fill, verticalAlignment=VAlign_Fill}}
  }
  SafeZone_bottombuttons {
   widgetClassPath=&SafeZone
   bIsVariable=true
   parent=Overlay_23
   widget=SafeZone_bottombuttons
   properties {padLeft=false, padTop=false}
   slot {class=OverlaySlot, properties {horizontalAlignment=HAlign_Right, verticalAlignment=VAlign_Bottom, padding=<ltrb 0 0 36 32>}}
  }
  Overlay_3 {widgetClassPath=&Overlay, parent=SafeZone_bottombuttons, widget=Overlay_3, slot {class=SafeZoneSlot}}
  CaptureForPostBufferUpdate_159 {
   widgetClassPath=&WBP_CaptureForPostBufferUpdate
   parent=Overlay_3
   widget=CaptureForPostBufferUpdate_159
   slot {class=OverlaySlot, properties {horizontalAlignment=HAlign_Fill, verticalAlignment=VAlign_Fill}}
  }
  Button_Close_Alt {
   widgetClassPath=/FortUILibrary/UIKit/Buttons/WBP_UIKit_Button_Quiet
   bInherited=true
   bIsVariable=true
   parent=Overlay_3
   widget=Button_Close_Alt
   properties {
    bDisplayInputActionWhenNotInteractable=false
    bHideInputAction=true
    bIsFocusable=false
    renderOpacity=0
    text=Close
    hoveredSlateSoundOverride {resourceObject=&Fortnite_UI_Ranked_Button_Hover_Cue}
    pressedSlateSoundOverride {resourceObject=/Game/Sounds/Fort_UI_Sounds/FrontEnd/New_FrontEnd/Ranked/Fortnite_UI_Ranked_Button_Back_Cue}
    selectedHoveredSlateSoundOverride {resourceObject=&Fortnite_UI_Ranked_Button_Hover_Cue}
    selectedPressedSlateSoundOverride {resourceObject=&Fortnite_UI_Ranked_Button_Back_Cue}
    triggeringInputAction {dataTable=/Game/Athena/UI/Frontend/Lobby/AthenaLobbyInputActionDataTable, rowName=HabaneroToggle}
   }
   slot {class=OverlaySlot}
  }
  Button_Close {
   widgetClassPath=/Game/UI/Frontend/BottomBar/WBP_BottomBarDecoyButton
   bInherited=true
   bIsVariable=true
   parent=Overlay_31
   widget=Button_Close
   properties {inputPriority=1001, priority=1001, triggeringInputAction {dataTable=&InputActionDataTable, rowName=Close}}
   slot {class=OverlaySlot}
  }
 ]
}
