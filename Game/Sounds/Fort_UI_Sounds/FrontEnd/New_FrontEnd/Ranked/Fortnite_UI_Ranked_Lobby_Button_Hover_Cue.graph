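# Sound cue for hovering over a Ranked lobby button in the Fortnite front-end UI.
# A Mixer combines two wave players, with the default hover sound muted at zero volume, leaving only the Ranked status slide-wipe audible at slightly reduced volume.
# It uses the UI sound class with virtualization disabled.
# It references a Ranked hover concurrency set alongside stop-oldest override settings that cap it at two instances.
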SoundCue /Game/Sounds/Fort_UI_Sounds/FrontEnd/New_FrontEnd/Ranked/Fortnite_UI_Ranked_Lobby_Button_Hover_Cue
root=Mixer_1
nodes [
 {id=Mixer_1, children [WavePlayer_1,WavePlayer_0], properties {inputVolume [0,0.9]}}
 {
  id=WavePlayer_0
  properties {soundWaveAssetPtr=/Game/Sounds/Fort_UI_Sounds/FrontEnd/New_FrontEnd/Ranked/wav/UI_Ranked_Lobby_Status_Hover_Slide_Wipe_01}
 }
 {id=WavePlayer_1, properties {soundWaveAssetPtr=/Game/Sounds/Fort_UI_Sounds/FrontEnd/DefaultSounds/PressedButtons/wav/UI_MP_Hover_Default_01}}
]
properties {
 maxDistance=2097152
 priority=2
 soundClassObject=&Fort_UI
 virtualizationMode=Disabled
 volumeMultiplier=1
 concurrencyOverrides {maxCount=2, resolutionRule=StopOldest}
 concurrencySet [&/RankedHoverConcurrency]
}
