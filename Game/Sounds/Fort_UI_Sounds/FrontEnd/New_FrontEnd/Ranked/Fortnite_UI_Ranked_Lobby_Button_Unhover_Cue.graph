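# Sound cue for the unhover sound of the Ranked lobby button in the front-end UI.
# A single wave player feeds the output with one slide-wipe wave, with no randomization or modulation nodes.
# The cue routes to the Fort_UI sound class at slightly reduced volume, uses a Ranked hover concurrency group to govern overlapping plays, and has virtualization disabled.
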SoundCue /Game/Sounds/Fort_UI_Sounds/FrontEnd/New_FrontEnd/Ranked/Fortnite_UI_Ranked_Lobby_Button_Unhover_Cue
root=WavePlayer_0
nodes [{
  id=WavePlayer_0
  properties {soundWaveAssetPtr=/Game/Sounds/Fort_UI_Sounds/FrontEnd/New_FrontEnd/Ranked/wav/UI_Ranked_Lobby_Status_Unhover_Slide_Wipe_01}
 }]
properties {
 maxDistance=2097152
 soundClassObject=&Fort_UI
 virtualizationMode=Disabled
 volumeMultiplier=0.9
 concurrencySet [&/RankedHoverConcurrency]
}
