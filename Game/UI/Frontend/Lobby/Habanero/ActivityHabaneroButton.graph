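# Lobby button widget titled "Ranked" that shows an On/Off toggle and Habanero icon, based on FortActivityHabaneroButton.
# Selection changes update TextToggle and play ToggleEmblemState forward or backward to light up the icon, gradients and triangle effect.
# Enable state drives a dimmer overlay and gates hover and press animations.
# The Pressed animation has no tracks, and PlayToggleAudio ends at a branch without playing sound.
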WidgetBlueprint /Game/UI/Frontend/Lobby/Habanero/ActivityHabaneroButton
animations [
 Hover {
  binding_count=2
  length_seconds=0.2
  object=&self:Hover
  sequence {
   bindings [
    Image_Bg {
     id=#1
     possessed_class=Image
     tracks [{
       type=MovieSceneWidgetMaterialTrack
       sections [{type=MovieSceneParameterSection, channels [AnimateHighlight {type=MovieSceneScriptingFloatChannel, keys [[0,0],[15.035,1],[19.995,1]]}]}]
      }]
    }
    Border_Title {
     id=#2
     possessed_class=Border
     tracks [{
       type=MovieSceneColorTrack
       display_name=Content Color and Opacity
       sections [{
         type=MovieSceneColorSection
         channels [
          Color.R {type=MovieSceneScriptingFloatChannel, keys [[0,1],[15,0]]}
          Color.G {type=MovieSceneScriptingFloatChannel, keys [[0,1],[15,0.004]]}
          Color.B {type=MovieSceneScriptingFloatChannel, keys [[0,1],[15,0.0723]]}
          Color.A {type=MovieSceneScriptingFloatChannel}
         ]
        }]
      }]
    }
   ]
   display_rate {numerator=100}
  }
 }
 Pressed {
  binding_count=0
  length_seconds=0.16
  object=&self:Pressed
  sequence {bindings [], display_rate {numerator=100}}
 }
 ToggleEmblemState {
  binding_count=4
  length_seconds=0.25
  object=&self:ToggleEmblemState
  sequence {bindings [
    HabaneroWidget {
     id=#3
     possessed_class=WBP_Habanero_Icon_C
     tracks [{
       type=MovieSceneColorTrack
       display_name=Color and Opacity
       sections [{
         type=MovieSceneColorSection
         end_frame=5
         start_frame=0
         channels [
          Color.R {type=MovieSceneScriptingFloatChannel}
          Color.G {type=MovieSceneScriptingFloatChannel}
          Color.B {type=MovieSceneScriptingFloatChannel}
          Color.A {type=MovieSceneScriptingFloatChannel, keys [[0,0.5],[5,1]]}
         ]
        }]
      }]
    }
    ImageGradientLime {
     id=#4
     possessed_class=Image
     tracks [{
       type=MovieSceneFloatTrack
       display_name=Render Opacity
       sections [{type=MovieSceneFloatSection, end_frame=5, start_frame=0, channels [None {type=MovieSceneScriptingFloatChannel, keys [[1,0],[3,1]]}]}]
      }]
    }
    ImageTriangle {
     id=#5
     possessed_class=Image
     tracks [
      {
       type=MovieSceneWidgetMaterialTrack
       sections [{
         type=MovieSceneParameterSection
         end_frame=5
         start_frame=0
         channels [FlareDensity {type=MovieSceneScriptingFloatChannel, keys [[0,0.6],[5,1.75]]}]
        }]
      }
      {
       type=MovieSceneColorTrack
       display_name=Color and Opacity
       sections [{
         type=MovieSceneColorSection
         end_frame=5
         start_frame=0
         channels [
          Color.R {type=MovieSceneScriptingFloatChannel, keys [[0,1],[5,0.402]]}
          Color.G {type=MovieSceneScriptingFloatChannel, const [1,1]}
          Color.B {type=MovieSceneScriptingFloatChannel, const [1,1]}
          Color.A {type=MovieSceneScriptingFloatChannel, keys [[0,0.02],[5,0.75]]}
         ]
        }]
      }
     ]
    }
    ImageGradientBlue {
     id=#6
     possessed_class=Image
     tracks [{
       type=MovieSceneColorTrack
       display_name=Color and Opacity
       sections [{
         type=MovieSceneColorSection
         end_frame=5
         start_frame=0
         channels [
          Color.R {type=MovieSceneScriptingFloatChannel, keys [[0,0.0144],[5,0]]}
          Color.G {type=MovieSceneScriptingFloatChannel, keys [[0,0.0356],[5,0.0319]]}
          Color.B {type=MovieSceneScriptingFloatChannel, keys [[0,0.147],[5,0.2961]]}
          Color.A {type=MovieSceneScriptingFloatChannel, const [1,2]}
         ]
        }]
      }]
    }
   ]}
 }
 Disabled {
  binding_count=1
  length_seconds=0.2
  object=&self:Disabled
  sequence {bindings [Image_Dimmer {
     id=#7
     possessed_class=Image
     tracks [{
       type=MovieSceneFloatTrack
       display_name=Render Opacity
       sections [{type=MovieSceneFloatSection, end_frame=4, start_frame=0, channels [None {type=MovieSceneScriptingFloatChannel, keys [[0,0],[4,1]]}]}]
      }]
    }]}
 }
]
graphs [
 EventGraph {nodes [
   PlayToggleAudio() {is=title K2Node_CustomEvent}
   > Branch(Condition=Get bSelection() -> (bSelection:bool))
   Event PreConstruct()
   > SetSidesToPad!(
    execute
    self:Safe Zone Object Reference=Get SafeZone() -> (SafeZone:Safe Zone Object Reference)
    InPadLeft:bool=Get PadLeft() -> (PadLeft:bool)
    InPadRight:bool=Get PadRight() -> (PadRight:bool)
    InPadTop:bool=false
    InPadBottom:bool=false
   ) -> (then)
   n1 = Event OnDesiredHabaneroSelectionChanged!() -> (OutputDelegate, then, bIsSelected:bool) {is=title K2Node_Event}
   > n2 = Set bSelection(bSelection:bool=@n1.bIsSelected) -> (Output_Get:bool)
   n3 = Get bSelection()
   SetToggleState!(execute=[@n9.then, @n2.then], self:Text Toggle Object Reference=Get TextToggle() -> (TextToggle:Text Toggle Object Reference), IsOn:bool=@n3) -> (then)
   > PlayAnimation(
    InAnimation=Get ToggleEmblemState() -> (ToggleEmblemState:Widget Animation Object Reference)
    PlayMode=Select(Option 0:EUMGSequencePlayMode Enum=Reverse, Option 1:EUMGSequencePlayMode Enum, Index=@n3) -> (ReturnValue:EUMGSequencePlayMode Enum)
   )
   Event On Clicked()
   > n4 = Sequence!(execute) -> (then_0, then_1, then_2:Exec)
   > n5 = Branch(Condition=Get isDisabled() -> (isDisabled:bool))
   Set ShouldPlayReversed(execute=@n5.else, ShouldPlayReversed:bool=true) -> (Output_Get:bool)
   n6 = Get Pressed() -> (Pressed:Widget Animation Object Reference)
   > n7 = "Play Animation Time Range with Finished event (legacy)"!(
    execute:Exec
    Widget:User Widget Object Reference=Self-Reference()
    InAnimation:Widget Animation Object Reference=@n6
    StartAtTime:float=GetAnimationCurrentTime!(self:User Widget Object Reference, InAnimation:Widget Animation Object Reference=@n6) -> (ReturnValue:float)
    EndAtTime:float=0
    NumLoopsToPlay:int=1
    PlayMode:EUMGSequencePlayMode Enum
    PlaybackSpeed:float=1
   ) -> (then:Exec, Result:UMGSequence Player Object Reference, Finished:Exec) {is=title K2Node_PlayAnimationTimeRange}
   Branch(execute=@n7.Finished, Condition=Get ShouldPlayReversed() -> (ShouldPlayReversed:bool))
   > Set ShouldPlayReversed(ShouldPlayReversed=false)
   > n8 = Branch(Condition=Get isDisabled())
   PlayAnimationReverse(execute=@n8.else, InAnimation=Get Pressed())
   PlayToggleAudio(execute=@n4.then_1) {is=title K2Node_CallFunction}
   Branch(
    execute=@n4.then_2
    Condition=IsUsingGamepad!(self:Fort Player Controller Object Reference=GetFirstLocalFortPlayerController!(bIncludeNonSpectators:bool=true, bIncludeSpectators:bool=false) -> (ReturnValue:Fort Player Controller Object Reference)) -> (ReturnValue:bool)
   )
   > PlayToggleAudio()
   Event Construct()
   > n9 = Toggle Disabled Dimmer!(execute, self, Is Disabled:bool=Get isDisabled()) -> (then)
   Event On Disabled() {is=title K2Node_Event}
   > n10 = Set isDisabled(isDisabled:bool=true) -> (Output_Get:bool)
   Toggle Disabled Dimmer(execute=[@n11.then, @n10.then], Is Disabled=Get isDisabled())
   Event On Enabled() {is=title K2Node_Event}
   > n11 = Set isDisabled(isDisabled=false)
   Event On Hovered()
   n12 = Get isDisabled()
   > n13 = Branch(Condition=@n12)
   n14 = Get Hover() -> (Hover:Widget Animation Object Reference)
   PlayAnimationForward(execute=@n13.else, InAnimation=@n14)
   Event On Unhovered()
   > n15 = Branch(Condition=@n12)
   PlayAnimationReverse(execute=@n15.else, InAnimation=@n14)
   Event On Selected()
   n16 = Get isDisabled()
   > n17 = Branch(Condition=@n16)
   n18 = Get Pressed()
   PlayAnimationForward(execute=@n17.else, InAnimation=@n18)
   Event On Deselected()
   > n19 = Branch(Condition=@n16)
   PlayAnimationReverse(execute=@n19.else, InAnimation=@n18)
  ]}
 Toggle Disabled Dimmer {nodes [
   n1 = Toggle Disabled Dimmer!() -> (then, Is Disabled:bool=~) {is=title K2Node_FunctionEntry}
   > PlayAnimation(InAnimation=Get Disabled() -> (Disabled:Widget Animation Object Reference), PlayMode=Select(Option 0=Reverse, Index=@n1))
  ]}
]
variables [
 bSelection {type=boolean}
 Button Description {type=string, default="{Text}"}
 ButtonBoxScaleSizeX {type=number, default=64}
 ButtonBoxScaleSizeY {type=number, default=64}
 ButtonMaterial {type=MaterialInterface, default=/Game/UI/DiscoveryShared/Art/MI_UI_Discovery_Button}
 ButtonSharpnessX {type=number, default=64}
 ButtonSharpnessY {type=number, default=64}
 DisabledParamName {type=string, default=bIsDisabled}
 HideButtonBackingOnGamepad {type=boolean}
 Is Fill Only Enabled {type=boolean}
 isDisabled {type=boolean}
 PadLeft {type=boolean}
 PadRight {type=boolean}
 SharpnessUParamName {type=string, default=Sharpness_U}
 SharpnessVParamName {type=string, default=Sharpness_V}
 ShouldPlayReversed {type=boolean}
 TextPadding {type=Margin, default=<ltrb 32 0 32 0>}
 TextShearX {type=number}
 TextShearY {type=number}
 UseBoxScaleUParamName {type=string, default=UseBoxScaleU}
 UseBoxScaleVParamName {type=string, default=UseBoxScaleV}
 UseBoxScalingX {type=number}
 UseBoxScalingY {type=number, default=1}
]
widgets {
 info {inheritedWidgetCount=2, parentClass=/Script/FortniteUI.FortActivityHabaneroButton, rootWidgetClass=&Overlay, widgetCount=16}
 widgets [
  Overlay_2 {widgetClassPath=&Overlay, widget=Overlay_2}
  ImageGradientBlue {
   widgetClassPath=&Image
   bIsVariable=true
   parent=Overlay_2
   widget=ImageGradientBlue
   properties {
    brush {
     bIsDynamicallyLoaded=false
     drawAs=Image
     imageType=FullColor
     mirroring=NoMirror
     resourceName=None
     resourceObject=/FortUILibrary/MaterialInstances/MI_UI_LinearGradient
     tiling=NoTile
     imageSize=<128 128>
     margin=<ltrb 0 0 0 0>
     outlineSettings {
      bUseBrushTransparency=false
      roundingType=HalfHeightRadius
      width=0
      color {colorUseRule=UseColor_Specified, specifiedColor=<rgba 0 0 0 0>}
      cornerRadii=<0 0 0 0>
     }
     tintColor {colorUseRule=UseColor_Specified, specifiedColor=<rgba 1 1 1 1>}
     uVRegion {bIsValid=false, max=<0 0>, min=<0 0>}
    }
    colorAndOpacity=<rgba 0 0.0319 0.2969 1>
    renderTransformPivot=<0.5 1>
   }
   slot {class=OverlaySlot, properties {horizontalAlignment=HAlign_Fill, verticalAlignment=VAlign_Center}}
  }
  ImageTriangle {
   widgetClassPath=&Image
   bIsVariable=true
   parent=Overlay_2
   widget=ImageTriangle
   properties {
    brush {
     bIsDynamicallyLoaded=false
     drawAs=Image
     imageType=FullColor
     mirroring=NoMirror
     resourceName=None
     resourceObject=/Game/UI/HUD/Materials/M_FX_TriangleScroll-Burst
     tiling=NoTile
     imageSize=<128 128>
     margin=<ltrb 0 0 0 0>
     outlineSettings {
      bUseBrushTransparency=false
      roundingType=HalfHeightRadius
      width=0
      color {colorUseRule=UseColor_Specified, specifiedColor=<rgba 0 0 0 0>}
      cornerRadii=<0 0 0 0>
     }
     tintColor {colorUseRule=UseColor_Specified, specifiedColor=<rgba 1 1 1 1>}
     uVRegion {bIsValid=false, max=<0 0>, min=<0 0>}
    }
    colorAndOpacity=<rgba 1 1 1 0.04>
    renderTransform {angle=0, scale=<-1 1>, shear=<0 0>, translation=<0 0>}
    renderTransformPivot=<0.5 1>
   }
   slot {class=OverlaySlot, properties {horizontalAlignment=HAlign_Fill, verticalAlignment=VAlign_Center}}
  }
  ImageGradientLime {
   widgetClassPath=&Image
   bIsVariable=true
   parent=Overlay_2
   widget=ImageGradientLime
   properties {
    visibility=SelfHitTestInvisible
    brush {
     bIsDynamicallyLoaded=false
     drawAs=Image
     imageType=FullColor
     mirroring=NoMirror
     resourceName=None
     resourceObject=&MI_UI_LinearGradient
     tiling=NoTile
     imageSize=<128 128>
     margin=<ltrb 0 0 0 0>
     outlineSettings {
      bUseBrushTransparency=false
      roundingType=HalfHeightRadius
      width=0
      color {colorUseRule=UseColor_Specified, specifiedColor=<rgba 0 0 0 0>}
      cornerRadii=<0 0 0 0>
     }
     tintColor {colorUseRule=UseColor_Specified, specifiedColor=<rgba 0.5776 1 0.0003 0.85>}
     uVRegion {bIsValid=false, max=<0 0>, min=<0 0>}
    }
    renderTransformPivot=<0.5 1>
   }
   slot {class=OverlaySlot, properties {horizontalAlignment=HAlign_Fill, verticalAlignment=VAlign_Center, padding=<ltrb 250 0 0 0>}}
  }
  TouchOnly_Size {
   widgetClassPath=&CommonHardwareVisibilityBorder
   parent=Overlay_2
   widget=TouchOnly_Size
   properties {visibilityQuery {
     autoDescription=" ANY( Input.Touch )"
     tokenStreamVersion=0
     userDescription=touch-only
     queryTokenStream [0,1,1,1,0]
     tagDictionary [{tagName=Input.Touch}]
    }}
   slot {class=OverlaySlot, properties {horizontalAlignment=HAlign_Center, verticalAlignment=VAlign_Center}}
  }
  touch_size {
   widgetClassPath=&Spacer
   parent=TouchOnly_Size
   widget=touch_size
   properties {size=<0 80>}
   slot {class=BorderSlot, properties {padding=<ltrb 0 0 0 0>}}
  }
  Image_Bg {
   widgetClassPath=&Image
   bIsVariable=true
   parent=Overlay_2
   widget=Image_Bg
   properties {brush {
     bIsDynamicallyLoaded=false
     drawAs=Image
     imageType=NoImage
     mirroring=NoMirror
     resourceName=None
     resourceObject=/Game/UI/DiscoveryShared/ModeSet/Materials/MI_UI_SidebarStyle_Button_CategoryInst
     tiling=NoTile
     imageSize=<32 32>
     margin=<ltrb 0 0 0 0>
     outlineSettings {
      bUseBrushTransparency=false
      roundingType=HalfHeightRadius
      width=0
      color {colorUseRule=UseColor_Specified, specifiedColor=<rgba 0 0 0 0>}
      cornerRadii=<0 0 0 0>
     }
     tintColor {colorUseRule=UseColor_Specified, specifiedColor=<rgba 1 1 1 1>}
     uVRegion {bIsValid=false, max=<0 0>, min=<0 0>}
    }}
   slot {class=OverlaySlot, properties {horizontalAlignment=HAlign_Fill, verticalAlignment=VAlign_Fill}}
  }
  SafeZone {
   widgetClassPath=&SafeZone
   bIsVariable=true
   parent=Overlay_2
   widget=SafeZone
   properties {padBottom=false, padLeft=false, padTop=false}
   slot {class=OverlaySlot, properties {horizontalAlignment=HAlign_Fill, verticalAlignment=VAlign_Center, padding=<ltrb 20 0 20 0>}}
  }
  ContentHB {widgetClassPath=&HorizontalBox, parent=SafeZone, widget=ContentHB, slot {class=SafeZoneSlot}}
  HorizontalBox_0 {
   widgetClassPath=&HorizontalBox
   parent=ContentHB
   widget=HorizontalBox_0
   slot {class=HorizontalBoxSlot, properties {horizontalAlignment=HAlign_Left, verticalAlignment=VAlign_Center, size {sizeRule=Fill, value=1}}}
  }
  Border_Title {
   widgetClassPath=&Border
   parent=HorizontalBox_0
   widget=Border_Title
   properties {
    background {
     bIsDynamicallyLoaded=false
     drawAs=NoDrawType
     imageType=NoImage
     mirroring=NoMirror
     resourceName=None
     resourceObject=null
     tiling=NoTile
     imageSize=<32 32>
     margin=<ltrb 0 0 0 0>
     outlineSettings {
      bUseBrushTransparency=false
      roundingType=HalfHeightRadius
      width=0
      color {colorUseRule=UseColor_Specified, specifiedColor=<rgba 0 0 0 0>}
      cornerRadii=<0 0 0 0>
     }
     tintColor {colorUseRule=UseColor_Specified, specifiedColor=<rgba 1 1 1 1>}
     uVRegion {bIsValid=false, max=<0 0>, min=<0 0>}
    }
    padding=<ltrb 0 0 0 0>
   }
   slot {class=HorizontalBoxSlot, properties {verticalAlignment=VAlign_Center, size {sizeRule=Fill, value=1}}}
  }
  Text_Title {
   widgetClassPath=&CommonTextBlock
   bIsVariable=true
   parent=Border_Title
   widget=Text_Title
   properties {
    justification=Center
    mobileFontSizeMultiplier=1.3
    scrollStyle=&TextScrollStyle-Base_C
    text=Ranked
    textTransformPolicy=ToUpper
    colorAndOpacity {colorUseRule=UseColor_Specified, specifiedColor=<rgba 0.5776 1 0 1>}
    font {
     bForceMonospaced=false
     bMaterialIsStencil=false
     fontMaterial=null
     fontObject=&HeadingNow
     letterSpacing=0
     monospacedWidth=1
     size=22.5
     skewAmount=0
     typefaceFontName=76 Bold Italic
     outlineSettings {
      bApplyOutlineToDropShadows=false
      bMiteredCorners=false
      bSeparateFillAlpha=false
      outlineMaterial=null
      outlineSize=0
      outlineColor=<rgba 0 0 0 1>
     }
    }
    shadowOffset=<0 0>
   }
   slot {class=BorderSlot, properties {padding=<ltrb 0 0 0 0>}}
  }
  HabaneroWidget {
   widgetClassPath=&/WBP_Habanero_Icon
   bInherited=true
   bIsVariable=true
   parent=HorizontalBox_0
   widget=HabaneroWidget
   properties {bInitializeToRepresentLocalUser=true, bSupportsActivationFocus=false, colorAndOpacity=<rgba 1 1 1 0.5>, iconMinSize=<75 75>}
   slot {class=HorizontalBoxSlot, properties {horizontalAlignment=HAlign_Left, padding=<ltrb 12 0 0 0>}}
  }
  InputActionWidget {
   widgetClassPath=&CommonActionWidget
   bInherited=true
   bIsVariable=true
   parent=ContentHB
   widget=InputActionWidget
   properties {
    progressMaterialParam=Percentage
    progressMaterialBrush {
     bIsDynamicallyLoaded=false
     drawAs=Image
     imageType=NoImage
     mirroring=NoMirror
     resourceName=None
     resourceObject=/Game/UI/Materials/HollowCircularProgressBar
     tiling=NoTile
     imageSize=<36 36>
     margin=<ltrb 0 0 0 0>
     outlineSettings {
      bUseBrushTransparency=false
      roundingType=HalfHeightRadius
      width=0
      color {colorUseRule=UseColor_Specified, specifiedColor=<rgba 0 0 0 0>}
      cornerRadii=<0 0 0 0>
     }
     tintColor {colorUseRule=UseColor_Specified, specifiedColor=<rgba 1 1 0 1>}
     uVRegion {bIsValid=false, max=<0 0>, min=<0 0>}
    }
   }
   slot {class=HorizontalBoxSlot, properties {horizontalAlignment=HAlign_Right, verticalAlignment=VAlign_Center, padding=<ltrb 0 0 5 0>}}
  }
  TextToggle {
   widgetClassPath=/Game/UI/DiscoveryShared/Status/TextToggle
   bIsVariable=true
   parent=ContentHB
   widget=TextToggle
   properties {offText=Off, onText=On}
   slot {class=HorizontalBoxSlot, properties {horizontalAlignment=HAlign_Right, verticalAlignment=VAlign_Center}}
  }
  Image_Dimmer {
   widgetClassPath=&Image
   bIsVariable=true
   parent=Overlay_2
   widget=Image_Dimmer
   properties {renderOpacity=0, visibility=SelfHitTestInvisible, colorAndOpacity=<rgba 0 0.0078 0.0392 0.8>}
   slot {class=OverlaySlot, properties {horizontalAlignment=HAlign_Fill, verticalAlignment=VAlign_Fill}}
  }
 ]
}
