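# Base holdable button for the Ranked (Habanero) lobby, with named content slots and a left panel that slides open and fades in.
# Hover expands it, unhover or unfocus reverses the animation at higher speed and starts a timer to hide it, and hold toggles it.
# CollapseException can block collapse on touch.
# HandleAudio offsets each expand or collapse sound by animation progress, optionally fading out the opposite one.
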WidgetBlueprint /Game/UI/Frontend/Lobby/Habanero/WBP_Expandable_Base
animations [Hovered {
  binding_count=3
  length_seconds=0.2
  object=&self:Hovered
  sequence {
   bindings [
    Overlay_LeftInnerTooltip {
     id=#1
     possessed_class=Overlay
     tracks [{
       type=MovieSceneByteTrack
       display_name=Visibility
       sections [{type=MovieSceneByteSection, end_frame=12, start_frame=0, channels [None {type=MovieSceneScriptingByteChannel, keys [[0,1],[0.99,4]]}]}]
      }]
    }
    OverlaySlot_0 {
     id=#2
     parent_id=#3
     possessed_class=OverlaySlot
     tracks [{
       type=MovieSceneMarginTrack
       display_name=Padding
       sections [{
         type=MovieSceneMarginSection
         end_frame=12
         start_frame=0
         channels [
          Left {type=MovieSceneScriptingFloatChannel}
          Top {type=MovieSceneScriptingFloatChannel}
          Right {type=MovieSceneScriptingFloatChannel, keys [[1,-500],[7,24],[12,24]]}
          Bottom {type=MovieSceneScriptingFloatChannel}
         ]
        }]
      }]
    }
    NamedSlot_LeftInnerTooltip {
     id=#3
     possessed_class=NamedSlot
     tracks [
      {
       type=MovieSceneFloatTrack
       display_name=Render Opacity
       sections [{type=MovieSceneFloatSection, channels [None {type=MovieSceneScriptingFloatChannel, keys [[4,0],[12,1]]}]}]
      }
      {
       type=MovieScene2DTransformTrack
       sections [{
         type=MovieScene2DTransformSection
         end_frame=12
         start_frame=0
         channels [
          Translation.X {type=MovieSceneScriptingFloatChannel, const [0,1]}
          Translation.Y {type=MovieSceneScriptingFloatChannel, keys [[0,-20],[7,-20],[12,0]]}
          Angle {type=MovieSceneScriptingFloatChannel, const [0,1]}
          Scale.X {type=MovieSceneScriptingFloatChannel, const [1,1]}
          Scale.Y {type=MovieSceneScriptingFloatChannel, const [1,1]}
          Shear.X {type=MovieSceneScriptingFloatChannel, const [0,1]}
          Shear.Y {type=MovieSceneScriptingFloatChannel, const [0,1]}
         ]
        }]
      }
     ]
    }
   ]
   display_rate {numerator=60}
  }
 }]
graphs [
 BP_OnHoldTriggered {nodes [
   n1 = On Hold Triggered!() -> (then, CurrentInputType:ECommonInputType Enum) {is=title K2Node_FunctionEntry}
   > n2 = Branch(Condition=Get bIsExtPartCollapsed() -> (bIsExtPartCollapsed:bool))
   > n3 = ExpandExtPart()
   Return Node!(
    execute=[@n3.then, @n4.then]
    ReturnValue:Event Reply Structure=Select(Option 0:Event Reply Structure=Unhandled(), Option 1:Event Reply Structure=Handled(), Index="Equal (Enum)"(A=@n1, B=Touch)) -> (ReturnValue:Event Reply Structure)
   )
   n4 = CollapseExtPart(execute=@n2.else)
  ]}
 CollapseExtPart {nodes [
   CollapseExtPart() {is=title K2Node_FunctionEntry}
   n1 = IsUsingTouch(PlayerController=GetOwningPlayer())
   n2 = Get CollapseException() -> (CollapseException:bool)
   > Branch(Condition=NOR Boolean!(A:bool=Get bIsExtPartCollapsed(), B:bool=AND Boolean(A=@n2, B=@n1)) -> (ReturnValue:bool))
   > PlayAnimationReverse(InAnimation=Get Hovered() -> (Hovered:Widget Animation Object Reference), PlaybackSpeed=4)
   > HandleAudio()
   > RequestTransition!(
    execute
    self:UIKit Block Object Reference=Get Block_Outline() -> (Block_Outline:WBP UIKit Block Outline Object Reference)
    State:EUIKitBlockVisualState Enum=Unfocused
   ) -> (then)
   > n3 = Set Timer by Event(
    Delegate=Create Event()
    Time=GetAnimationCurrentTime!(self:User Widget Object Reference, InAnimation:Widget Animation Object Reference=Get Hovered()) -> (ReturnValue:float)
    bLooping=false
   )
   > Set TimerHandler(TimerHandler:Timer Handle Structure=@n3) -> (Output_Get:Timer Handle Structure)
   > Set with Broadcast bIsExtPartCollapsed!(execute, bIsExtPartCollapsed:bool=true) -> (then, Output_Get:bool)
   > Return Node!(execute)
  ]}
 EventGraph {nodes [
   HideExpandableBox() {is=title K2Node_CustomEvent}
   > SetVisibility(
    self=[
     Get Overlay_LeftInnerTooltip() -> (Overlay_LeftInnerTooltip:Overlay Object Reference)
     Get NamedSlot_LeftInnerTooltip() -> (NamedSlot_LeftInnerTooltip:Named Slot Object Reference)
    ]
    InVisibility=Collapsed
   )
   Event OnInitialized()
   > Parent: OnInitialized() {is=title K2Node_CallParentFunction}
   > Set Total Anim Length(Total Anim Length:double=GetEndTime(self=Get Hovered())) -> (Output_Get:double)
   Event On Clicked()
   n1 = Get Expand_ClickedSound() -> (Expand_ClickedSound:Sound Base Object Reference)
   > Is Valid(InputObject=@n1)
   > PlaySound2D(Sound=@n1)
   Event On Hovered()
   > ExpandExtPart()
   n2 = Event On Unhovered()
   CollapseExtPart(execute=[@n2.then, @n3.then]) {is=title K2Node_CallFunction}
   Event OnAnimationFinished!() -> (OutputDelegate, then, Animation:Widget Animation Object Reference) {is=title K2Node_Event}
   > Call On Animation Finished Event!(execute:Exec, self:WBP Expandable Base Object Reference) -> (then:Exec) {is=title K2Node_CallDelegate}
   n3 = Event On Unfocused() {is=title K2Node_Event}
  ]}
 ExpandExtPart {nodes [
   ExpandExtPart() {is=title K2Node_FunctionEntry}
   > Branch(Condition=Get bIsExtPartCollapsed())
   > Clear and Invalidate Timer by Handle(Handle=Get TimerHandler() -> (TimerHandler:Timer Handle Structure))
   > PlayAnimationForward(InAnimation=Get Hovered())
   > HandleAudio()
   > RequestTransition(self=Get Block_Outline(), State=Focused)
   > SetVisibility(self=[Get NamedSlot_LeftInnerTooltip(), Get Overlay_LeftInnerTooltip()], InVisibility=SelfHitTestInvisible)
   > Set with Broadcast bIsExtPartCollapsed(bIsExtPartCollapsed=false)
   > Return Node()
  ]}
 HandleAudio {nodes [
   n1 = Get Remaining Time() -> (Remaining Time:double)
   n2 = Get Total Anim Length() -> (Total Anim Length:double)
   n3 = Get Sound Start Time() -> (Sound Start Time:double)
   PrintString!(
    execute
    InString=Append!(
     A="Remaining Time: "
     B="To String (Float)"(InDouble=@n1)
     C:String=" || Total Time: "
     D:String="To String (Float)"(InDouble=@n2)
     E:String="|| Sound Start Time: "
     F:String="To String (Float)"(InDouble=@n3)
    ) -> (ReturnValue)
    bPrintToScreen
    bPrintToLog
    TextColor
    Duration
    Key
   ) -> (then)
   > PrintString(InString=Append!(A="Sound Start Time: ", B) -> (ReturnValue))
   HandleAudio() {is=title K2Node_FunctionEntry}
   n4 = Get Hovered()
   n5 = GetEndTime(self=@n4)
   n6 = GetAnimationCurrentTime(InAnimation=@n4)
   n7 = GetAnimationCurrentTime(InAnimation=@n4)
   n8 = GetStartTime!(self:Widget Animation Object Reference=@n4) -> (ReturnValue:float)
   > n9 = Set Remaining Time(Remaining Time:double=Select(Option 0:double=float - float(A=@n7, B=@n8), Option 1:double=float - float(A=@n5, B=@n6), Index=Get bIsExtPartCollapsed()) -> (ReturnValue:double)) -> (Output_Get:double)
   > Set Sound Start Time(Sound Start Time:double=float - float(A=Get Total Anim Length(), B=@n9)) -> (Output_Get:double)
   > n10 = CreateSound2D!(
    execute
    Sound:Sound Base Object Reference=Select(
     Option 0:Sound Base Object Reference=Get Expand_UnhoveredSound() -> (Expand_UnhoveredSound:Sound Base Object Reference)
     Option 1:Sound Base Object Reference=Get Expand_HoveredSound() -> (Expand_HoveredSound:Sound Base Object Reference)
     Index=Get bIsExtPartCollapsed()
    ) -> (ReturnValue:Sound Base Object Reference)
    VolumeMultiplier:float=1
    PitchMultiplier:float=1
    StartTime:float=Get Sound Start Time()
    ConcurrencySettings:Sound Concurrency Object Reference
    bPersistAcrossLevelTransition:bool=false
    bAutoDestroy:bool=true
   ) -> (then, ReturnValue:Audio Component Object Reference)
   > FadeIn!(
    execute
    self:Audio Component Object Reference=@n10
    FadeInDuration:float=Get FadeInTime() -> (FadeInTime:double)
    FadeVolumeLevel:float=1
    StartTime:float=Get Sound Start Time()
    FadeCurve:EAudioFaderCurve Enum=Logarithmic
   ) -> (then)
   > n11 = Branch(Condition=Get bIsExtPartCollapsed())
   > n12 = Set AudioComponent_Hovered(AudioComponent_Hovered:Audio Component Object Reference=@n10) -> (Output_Get:Audio Component Object Reference)
   Branch(execute=[@n12.then, @n13.then], Condition=Get StopPrevSound() -> (StopPrevSound:bool))
   > FadeOut!(
    execute
    self:Audio Component Object Reference=Select(
     Option 0:Audio Component Object Reference=Get AudioComponent_Hovered() -> (AudioComponent_Hovered:Audio Component Object Reference)
     Option 1:Audio Component Object Reference=Get AudioComponent_Unhovered() -> (AudioComponent_Unhovered:Audio Component Object Reference)
     Index=Get bIsExtPartCollapsed()
    ) -> (ReturnValue:Audio Component Object Reference)
    FadeOutDuration:float=Get FadeOutTime() -> (FadeOutTime:double)
    FadeVolumeLevel:float=0
    FadeCurve:EAudioFaderCurve Enum=Sin
   ) -> (then)
   n13 = Set AudioComponent_Unhovered(execute=@n11.else, AudioComponent_Unhovered:Audio Component Object Reference=@n10) -> (Output_Get:Audio Component Object Reference)
  ]}
 OnAnimationFinishedEvent {nodes [OnAnimationFinishedEvent() {is=title K2Node_FunctionEntry}]}
]
variables [
 AudioComponent_Hovered {type=AudioComponent, default=null}
 AudioComponent_Unhovered {type=AudioComponent, default=null}
 bIsExtPartCollapsed {type=boolean, default=true}
 CollapseException {type=boolean}
 Expand_ClickedSound {type=SoundBase, default=/Game/Sounds/Fort_UI_Sounds/FrontEnd/New_FrontEnd/Ranked/Fortnite_UI_Ranked_Button_Default_Cue}
 Expand_HoveredSound {type=SoundBase, default=/Game/Sounds/Fort_UI_Sounds/FrontEnd/New_FrontEnd/Ranked/Fortnite_UI_Ranked_Lobby_Button_Hover_Cue}
 Expand_UnhoveredSound {type=SoundBase, default=/Game/Sounds/Fort_UI_Sounds/FrontEnd/New_FrontEnd/Ranked/Fortnite_UI_Ranked_Lobby_Button_Unhover_Cue}
 FadeInTime {type=number, default=0.05}
 FadeOutTime {type=number, default=0.1}
 OnAnimationFinishedEvent {}
 RankInfo {type=FortHabaneroDisplayData, default=/Game/UI/Habanero/HabaneroDisplayData}
 StopPrevSound {type=boolean, default=true}
 TimerHandler {type=TimerHandle}
 Total Anim Length {type=number}
]
widgets {
 info {inheritedWidgetCount=1, parentClass=/Script/FortniteUI.UIKitHoldableModularButton, rootWidgetClass=&GridPanel, widgetCount=12}
 widgets [
  Root {widgetClassPath=&GridPanel, bIsVariable=true, widget=Root, properties {columnFill [1], rowFill [1]}}
  WBP_CaptureForPostBufferUpdate {
   widgetClassPath=&WBP_CaptureForPostBufferUpdate
   bIsVariable=true
   parent=Root
   slot {class=GridSlot, properties {layer=-10}}
  }
  NamedSlot_UnderContent {
   widgetClassPath=&NamedSlot
   bIsVariable=true
   parent=Root
   widget=NamedSlot_UnderContent
   slot {class=GridSlot, properties {layer=-6}}
  }
  UIKit_Backplate {
   widgetClassPath=&WBP_UIKit_Backplate
   bIsVariable=true
   parent=Root
   widget=UIKit_Backplate
   properties {brightness=Medium, cornerRadius=Interactive Component}
   slot {class=GridSlot, properties {columnSpan=1, rowSpan=1}}
  }
  NamedSlot_Main {widgetClassPath=&NamedSlot, bIsVariable=true, parent=Root, widget=NamedSlot_Main, slot {class=GridSlot, properties {layer=1}}}
  GridPanel_0 {
   widgetClassPath=&GridPanel
   parent=Root
   widget=GridPanel_0
   properties {columnFill [1], rowFill [1]}
   slot {class=GridSlot, properties {layer=5, padding=<ltrb 8 8 8 8>}}
  }
  Overlay_LeftInnerTooltip {
   widgetClassPath=&Overlay
   bIsVariable=true
   parent=GridPanel_0
   widget=Overlay_LeftInnerTooltip
   properties {clipping=ClipToBounds, visibility=Collapsed}
   slot {class=GridSlot}
  }
  NamedSlot_LeftInnerTooltip {
   widgetClassPath=&NamedSlot
   bIsVariable=true
   parent=Overlay_LeftInnerTooltip
   widget=NamedSlot_LeftInnerTooltip
   slot {class=OverlaySlot, properties {verticalAlignment=VAlign_Fill}}
  }
  SizeBox_Content {
   widgetClassPath=&SizeBox
   bIsVariable=true
   parent=GridPanel_0
   widget=SizeBox_Content
   properties {bOverride_HeightOverride=true, bOverride_WidthOverride=true, heightOverride=88, widthOverride=88}
   slot {class=GridSlot, properties {column=1}}
  }
  NamedSlot_Content {widgetClassPath=&NamedSlot, bIsVariable=true, parent=SizeBox_Content, widget=NamedSlot_Content, slot {class=SizeBoxSlot}}
  Block_Outline {
   widgetClassPath=/FortUILibrary/UIKit/Buttons/Blocks/WBP_UIKit_Block_Outline
   bIsVariable=true
   parent=Root
   widget=Block_Outline
   properties {bottomLeftCornerRadius=30, bottomRightCornerRadius=30, radiusInPixels=true, topLeftCornerRadius=30, topRightCornerRadius=30}
   slot {class=GridSlot, properties {layer=10}}
  }
  InputActionWidget {widgetClassPath=&CommonActionWidget, bInherited=true, widget=null}
 ]
}
